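# Niagara system for the Thrive character body cosmetic, adding flowing ribbons and small floating mesh blobs along the sleeves.
# BoneFlow_Sleeves bursts partitioned ribbons that FN_BoneFlow continually wraps around offset sleeve skeletal-mesh bones.
# Sleeves_Blobas spawns meshes on similar paths, which then rise with curl noise, shrink and follow owner movement.
# Both tint by time of day; User.bisHandHidden zeroes ribbon alpha and switches the blobs' visibility tag.
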NiagaraSystem /BRCosmetics/Characters/Player/Female/Medium/Bodies/F_MED_Thrive/FX/NS_Thrive_Body
dependencies {
 usedDataInterfaces [
  &NiagaraDataInterfaceCurve
  &NiagaraDataInterfaceArrayInt32
  &NiagaraDataInterfaceArrayFloat3
  &NiagaraDataInterfaceColorCurve
  &NiagaraDataInterfaceCurlNoise
 ]
 usedDynamicInputs []
 usedModules [&SystemState]
}
emitters {
 BoneFlow_Sleeves {
  emitterSpawnScript [
   /Game/Effects/Niagara/Modules/Partition/FN_PartitionAttributes(Number of Bone Sets=1, Particles Per Partition=10, Partitions=2)
   /SetVariables() {isSetParameters=true, module=SetVariables_#1}
  ]
  emitterUpdateScript [
   &EmitterState(Life Cycle Mode=::NewEnumerator1, Loop Behavior=::NewEnumerator1, Loop Duration Mode=::NewEnumerator1, UseLoopDelay=false)
   &SpawnBurst_Instantaneous(Spawn Count=$Emitter.FN_PartitionAttributes.Spawn Total, Use Spawn Probability=false)
  ]
  particleSpawnScript [
   /Game/Effects/Niagara/Modules/Partition/FN_Particle_partition_v5/FN_Particle_Partition_v5(Enable Ribbon ID=true, Pattern:ENiagaraPartitionWaveForm=::NewEnumerator1, Seed:NiagaraInt32=$Emitter.Seed, Shuffle Partition ID=true)
   &InitializeRibbon(
    Module.Write Color=true
    Module.Write Lifetime=true
    Module.Write Mass=false
    Module.Write Material Random=true
    Module.Write Position=false
    Module.Write RibbonTwist=false
    Module.Write RibbonWidth=true
    Ribbon Width=10
   )
   /Game/Effects/Niagara/Modules/BoneFlow/FN_Initialize_BoneFlow_Attributes_v2(
    Angle:NiagaraFloat=&Multiply_Float(
     A=$Particles.Partition.NormalizedID
     B=&UniformRangedFloat(Minimum=0, Maximum=0.5, Random Seed=$Particles.Partition.ID)
    )
    Radius:NiagaraFloat=&FloatFromCurve(
     FloatCurve=DataInterface() {propertyValues {Curve {keys [
         {arriveTangent=77401.0078, interpMode=RCIM_Cubic, leaveTangent=50.7651, tangentMode=RCTM_User, time=0, value=0}
         {arriveTangent=-2.162, interpMode=RCIM_Cubic, leaveTangent=-2.162, tangentMode=RCTM_User, time=0.25, value=5.5786}
         <at 0.5 0 RCIM_Cubic>
         <at 0.75 5.5786 RCIM_Cubic>
         {arriveTangent=-40.3134, interpMode=RCIM_Cubic, leaveTangent=-1467.5861, tangentMode=RCTM_User, time=1, value=0}
        ]}}}
     CurveIndex=$Particles.Partition.NormalizedID
     Scale Curve=0
    )
    T:NiagaraFloat=$Particles.Partition.NormalizedID
   )
   &FN_FlipRibbonV(Seed:NiagaraInt32=$Particles.Partition.ID)
  ]
  particleUpdateScript [
   &ParticleState(Kill Particles When Lifetime Has Elapsed=false, Let Infinitely Lived Particles Die When Emitter Deactivates=true)
   &FN_BoneFlow(
    Angle:NiagaraFloat=&Add_Float(
     A=&Multiply_Float(A=$Particles.Initial.FN_Initialize_BoneFlow_Attributes_v2.Angle, B=10)
     B=&Multiply_Float(A=$Emitter.Age, B=0.5)
    )
    Bone Offsets=DataInterface() {propertyValues {
      GpuSyncMode=SyncCpuToGpu
      MaxElements=0
      FloatData [<-5 0 0>, <0 8 -3>, <-5 0 2>, <8 -12 0>, <-5 -3 -10>, <-6 0 0>, <0 -10 3>, <12 -2 0>, <-7 15 0>, <-10 0 500>]
     }}
    Bone Sets=DataInterface() {propertyValues {GpuSyncMode=SyncCpuToGpu, MaxElements=0, IntData [0,4,5,9]}}
    Enable Position Offsets=true
    Height:NiagaraFloat=$Particles.Partition.NormalizedID
    Radius:NiagaraFloat=$Particles.Initial.FN_Initialize_BoneFlow_Attributes_v2.Radius
    SkeletalMesh:NiagaraDataInterfaceSkeletalMesh=$User.SkeletalMeshSleves
   ) {module=FN_BoneFlow001}
   &ScaleRibbonWidth(Ribbon Width Scale=&FloatFromCurve(
     FloatCurve=DataInterface() {propertyValues {Curve {keys [
         {arriveTangent=11.1156, interpMode=RCIM_Cubic, leaveTangent=11.1156, tangentMode=RCTM_User, time=0, value=0}
         <at 0.2 1 RCIM_Cubic>
         {arriveTangent=-13.1608, interpMode=RCIM_Cubic, leaveTangent=-1.0422, tangentMode=RCTM_Break, time=0.4, value=0}
         {arriveTangent=-0.0189, interpMode=RCIM_Cubic, leaveTangent=9.5172, tangentMode=RCTM_Break, time=0.55, value=0}
         <at 0.7 1 RCIM_Cubic>
         {arriveTangent=-13.6494, interpMode=RCIM_Cubic, leaveTangent=-13.6494, tangentMode=RCTM_User, time=0.85, value=0}
        ]}}}
     CurveIndex=$Particles.Partition.NormalizedID
    ))
   &ScaleColor(ScaleA=true, ScaleRGB=true)
   /BRCosmetics/Weapons/FORT_Melee/Pickaxe_Pumpkin_Spice_1h/FX/ColorFromTOD(
    Afternoon:LinearColor=<rgba 1 0.1534 0.3305 0.5>
    Evening:LinearColor=<rgba 0.1545 0.0099 0.0631 0.5>
    FortniteNPC_FortniteActiveTimeOfDay:LinearColor=$NPC.FortniteNPC.FortniteActiveTimeOfDay
    Morning:LinearColor=<rgba 0.2656 0.048 0.1538 0.5>
    Night:LinearColor=<rgba 0.2483 0.099 0.1702 0.5>
   )
   /SetVariables(Particles.Color=&MakeLinearColorFromVectorAndFloat(
     "Vector (RGB)"=&MakeVectorFromLinearColorRGB(Color=$Particles.Color)
     "Float (Alpha)"=&Multiply_Float(
      A=&MakeFloatFromLinearColor(LinearColor=$Particles.Color, Channel:ENiagaraFloatFromLinearColorOptions=::NewEnumerator3)
      B=&MakeCustomFloatFromBool(Bool=$User.bisHandHidden, True Float=0, False Float=1)
     )
    )) {isSetParameters=true}
  ]
  rendererClasses [&NiagaraRibbonRendererProperties]
  renderers [{rendererClass=&NiagaraRibbonRendererProperties}]
 }
 Sleeves_Blobas {
  emitterSpawnScript [
   &FN_PartitionAttributes(Number of Bone Sets=1, Particles Per Partition=10)
   /SetVariables() {isSetParameters=true, module=SetVariables_#1}
  ]
  emitterUpdateScript [
   &EmitterState(Life Cycle Mode=::NewEnumerator1, Loop Behavior=::NewEnumerator1, Loop Duration Mode=::NewEnumerator1, UseLoopDelay=false)
   &SpawnRate(SpawnRate=5, Use Spawn Probability=false)
  ]
  particleSpawnScript [
   &FN_Particle_Partition_v5(Enable Ribbon ID=true, Pattern:ENiagaraPartitionWaveForm=::NewEnumerator1, Seed:NiagaraInt32=$Emitter.Seed, Shuffle Partition ID=true)
   &InitializeParticle(Mesh Scale Mode=::NewEnumerator3, Mesh Uniform Scale=0.04, UsePositionOffset=false, Write Material Random=true)
   &InitializeRibbon(
    Color=<rgba 1 0.1534 0.3305 1>
    Module.Write Color=true
    Module.Write Lifetime=true
    Module.Write Mass=false
    Module.Write Material Random=true
    Module.Write Position=false
    Module.Write RibbonTwist=false
    Module.Write RibbonWidth=true
    Ribbon Width=10
   ) {enabled=false}
   &FN_Initialize_BoneFlow_Attributes_v2(
    Angle:NiagaraFloat=&Multiply_Float(
     A=$Particles.Partition.NormalizedID
     B=&UniformRangedFloat(Minimum=0, Maximum=0.5, Random Seed=$Particles.Partition.ID)
    )
    Radius:NiagaraFloat=&FloatFromCurve(
     FloatCurve=DataInterface() {propertyValues {Curve {keys [
         {arriveTangent=77401.0078, interpMode=RCIM_Cubic, leaveTangent=50.7651, tangentMode=RCTM_User, time=0, value=0}
         {arriveTangent=16.47, interpMode=RCIM_Cubic, leaveTangent=16.47, time=0.13, value=5.5786}
         <at 0.4924 8.1096 RCIM_Cubic>
         {arriveTangent=-15.9761, interpMode=RCIM_Cubic, leaveTangent=-15.9761, time=0.8801, value=5.5786}
         {arriveTangent=-40.3134, interpMode=RCIM_Cubic, leaveTangent=-1467.5861, tangentMode=RCTM_User, time=1, value=0}
        ]}}}
     CurveIndex=$Particles.Partition.NormalizedID
     Scale Curve=0
    )
    T:NiagaraFloat=$Particles.Partition.NormalizedID
   )
   &FN_BoneFlow(
    Angle:NiagaraFloat=&Add_Float(
     A=&Multiply_Float(A=$Particles.Initial.FN_Initialize_BoneFlow_Attributes_v2.Angle, B=10)
     B=&Multiply_Float(A=$Emitter.Age, B=0.5)
    )
    Bone Offsets=DataInterface() {propertyValues {
      GpuSyncMode=SyncCpuToGpu
      MaxElements=0
      FloatData [<-5 0 0>, <0 8 -3>, <-5 0 2>, <8 -12 0>, <-5 -3 -10>, <-5 0 0>, <0 -10 3>, <7 -2 0>, <-5 8 0>, <5 3 0>]
     }}
    Bone Sets=DataInterface() {propertyValues {GpuSyncMode=SyncCpuToGpu, MaxElements=0, IntData [0,4,5,9]}}
    Enable Position Offsets=true
    Height:NiagaraFloat=$Particles.Partition.NormalizedID
    Radius:NiagaraFloat=$Particles.Initial.FN_Initialize_BoneFlow_Attributes_v2.Radius
    SkeletalMesh:NiagaraDataInterfaceSkeletalMesh=$User.SkeletalMeshSleves
   ) {module=FN_BoneFlow001}
   &ColorFromTOD(
    Afternoon:LinearColor=<rgba 1 0.1534 0.3305 0.5>
    Evening:LinearColor=<rgba 0.1545 0.0099 0.0631 0.5>
    FortniteNPC_FortniteActiveTimeOfDay:LinearColor=$NPC.FortniteNPC.FortniteActiveTimeOfDay
    Morning:LinearColor=<rgba 0.2656 0.048 0.1538 0.5>
    Night:LinearColor=<rgba 0.2483 0.099 0.1702 0.5>
   )
  ]
  particleUpdateScript [
   &ParticleState()
   &InheritSourceMovement(
    Applied Position Scale=&Multiply_VectorByFloat(
     Vector=<0.99 0.99 0.99>
     Float=&FloatFromCurve(
      FloatCurve=DataInterface() {propertyValues {Curve {keys [<at 0 0.8>,<at 1 1>]}}}
      CurveIndex=&NormalizeFloat(Value=&VectorLength(Vector=$Engine.Owner.Velocity), Min=100, Max=500)
     )
    )
    ApplyForce=false
    ApplyPosition=true
    ApplyPositionFalloff=false
   )
   &GravityForce(Gravity=<0 0 50>)
   &CurlNoiseForce(Noise Frequency=100, Pan Noise=false)
   &Drag(Use Linear Drag=true, Use Rotational Drag=false)
   &SolveForcesAndVelocity(Clamp Velocity=false, Limit Acceleration=false)
   &ScaleMeshSize(Scale Factor=&VectorFromFloat(Value=&FloatFromCurve(FloatCurve=DataInterface() {propertyValues {Curve {keys [
          {arriveTangent=9.846, interpMode=RCIM_Cubic, leaveTangent=9.846, tangentMode=RCTM_User, time=0, value=0}
          {arriveTangent=-2.8092, interpMode=RCIM_Cubic, leaveTangent=-2.8092, tangentMode=RCTM_User, time=0.1, value=1}
          {interpMode=RCIM_Cubic, tangentMode=RCTM_User, time=1, value=0}
         ]}}})))
   /SetVariables(Particles.VisibilityTag=&MakeCustomIntFromBool(Bool=$User.bisHandHidden, True INT=1, False INT=0)) {isSetParameters=true}
  ]
  rendererClasses [&NiagaraMeshRendererProperties]
  renderers [{rendererClass=&NiagaraMeshRendererProperties, rendererIndex=0}]
 }
}
summary {
 systemName=NS_Thrive_Body
 emitters [
  {emitterName=BoneFlow_Sleeves, rendererClasses [&NiagaraRibbonRendererProperties]}
  {emitterName=Sleeves_Blobas, rendererClasses [&NiagaraMeshRendererProperties]}
 ]
 userVariables [
  User.bisHandHidden {type=NiagaraBool}
  User.BoneFlowSmallSpawn {type=NiagaraInt32}
  User.SkeletalMeshSleves {
   type=NiagaraDataInterfaceSkeletalMesh
   defaultValue {
    struct=NiagaraExt_VariableValue_DataInterface
    value {dataInterface=&self:NiagaraDataInterfaceSkeletalMesh_0}
   }
  }
 ]
}
