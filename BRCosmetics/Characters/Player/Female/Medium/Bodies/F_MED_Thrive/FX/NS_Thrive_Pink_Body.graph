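# Niagara body effect for the pink variant of the female Thrive character cosmetic, dressing arms, sleeves, hip bow, necklace and center.
# Bone-flow and attached ribbons, rising curl-noise blobs, necklace glow sprites and a camera-offset light follow per-part skeletal mesh inputs.
# System.Thrive_NormalizedLevel drives reactive color, glow, light strength, some spawn rates and arm ribbon size.
# Time of day tints blobs and ribbons; a hand-hidden flag controls arm and sleeve visibility.
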NiagaraSystem /BRCosmetics/Characters/Player/Female/Medium/Bodies/F_MED_Thrive/FX/NS_Thrive_Pink_Body
dependencies {
 usedDataInterfaces [
  &NiagaraDataInterfaceCurve
  &NiagaraDataInterfaceArrayInt32
  &NiagaraDataInterfaceArrayFloat3
  &NiagaraDataInterfaceColorCurve
  &NiagaraDataInterfaceCurlNoise
  &NiagaraDataInterfaceArrayFloat
  &NiagaraDataInterfaceCamera
 ]
 usedDynamicInputs [&Lerp_Float, &FN_Select_Color_from_Time_Of_Day]
 usedModules [
  &self:SystemSpawnScript.SystemScriptSource.SystemScriptGraph.NiagaraNodeAssignment_0.SetVariables_#1
  &self:SystemSpawnScript.SystemScriptSource.SystemScriptGraph.NiagaraNodeAssignment_2.SetVariables_#2
  &self:SystemSpawnScript.SystemScriptSource.SystemScriptGraph.NiagaraNodeAssignment_1.SetVariables_#3
  &SystemState
 ]
}
emitters {
 Blobs_Arms {
  emitterUpdateScript [
   &EmitterState(Life Cycle Mode=::NewEnumerator1, Loop Duration=1, UseLoopDelay=false)
   &SpawnRate(
    SpawnRate=&Add_Float(A=2, B=&Multiply_Float(A=3, B=$System.Thrive_NormalizedLevel))
    Use Spawn Probability=false
   )
  ]
  particleSpawnScript [
   &FN_SetSpawnGroup(Number of Spawn Groups=2)
   &InitializeParticle(
    Lifetime Mode=::NewEnumerator1
    Mesh Scale Mode=::NewEnumerator3
    Mesh Uniform Scale=0.05
    UsePositionOffset=false
    Write Material Random=true
   )
   &FN_BoneFlow(
    Height:NiagaraFloat=&RandomRangeFloat(Minimum=0, Maximum=1)
    SkeletalMesh:NiagaraDataInterfaceSkeletalMesh=$User.SkeletalMeshArms
   )
   /BRCosmetics/Weapons/FORT_Melee/Pickaxe_Pumpkin_Spice_1h/FX/ColorFromTOD(
    Afternoon:LinearColor=<rgba 1 0.8682 0.8281 0.5>
    Evening:LinearColor=<rgba 0.592 0.4832 0.4615 0.5>
    FortniteNPC_FortniteActiveTimeOfDay:LinearColor=$NPC.FortniteNPC.FortniteActiveTimeOfDay
    Morning:LinearColor=<rgba 0.2656 0.1771 0.1259 0.5>
    Night:LinearColor=<rgba 0.4726 0.6267 0.6927 0.5>
   )
  ]
  particleUpdateScript [
   &ParticleState()
   &InheritSourceMovement(
    Applied Position Scale=&Multiply_VectorByFloat(
     Vector=<0.99 0.99 0.99>
     Float=&FloatFromCurve(
      FloatCurve=DataInterface() {propertyValues {Curve {keys [<at 0 0.8>,<at 1 1>]}}}
      CurveIndex=&NormalizeFloat(Value=&VectorLength(Vector=$Engine.Owner.Velocity), Min=100, Max=500)
     )
    )
    ApplyForce=false
    ApplyPosition=true
    ApplyPositionFalloff=false
   )
   &GravityForce(Gravity=<0 0 25>)
   &CurlNoiseForce(Noise Frequency=100, Pan Noise=false)
   &Drag(Use Linear Drag=true, Use Rotational Drag=false)
   /Game/Effects/Niagara/Modules/BoneFlow/FN_BoneFlow_v4/FN_BoneFlow(Attached=false, BoneID:NiagaraInt32=$Particles.SpawnGroup, Interpolate Mesh Sampling=false, Write Position Attribute=true) {enabled=false, module=FN_BoneFlow001}
   &SolveForcesAndVelocity(Clamp Velocity=false, Limit Acceleration=false)
   &ScaleMeshSize(Scale Factor=&VectorFromFloat(Value=&FloatFromCurve(FloatCurve=DataInterface() {propertyValues {Curve {keys [
          {arriveTangent=9.846, interpMode=RCIM_Cubic, leaveTangent=9.846, tangentMode=RCTM_User, time=0, value=0}
          {arriveTangent=-1.8601, interpMode=RCIM_Cubic, leaveTangent=-1.8601, tangentMode=RCTM_User, time=0.1, value=1}
          {interpMode=RCIM_Cubic, tangentMode=RCTM_User, time=1, value=0}
         ]}}})))
   &ScaleColor(
    Scale RGB=&MakeVectorFromLinearColorRGB(Color=&ColorFromCurve(
      ColorCurve=DataInterface() {propertyValues {
        AlphaCurve {keys [<at 0 1 RCIM_Cubic>]}
        BlueCurve {keys [<at 0 1 RCIM_Cubic>,<at 0.4111 0.6334>,<at 1 6 RCIM_Cubic>]}
        GreenCurve {keys [<at 0 1 RCIM_Cubic>,<at 0.4111 0.166>,<at 1 0 RCIM_Cubic>]}
        RedCurve {keys [<at 0 1 RCIM_Cubic>,<at 0.4111 1>,<at 1 1.8052 RCIM_Cubic>]}
       }}
      CurveIndex=$System.Thrive_NormalizedLevel
     ))
    ScaleA=true
    ScaleRGB=true
   )
   /SetVariables(Particles.VisibilityTag=&MakeCustomIntFromBool(Bool=$User.bisHandHidden, True INT=1, False INT=0)) {isSetParameters=true, module=SetVariables_#4}
  ]
  rendererClasses [&NiagaraSpriteRendererProperties, &NiagaraMeshRendererProperties]
  renderers [
   {rendererClass=&NiagaraSpriteRendererProperties}
   {rendererClass=&NiagaraMeshRendererProperties, rendererIndex=1}
  ]
 }
 BoneFlow_Sleeves {
  emitterSpawnScript [
   /Game/Effects/Niagara/Modules/Partition/FN_PartitionAttributes(Number of Bone Sets=1, Particles Per Partition=10, Partitions=2)
   /SetVariables() {isSetParameters=true, module=SetVariables_#5}
  ]
  emitterUpdateScript [
   &EmitterState(Life Cycle Mode=::NewEnumerator1, Loop Behavior=::NewEnumerator1, Loop Duration Mode=::NewEnumerator1, UseLoopDelay=false)
   &SpawnBurst_Instantaneous(Spawn Count=$Emitter.FN_PartitionAttributes.Spawn Total, Use Spawn Probability=false)
  ]
  particleSpawnScript [
   /Game/Effects/Niagara/Modules/Partition/FN_Particle_partition_v5/FN_Particle_Partition_v5(Enable Ribbon ID=true, Pattern:ENiagaraPartitionWaveForm=::NewEnumerator1, Seed:NiagaraInt32=$Emitter.Seed, Shuffle Partition ID=true)
   &InitializeRibbon(
    Module.Write Color=true
    Module.Write Lifetime=true
    Module.Write Mass=false
    Module.Write Material Random=true
    Module.Write Position=false
    Module.Write RibbonTwist=false
    Module.Write RibbonWidth=true
    Ribbon Width=10
   )
   /Game/Effects/Niagara/Modules/BoneFlow/FN_Initialize_BoneFlow_Attributes_v2(
    Angle:NiagaraFloat=&Multiply_Float(
     A=$Particles.Partition.NormalizedID
     B=&UniformRangedFloat(Minimum=0, Maximum=0.5, Random Seed=$Particles.Partition.ID)
    )
    Radius:NiagaraFloat=&FloatFromCurve(
     FloatCurve=DataInterface() {propertyValues {Curve {keys [
         {arriveTangent=77401.0078, interpMode=RCIM_Cubic, leaveTangent=50.7651, tangentMode=RCTM_User, time=0, value=0}
         {arriveTangent=-2.162, interpMode=RCIM_Cubic, leaveTangent=-2.162, tangentMode=RCTM_User, time=0.25, value=5.5786}
         <at 0.5 0 RCIM_Cubic>
         <at 0.75 5.5786 RCIM_Cubic>
         {arriveTangent=-40.3134, interpMode=RCIM_Cubic, leaveTangent=-1467.5861, tangentMode=RCTM_User, time=1, value=0}
        ]}}}
     CurveIndex=$Particles.Partition.NormalizedID
     Scale Curve=0
    )
    T:NiagaraFloat=$Particles.Partition.NormalizedID
   )
   &FN_FlipRibbonV(Seed:NiagaraInt32=$Particles.Partition.ID)
  ]
  particleUpdateScript [
   &ParticleState(Kill Particles When Lifetime Has Elapsed=false, Let Infinitely Lived Particles Die When Emitter Deactivates=true)
   /Game/Effects/Niagara/Modules/BoneFlow/FN_BoneFlow(
    Angle:NiagaraFloat=&Add_Float(
     A=&Multiply_Float(A=$Particles.Initial.FN_Initialize_BoneFlow_Attributes_v2.Angle, B=10)
     B=&Multiply_Float(A=$Emitter.Age, B=0.5)
    )
    Bone Offsets=DataInterface() {propertyValues {
      GpuSyncMode=SyncCpuToGpu
      MaxElements=0
      FloatData [<-5 0 0>, <0 8 -3>, <-5 0 2>, <8 -12 0>, <-5 -3 -10>, <-5 0 0>, <0 -10 3>, <7 -2 0>, <-5 8 0>, <5 3 0>]
     }}
    Bone Sets=DataInterface() {propertyValues {GpuSyncMode=SyncCpuToGpu, MaxElements=0, IntData [0,4,5,9]}}
    Height:NiagaraFloat=$Particles.Partition.NormalizedID
    Radius:NiagaraFloat=$Particles.Initial.FN_Initialize_BoneFlow_Attributes_v2.Radius
    SkeletalMesh:NiagaraDataInterfaceSkeletalMesh=$User.SkeletalMeshSleves
   ) {enabled=false, module=FN_BoneFlow001}
   &FN_BoneFlow(
    Angle:NiagaraFloat=&Add_Float(
     A=&Multiply_Float(A=$Particles.Initial.FN_Initialize_BoneFlow_Attributes_v2.Angle, B=10)
     B=&Multiply_Float(A=$Emitter.Age, B=0.5)
    )
    Bone Offsets=DataInterface() {propertyValues {
      GpuSyncMode=SyncCpuToGpu
      MaxElements=0
      FloatData [<-5 0 0>, <0 8 -3>, <-5 0 2>, <8 -12 0>, <-5 -3 -10>, <-6 0 0>, <0 -10 3>, <12 -2 0>, <-7 15 0>, <-10 0 500>]
     }}
    Bone Sets=DataInterface() {propertyValues {GpuSyncMode=SyncCpuToGpu, MaxElements=0, IntData [0,4,5,9]}}
    Height:NiagaraFloat=$Particles.Partition.NormalizedID
    Radius:NiagaraFloat=$Particles.Initial.FN_Initialize_BoneFlow_Attributes_v2.Radius
    SkeletalMesh:NiagaraDataInterfaceSkeletalMesh=$User.SkeletalMeshSleves
   )
   &ScaleRibbonWidth(Ribbon Width Scale=&FloatFromCurve(
     FloatCurve=DataInterface() {propertyValues {Curve {keys [
         {arriveTangent=11.1156, interpMode=RCIM_Cubic, leaveTangent=11.1156, tangentMode=RCTM_User, time=0, value=0}
         <at 0.2 1 RCIM_Cubic>
         {arriveTangent=-13.1608, interpMode=RCIM_Cubic, leaveTangent=-1.0422, tangentMode=RCTM_Break, time=0.4, value=0}
         {arriveTangent=-0.0189, interpMode=RCIM_Cubic, leaveTangent=9.5172, tangentMode=RCTM_Break, time=0.55, value=0}
         <at 0.7 1 RCIM_Cubic>
         {arriveTangent=-13.6494, interpMode=RCIM_Cubic, leaveTangent=-13.6494, tangentMode=RCTM_User, time=0.85, value=0}
        ]}}}
     CurveIndex=$Particles.Partition.NormalizedID
    ))
   &ScaleRibbonWidth(Ribbon Width Scale=0.25) {enabled=false, module=ScaleRibbonWidth001}
   &ScaleColor(
    Scale RGB=&MakeVectorFromLinearColorRGB(Color=&ColorFromCurve(
      ColorCurve=DataInterface() {propertyValues {
        AlphaCurve {keys [<at 0 1 RCIM_Cubic>]}
        BlueCurve {keys [<at 0 1 RCIM_Cubic>,<at 1 6 RCIM_Cubic>]}
        GreenCurve {keys [<at 0 1 RCIM_Cubic>,<at 1 0 RCIM_Cubic>]}
        RedCurve {keys [<at 0 1 RCIM_Cubic>,<at 1 1.8052 RCIM_Cubic>]}
       }}
      CurveIndex=$System.Thrive_NormalizedLevel
     ))
    ScaleA=true
    ScaleRGB=true
   )
   &ColorFromTOD(
    Afternoon:LinearColor=<rgba 1 0.1534 0.3305 0.5>
    Evening:LinearColor=<rgba 0.1545 0.0099 0.0631 0.5>
    FortniteNPC_FortniteActiveTimeOfDay:LinearColor=$NPC.FortniteNPC.FortniteActiveTimeOfDay
    Morning:LinearColor=<rgba 0.2656 0.048 0.1538 0.5>
    Night:LinearColor=<rgba 0.2483 0.099 0.1702 0.5>
   )
   &DynamicMaterialParameters(Write Parameter Index 0=false) {enabled=false}
   /SetVariables(Particles.Color=&MakeLinearColorFromVectorAndFloat(
     "Vector (RGB)"=&MakeVectorFromLinearColorRGB(Color=$Particles.Color)
     "Float (Alpha)"=&Multiply_Float(
      A=&MakeFloatFromLinearColor(LinearColor=$Particles.Color, Channel:ENiagaraFloatFromLinearColorOptions=::NewEnumerator3)
      B=&MakeCustomFloatFromBool(Bool=$User.bisHandHidden, True Float=0, False Float=1)
     )
    )) {isSetParameters=true}
  ]
  rendererClasses [&NiagaraRibbonRendererProperties]
  renderers [{rendererClass=&NiagaraRibbonRendererProperties}]
 }
 FN_AttachedRibbon {
  emitterUpdateScript [
   &EmitterState()
   &FN_SpawnRibbon_v2(Number of Ribbons=8, Spawn Mode=::NewEnumerator0, SpawnRate=30)
   &FN_Integrator() {enabled=false}
   &FN_Timer()
  ]
  particleSpawnScript [
   /Game/Effects/Niagara/Modules/Ribbons/FN_InitializeRibbon_v4(
    Enable Skeletal Position=true
    Lifetime:NiagaraFloat=&Multiply_Float(A=0.45, B=$System.Thrive_NormalizedLevel)
    SkeletalMesh:NiagaraDataInterfaceSkeletalMesh=$User.SkeletalMeshArmsRibbons
   )
   &FN_FlipRibbonV()
   &AddVelocity(
    Coordinate Space=::World
    Velocity=&MakeVector(
     X=0
     Y=0
     Z=&SelectFloatFromArray(Direct Array Index=$Particles.SpawnGroup, Float Selection Array=DataInterface() {propertyValues {FloatData [25,50,45,33,50,40,50,30]}})
    )
   )
  ]
  particleUpdateScript [
   &InheritSourceMovement(Applied Position Scale=<0.9 0.9 0.9>, ApplyForce=false, ApplyPosition=true, ApplyPositionFalloff=false)
   &SolveForcesAndVelocity(Clamp Velocity=false, Limit Acceleration=false)
   /Game/Effects/Niagara/Modules/Ribbons/FN_UpdateRibbon_v4(Enable Skeletal Position=true, SkeletalMesh:NiagaraDataInterfaceSkeletalMesh=$User.SkeletalMeshArmsRibbons)
   &DynamicMaterialParameters(
    Index 2 Param 4=1
    Index 3 Param 1=&Add_Float(
     A=&SelectFloatFromArray(Direct Array Index=$Particles.SpawnGroup, Float Selection Array=DataInterface() {propertyValues {FloatData [1,2,3,4,5,6,7,8]}})
     B=$Emitter.FN_Timer.Value
    )
    Module.Param0WriteEnabled2=true
    Module.Param0WriteEnabled3=true
    Module.Param1WriteEnabled2=true
    Module.Param1WriteEnabled3=true
    Module.Param2WriteEnabled2=true
    Module.Param2WriteEnabled3=true
    Module.Param3WriteEnabled2=true
    Module.Param3WriteEnabled3=true
    Write Parameter Index 0=false
    Write Parameter Index 2=true
    Write Parameter Index 3=true
   )
   &ScaleRibbonWidth(Ribbon Width Scale=&FloatFromCurve(
     FloatCurve=DataInterface() {propertyValues {Curve {keys [
         {arriveTangent=1.5171, interpMode=RCIM_Cubic, leaveTangent=1.5171, tangentMode=RCTM_User, time=0, value=0}
         {interpMode=RCIM_Cubic, tangentMode=RCTM_User, time=1, value=1}
        ]}}}
     CurveIndex=$System.Thrive_NormalizedLevel
     Scale Curve=&FloatFromCurve(FloatCurve=DataInterface() {propertyValues {Curve {keys [
          {interpMode=RCIM_Cubic, tangentMode=RCTM_User, time=0, value=1}
          {arriveTangent=-1, interpMode=RCIM_Cubic, leaveTangent=-1, time=0.8, value=1}
          {arriveTangent=-8.9128, interpMode=RCIM_Cubic, leaveTangent=-8.9128, tangentMode=RCTM_User, time=1, value=0}
         ]}}})
    ))
   &ScaleRibbonWidth(Ribbon Width Scale=&FloatFromCurve(FloatCurve=DataInterface() {propertyValues {Curve {keys [
         {interpMode=RCIM_Cubic, tangentMode=RCTM_User, time=0, value=1}
         {arriveTangent=-1, interpMode=RCIM_Cubic, leaveTangent=-1, time=0.8, value=1}
         {arriveTangent=-8.9128, interpMode=RCIM_Cubic, leaveTangent=-8.9128, tangentMode=RCTM_User, time=1, value=0}
        ]}}})) {enabled=false, module=ScaleRibbonWidth001}
   &ScaleColor(
    Scale RGB=&MakeVectorFromLinearColorRGB(Color=&ColorFromCurve(
      ColorCurve=DataInterface() {propertyValues {
        AlphaCurve {keys [<at 0 1 RCIM_Cubic>]}
        BlueCurve {keys [<at 0 1 RCIM_Cubic>,<at 1 6 RCIM_Cubic>]}
        GreenCurve {keys [<at 0 1 RCIM_Cubic>,<at 1 0 RCIM_Cubic>]}
        RedCurve {keys [<at 0 1 RCIM_Cubic>,<at 1 1.8052 RCIM_Cubic>]}
       }}
      CurveIndex=$System.Thrive_NormalizedLevel
     ))
    ScaleA=true
    ScaleRGB=true
   ) {module=ScaleColor001}
   &ColorFromTOD(
    Afternoon:LinearColor=<rgba 1 0.1534 0.3305 0.5>
    Evening:LinearColor=<rgba 0.1545 0.0099 0.0631 0.5>
    FortniteNPC_FortniteActiveTimeOfDay:LinearColor=$NPC.FortniteNPC.FortniteActiveTimeOfDay
    Morning:LinearColor=<rgba 0.2656 0.048 0.1538 0.5>
    Night:LinearColor=<rgba 0.2483 0.099 0.1702 0.5>
   )
  ]
  rendererClasses [&NiagaraRibbonRendererProperties]
  renderers [
   {rendererClass=&NiagaraRibbonRendererProperties}
   {rendererClass=&NiagaraRibbonRendererProperties, rendererIndex=1}
  ]
 }
 HipRibbon_Blobs {
  emitterUpdateScript [
   &EmitterState()
   &SpawnRate(
    SpawnRate=&Add_Float(A=2, B=&Multiply_Float(A=3, B=$System.Thrive_NormalizedLevel))
    Use Spawn Probability=false
   )
  ]
  particleSpawnScript [
   &FN_SetSpawnGroup(Number of Spawn Groups=6)
   &InitializeParticle(
    Lifetime Mode=::NewEnumerator1
    Mesh Scale Mode=::NewEnumerator3
    Mesh Uniform Scale=0.045
    UsePositionOffset=false
    Write Material Random=true
   )
   &SkeletalMeshLocation(Bone / Socket Index=$Particles.SpawnGroup, Bone Sampling Mode=::NewEnumerator4, MaskSpawnGroup=false, Skeletal Mesh=$User.SkeletalMeshBow)
   &AddVelocity(
    Velocity=&RotateVectorByQuaternion(Vector To Rotate=<-1 0 0>, Quaternion=$Particles.SkeletalMeshLocation.SampledRotation)
    Velocity Speed Scale=10
   ) {enabled=false}
   &ColorFromTOD(
    Afternoon:LinearColor=<rgba 1 0.3332 0.2934 0.5>
    Evening:LinearColor=<rgba 1 0.108 0.3328 0.5>
    FortniteNPC_FortniteActiveTimeOfDay:LinearColor=$NPC.FortniteNPC.FortniteActiveTimeOfDay
    Morning:LinearColor=<rgba 1 0.2038 0.4641 0.5>
    Night:LinearColor=<rgba 1 0.166 0.6334 0.5>
   )
  ]
  particleUpdateScript [
   &ParticleState()
   &InheritSourceMovement(
    Applied Position Scale=&Multiply_VectorByFloat(
     Vector=<0.99 0.99 0.99>
     Float=&FloatFromCurve(
      FloatCurve=DataInterface() {propertyValues {Curve {keys [<at 0 0.8>,<at 1 1>]}}}
      CurveIndex=&NormalizeFloat(Value=&VectorLength(Vector=$Engine.Owner.Velocity), Min=100, Max=500)
     )
    )
    ApplyForce=false
    ApplyPosition=true
    ApplyPositionFalloff=false
   )
   &GravityForce(Gravity=&MakeVector(
     X=0
     Y=0
     Z=&FloatFromCurve(
      FloatCurve=DataInterface() {propertyValues {Curve {keys [
          {arriveTangent=3.0184, interpMode=RCIM_Cubic, leaveTangent=3.0184, tangentMode=RCTM_User, time=0, value=0}
          {interpMode=RCIM_Cubic, tangentMode=RCTM_User, time=0.4, value=0.5}
          {tangentMode=RCTM_User, time=0.7, value=-0.5}
         ]}}}
      Scale Curve=20
     )
    )) {enabled=false}
   &GravityForce(Gravity=<0 0 40>) {module=GravityForce001}
   &CurlNoiseForce(Noise Frequency=100, Pan Noise=false)
   &Drag(Use Linear Drag=true, Use Rotational Drag=false)
   &SolveForcesAndVelocity(Clamp Velocity=false, Limit Acceleration=false)
   &ScaleMeshSize(Scale Factor=&VectorFromFloat(Value=&FloatFromCurve(FloatCurve=DataInterface() {propertyValues {Curve {keys [
          {arriveTangent=9.846, interpMode=RCIM_Cubic, leaveTangent=9.846, tangentMode=RCTM_User, time=0, value=0}
          {arriveTangent=-3.4392, interpMode=RCIM_Cubic, leaveTangent=-3.4392, tangentMode=RCTM_User, time=0.1, value=1}
          {interpMode=RCIM_Cubic, tangentMode=RCTM_User, time=1, value=0}
         ]}}})))
   &ScaleColor(
    Scale RGB=&MakeVectorFromLinearColorRGB(Color=&ColorFromCurve(
      ColorCurve=DataInterface() {propertyValues {
        AlphaCurve {keys [<at 0 1 RCIM_Cubic>]}
        BlueCurve {keys [<at 0 1 RCIM_Cubic>,<at 1 6 RCIM_Cubic>]}
        GreenCurve {keys [<at 0 1 RCIM_Cubic>,<at 1 0 RCIM_Cubic>]}
        RedCurve {keys [<at 0 1 RCIM_Cubic>,<at 1 1.8052 RCIM_Cubic>]}
       }}
      CurveIndex=$System.Thrive_NormalizedLevel
     ))
    ScaleA=true
    ScaleRGB=true
   )
  ]
  rendererClasses [&NiagaraSpriteRendererProperties, &NiagaraMeshRendererProperties]
  renderers [
   {rendererClass=&NiagaraSpriteRendererProperties}
   {rendererClass=&NiagaraMeshRendererProperties, rendererIndex=1}
  ]
 }
 HipRibbon_Ribbons {
  emitterSpawnScript [
   /SetVariables(Emitter.RandomSeed:NiagaraInt32=/Niagara/DynamicInputs/Random/RandomInt(Randomness Mode:ENiagaraRandomnessMode=::NewEnumerator0, Random Seed:NiagaraInt32=0, Fixed Random Seed:NiagaraBool=false)) {isSetParameters=true}
   /SetVariables(Emitter.ParticlesPerGroup=10) {isSetParameters=true}
  ]
  emitterUpdateScript [
   /SetVariables(Emitter.NumberOfGroups:NiagaraInt32=&RandomRangeInt(Minimum=2, Maximum=2, Evaluation Type=::NewEnumerator1, Random Seed=$Emitter.LoopCount, Fixed Random Seed=true)) {isSetParameters=true}
   &EmitterState(Life Cycle Mode=::NewEnumerator1, Loop Behavior=::NewEnumerator1, Loop Duration Mode=::NewEnumerator1, UseLoopDelay=false)
   &SpawnBurst_Instantaneous(
    Spawn Count=&Multiply_Int(A=$Emitter.ParticlesPerGroup, B=$Emitter.NumberOfGroups)
    Use Spawn Probability=false
   )
   &FN_Timer()
  ]
  particleSpawnScript [
   /SetVariables(Particles.LoopCount:NiagaraInt32=$Emitter.LoopCount) {isSetParameters=true}
   &FN_SetSpawnGroup(
    Inclusivity=true
    Normalize Spawn Group=true
    Particles Per Group:NiagaraInt32=$Emitter.ParticlesPerGroup
    Pattern=::NewEnumerator0
    Set Ribbon ID and Ribbon Link Order=true
   )
   &InitializeRibbon(
    Lifetime=5
    Module.Write Color=true
    Module.Write Lifetime=true
    Module.Write Mass=true
    Module.Write Material Random=true
    Module.Write Position=false
    Module.Write RibbonTwist=false
    Module.Write RibbonWidth=true
    Ribbon Width=7
   )
   &FN_FlipRibbonV()
  ]
  particleUpdateScript [
   &ParticleState(Kill Particles When Lifetime Has Elapsed=false, Let Infinitely Lived Particles Die When Emitter Deactivates=true)
   /Game/Effects/Niagara/Modules/BoneFlow/FN_BoneFlow_v4/FN_BoneFlow(
    Angle:NiagaraFloat=&Add_Float(
     A=&Multiply_Float(
      A=$Particles.Normalized Spawn Group
      B=/Game/Effects/Niagara/DynamicInputs/Random/FN_SeededRandomRangeFloat(
       Minimum:NiagaraFloat=-1
       Maximum:NiagaraFloat=1
       Random Seed:NiagaraInt32=&Add_Integer(A=$Particles.LoopCount, B=$Particles.SpawnGroup)
      )
     )
     B=&Multiply_Float(A=$Particles.Age, B=0.1)
    )
    BoneID:NiagaraInt32=&ModuloInt(A=$Particles.SpawnGroup, B=3)
    Height:NiagaraFloat=$Particles.Normalized Spawn Group
    Interpolate Mesh Sampling=false
    Radius:NiagaraFloat=&Multiply_Float(A=$Particles.Normalized Spawn Group, B=5)
    SkeletalMesh:NiagaraDataInterfaceSkeletalMesh=$User.SkeletalMeshBow
    Write Position Attribute=true
   ) {enabled=false, module=FN_BoneFlow001}
   &FN_BoneFlow(
    Angle:NiagaraFloat=&Add_Float(
     A=&Multiply_Float(
      A=$Particles.Normalized Spawn Group
      B=&FN_SeededRandomRangeFloat(
       Minimum:NiagaraFloat=-1
       Maximum:NiagaraFloat=1
       Random Seed:NiagaraInt32=&Add_Integer(A=$Particles.LoopCount, B=$Particles.SpawnGroup)
      )
     )
     B=&Multiply_Float(A=$Particles.Age, B=0.1)
    )
    BoneID:NiagaraInt32=&ModuloInt(A=$Particles.SpawnGroup, B=3)
    Height:NiagaraFloat=$Particles.Normalized Spawn Group
    Interpolate Mesh Sampling=false
    Radius:NiagaraFloat=&FloatFromCurve(
     FloatCurve=DataInterface() {propertyValues {Curve {keys [
         {interpMode=RCIM_Cubic, tangentMode=RCTM_User, time=0, value=0}
         {interpMode=RCIM_Cubic, tangentMode=RCTM_User, time=0.5, value=1}
         {tangentMode=RCTM_User, time=1, value=0}
        ]}}}
     CurveIndex=$Particles.Normalized Spawn Group
     Scale Curve=3
    )
    SkeletalMesh:NiagaraDataInterfaceSkeletalMesh=$User.SkeletalMeshBow
    Write Position Attribute=true
   )
   /SetVariables(Particles.RibbonFacing=$Output.FN_BoneFlow.SplineBinormal) {isSetParameters=true}
   &FN_RibbonMaterial(
    Color=<rgba 0 1 1 9>
    Opacity Pan U Offset:NiagaraFloat=&FN_SeededRandomRangeFloat(
     Minimum:NiagaraFloat=0
     Maximum:NiagaraFloat=2
     Random Seed:NiagaraInt32=&Add_Integer(A=$Particles.LoopCount, B=$Particles.SpawnGroup)
    )
    Opacity Pan U Speed:NiagaraFloat=&FN_SeededRandomRangeFloat(
     Minimum:NiagaraFloat=0.1364
     Maximum:NiagaraFloat=0.5182
     Random Seed:NiagaraInt32=&Add_Integer(A=$Particles.LoopCount, B=$Particles.SpawnGroup)
    )
    Opacity Scale U:NiagaraFloat=&FN_SeededRandomRangeFloat(
     Minimum:NiagaraFloat=0.4273
     Maximum:NiagaraFloat=1.0032
     Random Seed:NiagaraInt32=&Add_Integer(A=$Particles.LoopCount, B=$Particles.SpawnGroup)
    )
    Scale Alpha:NiagaraFloat=&FloatFromCurve(
     FloatCurve=DataInterface() {propertyValues {Curve {keys [<at -0.007 0.0112>,<at 0.2367 0.9629>,<at 0.7537 0.9995>,<at 1 0>]}}}
     Scale Curve=&FloatFromCurve(
      FloatCurve=DataInterface() {propertyValues {Curve {keys [<at 0.0007 0.022>,<at 0.2494 1.0225>,<at 0.8089 0.9748>,<at 1 0>]}}}
      CurveIndex=$Particles.Normalized Spawn Group
     )
    )
    UV Basis:NiagaraFloat=$Particles.Normalized Spawn Group
   ) {enabled=false}
   &DynamicMaterialParameters(
    Index 2 Param 4=1
    Index 3 Param 1=&Add_Float(
     A=&SelectFloatFromArray(Direct Array Index=$Particles.SpawnGroup, Float Selection Array=DataInterface() {propertyValues {FloatData [1,2]}})
     B=$Emitter.FN_Timer.Value
    )
    Module.Param0WriteEnabled2=true
    Module.Param0WriteEnabled3=true
    Module.Param1WriteEnabled2=true
    Module.Param1WriteEnabled3=true
    Module.Param2WriteEnabled2=true
    Module.Param2WriteEnabled3=true
    Module.Param3WriteEnabled2=true
    Module.Param3WriteEnabled3=true
    Write Parameter Index 0=false
    Write Parameter Index 2=true
    Write Parameter Index 3=true
   )
   &CameraOffset(Camera Offset Amount=100) {enabled=false}
   &ScaleColor(
    Scale RGB=&MakeVectorFromLinearColorRGB(Color=&ColorFromCurve(
      ColorCurve=DataInterface() {propertyValues {
        AlphaCurve {keys [<at 0 1 RCIM_Cubic>]}
        BlueCurve {keys [<at 0 1 RCIM_Cubic>,<at 1 6 RCIM_Cubic>]}
        GreenCurve {keys [<at 0 1 RCIM_Cubic>,<at 1 0 RCIM_Cubic>]}
        RedCurve {keys [<at 0 1 RCIM_Cubic>,<at 1 1.8052 RCIM_Cubic>]}
       }}
      CurveIndex=$System.Thrive_NormalizedLevel
     ))
    ScaleA=true
    ScaleRGB=true
   )
   &ColorFromTOD(
    Afternoon:LinearColor=<rgba 1 0.1534 0.3305 0.5>
    Evening:LinearColor=<rgba 0.1545 0.0099 0.0631 0.5>
    FortniteNPC_FortniteActiveTimeOfDay:LinearColor=$NPC.FortniteNPC.FortniteActiveTimeOfDay
    Morning:LinearColor=<rgba 0.2656 0.048 0.1538 0.5>
    Night:LinearColor=<rgba 0.2483 0.099 0.1702 0.5>
   )
   &ScaleRibbonWidth(Ribbon Width Scale=&FloatFromCurve(
     FloatCurve=DataInterface() {propertyValues {Curve {keys [
         {arriveTangent=5.8415, interpMode=RCIM_Cubic, leaveTangent=5.8415, tangentMode=RCTM_User, time=0, value=0}
         {interpMode=RCIM_Cubic, tangentMode=RCTM_User, time=0.5, value=1}
         {arriveTangent=-5.6133, interpMode=RCIM_Cubic, leaveTangent=-5.6133, tangentMode=RCTM_User, time=1, value=0}
        ]}}}
     CurveIndex=$Particles.Normalized Spawn Group
    ))
  ]
  rendererClasses [&NiagaraRibbonRendererProperties]
  renderers [
   {rendererClass=&NiagaraRibbonRendererProperties}
   {rendererClass=&NiagaraRibbonRendererProperties, rendererIndex=1}
  ]
 }
 Light {
  emitterUpdateScript [
   &EmitterState(Life Cycle Mode=::NewEnumerator1, Loop Behavior=::NewEnumerator1, Loop Duration Mode=::NewEnumerator1, UseLoopDelay=false)
   &SpawnBurst_Instantaneous(Use Spawn Probability=false, UseLoopCountLimit=false)
  ]
  particleSpawnScript [&InitializeParticle(Color Mode=::NewEnumerator0, UsePositionOffset=false, Write Material Random=true)]
  particleUpdateScript [
   &ParticleState(Kill Particles When Lifetime Has Elapsed=false, Let Infinitely Lived Particles Die When Emitter Deactivates=true)
   &SkeletalMeshLocation(Bone Sampling Mode=::NewEnumerator4, MaskSpawnGroup=false, Sampled Position Offset=<-10 0 0>, Skeletal Mesh=$User.SkeletalMeshCenter)
   /SetVariables(Particles.LightCamOffset:Vector3f=&Multiply_VectorByFloat(
     Vector=&NormalizeVector(
      Vector To Normalize=&Subtract_Vector(
       A=Linked() {linkedVariable {name=Particles.Position, type=NiagaraPosition}}
       B=&ReturnCameraVector(Camera Property=::NewEnumerator0, Coordinate Space=::World)
      )
      Safe Normalize=true
      Safe Fallback Threshold=0
      Safe Fallback Vector=<0 0 0.001>
     )
     Float=25
    )) {isSetParameters=true}
   &FN_Light_Attributes(
    Color=<rgba 1 0.166 0.6334 1>
    Create Color Attribute=true
    Create Exponent Attribute=true
    Create Position Attribute=true
    Create Radius Attribute=true
    Exponent=6
    Position=&Add_Vector(A=Linked() {linkedVariable {name=Particles.Position, type=NiagaraPosition}}, B=$Particles.LightCamOffset)
    Radius=250
    Scale Color=&Multiply_Float(A=4, B=$System.Thrive_NormalizedLevel)
   )
  ]
  rendererClasses [&NiagaraLightRendererProperties, &NiagaraSpriteRendererProperties]
  renderers [
   {rendererClass=&NiagaraLightRendererProperties, rendererIndex=0}
   {rendererClass=&NiagaraSpriteRendererProperties, rendererIndex=1}
  ]
 }
 Necklace_Glows {
  emitterUpdateScript [
   &EmitterState(Life Cycle Mode=::NewEnumerator1, Loop Behavior=::NewEnumerator1, Loop Duration Mode=::NewEnumerator1, UseLoopDelay=false)
   &SpawnBurst_Instantaneous(Spawn Count=5, Use Spawn Probability=false, UseLoopCountLimit=false)
  ]
  particleSpawnScript [
   &FN_SetSpawnGroup(Number of Spawn Groups=5)
   &InitializeParticle(Color=<rgba 0 1 0.8641 0.25>, Sprite Size Mode=::NewEnumerator3, UsePositionOffset=false, Write Material Random=true)
   &FN_Sprite_Attributes(
    Color=<rgba 0 2 1.7282 1>
    Create Color Attribute=true
    Create SpriteSize Attribute=true
    SpriteSize=&Multiply_Vector2D_ByFloat(Float=0.5, Vector2D=$Particles.SpriteSize)
    Write to Module Namespace=true
   )
  ]
  particleUpdateScript [
   &ParticleState(Kill Particles When Lifetime Has Elapsed=false, Let Infinitely Lived Particles Die When Emitter Deactivates=true)
   &SkeletalMeshLocation(
    Bone / Socket Index=$Particles.SpawnGroup
    Bone Sampling Mode=::NewEnumerator4
    MaskSpawnGroup=false
    Sampled Position Offset=<2.5 0 0>
    Skeletal Mesh=$User.SkeletalMeshNecklace
   )
   &ScaleColor(
    Scale Alpha=&FloatFromCurve(
     FloatCurve=DataInterface() {propertyValues {Curve {keys [
         {arriveTangent=1.9311, interpMode=RCIM_Cubic, leaveTangent=1.9311, tangentMode=RCTM_User, time=0, value=0}
         {interpMode=RCIM_Cubic, tangentMode=RCTM_User, time=1, value=1}
        ]}}}
     CurveIndex=$System.Thrive_NormalizedLevel
    )
    ScaleA=true
    ScaleRGB=false
   )
   /SetVariables(Particles.FN_Sprite_Attributes.Color:LinearColor=&MultiplyLinearColor(
     Linear Color=$Particles.FN_Sprite_Attributes.Color
     Scale Factor=&MakeVector4_FromLinearColor(LinearColor=$Transient.ParticleColorScaleFactor)
    )) {isSetParameters=true}
   &CameraOffset(Camera Offset Amount=1)
  ]
  rendererClasses [&NiagaraSpriteRendererProperties]
  renderers [
   {rendererClass=&NiagaraSpriteRendererProperties}
   {rendererClass=&NiagaraSpriteRendererProperties, rendererIndex=1}
  ]
 }
 ReactivityDEBUG {
  bEnabled=false
  emitterUpdateScript [
   &EmitterState(Life Cycle Mode=::NewEnumerator1, Loop Behavior=::NewEnumerator1, Loop Duration Mode=::NewEnumerator1, UseLoopDelay=false)
   &SpawnRate(
    SpawnRate=&Multiply_Float(A=&Multiply_Float(A=$System.Thrive_NormalizedLevel, B=50), B=0)
    Use Spawn Probability=false
   )
  ]
  particleSpawnScript [
   &InitializeParticle(Color Mode=::NewEnumerator0, Sprite Size Mode=::NewEnumerator3, UsePositionOffset=false, Write Material Random=true)
   &AddVelocity(Module.Use Falloff Distance=false, Override Seed=false, Velocity Mode=::NewEnumerator1, Velocity Speed=50) {enabled=false}
   &ShapeLocation(MaskSpawnGroup=false, Override Seed=false, Sphere Radius=250, UseOwnerScale=true)
  ]
  particleUpdateScript [
   &SolveForcesAndVelocity(Clamp Velocity=false, Limit Acceleration=false)
   &ParticleState()
   &ScaleColor(
    Scale RGB=&MakeVectorFromLinearColorRGB(Color=&ColorFromCurve(
      ColorCurve=DataInterface() {propertyValues {
        AlphaCurve {keys [<at 0 1 RCIM_Cubic>]}
        BlueCurve {keys [
          <at 0 1 RCIM_Cubic>
          <at 0.1712 1>
          {arriveTangent=-1.4215, interpMode=RCIM_Cubic, leaveTangent=-1.4215, time=0.4518, value=0}
          <at 0.7035 0>
          <at 0.9877 0>
         ]}
        GreenCurve {keys [
          <at 0 1 RCIM_Cubic>
          <at 0.1712 0>
          {arriveTangent=-0.4719, interpMode=RCIM_Cubic, leaveTangent=-0.4719, time=0.4518, value=0}
          <at 0.7035 0.668>
          <at 0.9877 1>
         ]}
        RedCurve {keys [<at 0 1 RCIM_Cubic>,<at 0.1712 0.2076>,<at 0.4518 1 RCIM_Cubic>,<at 0.7035 1>,<at 0.9877 0.249>]}
       }}
      CurveIndex=&Multiply_Float(
       A=&OneMinusFloat(Float=&ClampFloat(Float=&Divide_Float(A=&Subtract_Float(A=$Engine.Time, B=$User.Thrive_GameTimeFromBP), B=$User.Thrive_EffectDecayRate)))
       B=&Divide_Float(A=$User.Thrive_EffectLevel, B=$User.Thrive_MaxLevel)
      )
     ))
    ScaleA=true
    ScaleRGB=true
   ) {enabled=false}
   &ScaleColor(
    Scale RGB=&MakeVectorFromLinearColorRGB(Color=&ColorFromCurve(
      ColorCurve=DataInterface() {propertyValues {
        AlphaCurve {keys [<at 0 1 RCIM_Cubic>]}
        BlueCurve {keys [
          <at 0 1 RCIM_Cubic>
          <at 0.1712 1>
          {arriveTangent=-1.4215, interpMode=RCIM_Cubic, leaveTangent=-1.4215, time=0.4518, value=0}
          <at 0.7035 0>
          <at 0.9877 0>
         ]}
        GreenCurve {keys [
          <at 0 1 RCIM_Cubic>
          <at 0.1712 0>
          {arriveTangent=-0.4719, interpMode=RCIM_Cubic, leaveTangent=-0.4719, time=0.4518, value=0}
          <at 0.7035 0.668>
          <at 0.9877 1>
         ]}
        RedCurve {keys [<at 0 1 RCIM_Cubic>,<at 0.1712 0.2076>,<at 0.4518 1 RCIM_Cubic>,<at 0.7035 1>,<at 0.9877 0.249>]}
       }}
      CurveIndex=$System.Thrive_NormalizedLevel
     ))
    ScaleA=true
    ScaleRGB=true
   ) {module=ScaleColor001}
  ]
  rendererClasses [&NiagaraSpriteRendererProperties]
  renderers [{rendererClass=&NiagaraSpriteRendererProperties}]
 }
 Sleeves_Blobas {
  emitterSpawnScript [
   &FN_PartitionAttributes(Number of Bone Sets=1, Particles Per Partition=10)
   /SetVariables() {isSetParameters=true, module=SetVariables_#5}
  ]
  emitterUpdateScript [
   &EmitterState(Life Cycle Mode=::NewEnumerator1, Loop Behavior=::NewEnumerator1, Loop Duration Mode=::NewEnumerator1, UseLoopDelay=false)
   &SpawnRate(SpawnRate=5, Use Spawn Probability=false)
  ]
  particleSpawnScript [
   &FN_Particle_Partition_v5(Enable Ribbon ID=true, Pattern:ENiagaraPartitionWaveForm=::NewEnumerator1, Seed:NiagaraInt32=$Emitter.Seed, Shuffle Partition ID=true)
   &InitializeParticle(Mesh Scale Mode=::NewEnumerator3, Mesh Uniform Scale=0.04, UsePositionOffset=false, Write Material Random=true)
   &InitializeRibbon(
    Color=<rgba 1 0.1534 0.3305 1>
    Module.Write Color=true
    Module.Write Lifetime=true
    Module.Write Mass=false
    Module.Write Material Random=true
    Module.Write Position=false
    Module.Write RibbonTwist=false
    Module.Write RibbonWidth=true
    Ribbon Width=10
   ) {enabled=false}
   &FN_Initialize_BoneFlow_Attributes_v2(
    Angle:NiagaraFloat=&Multiply_Float(
     A=$Particles.Partition.NormalizedID
     B=&UniformRangedFloat(Minimum=0, Maximum=0.5, Random Seed=$Particles.Partition.ID)
    )
    Radius:NiagaraFloat=&FloatFromCurve(
     FloatCurve=DataInterface() {propertyValues {Curve {keys [
         {arriveTangent=77401.0078, interpMode=RCIM_Cubic, leaveTangent=50.7651, tangentMode=RCTM_User, time=0, value=0}
         {arriveTangent=16.47, interpMode=RCIM_Cubic, leaveTangent=16.47, time=0.13, value=5.5786}
         <at 0.4924 8.1096 RCIM_Cubic>
         {arriveTangent=-15.9761, interpMode=RCIM_Cubic, leaveTangent=-15.9761, time=0.8801, value=5.5786}
         {arriveTangent=-40.3134, interpMode=RCIM_Cubic, leaveTangent=-1467.5861, tangentMode=RCTM_User, time=1, value=0}
        ]}}}
     CurveIndex=$Particles.Partition.NormalizedID
     Scale Curve=0
    )
    T:NiagaraFloat=$Particles.Partition.NormalizedID
   )
   /Game/Effects/Niagara/Modules/BoneFlow/FN_BoneFlow(
    Angle:NiagaraFloat=&Add_Float(
     A=&Multiply_Float(A=$Particles.Initial.FN_Initialize_BoneFlow_Attributes_v2.Angle, B=10)
     B=&Multiply_Float(A=$Emitter.Age, B=0.5)
    )
    Bone Offsets=DataInterface() {propertyValues {
      GpuSyncMode=SyncCpuToGpu
      MaxElements=0
      FloatData [<-5 0 0>, <0 8 -3>, <-5 0 2>, <8 -12 0>, <-5 -3 -10>, <-5 0 0>, <0 -10 3>, <7 -2 0>, <-5 8 0>, <5 3 0>]
     }}
    Bone Sets=DataInterface() {propertyValues {GpuSyncMode=SyncCpuToGpu, MaxElements=0, IntData [0,4,5,9]}}
    Height:NiagaraFloat=$Particles.Partition.NormalizedID
    Radius:NiagaraFloat=$Particles.Initial.FN_Initialize_BoneFlow_Attributes_v2.Radius
    SkeletalMesh:NiagaraDataInterfaceSkeletalMesh=$User.SkeletalMeshSleves
   ) {module=FN_BoneFlow001}
   &ColorFromTOD(
    Afternoon:LinearColor=<rgba 1 0.1534 0.3305 0.5>
    Evening:LinearColor=<rgba 0.1545 0.0099 0.0631 0.5>
    FortniteNPC_FortniteActiveTimeOfDay:LinearColor=$NPC.FortniteNPC.FortniteActiveTimeOfDay
    Morning:LinearColor=<rgba 0.2656 0.048 0.1538 0.5>
    Night:LinearColor=<rgba 0.2483 0.099 0.1702 0.5>
   )
  ]
  particleUpdateScript [
   &ParticleState()
   &InheritSourceMovement(
    Applied Position Scale=&Multiply_VectorByFloat(
     Vector=<0.99 0.99 0.99>
     Float=&FloatFromCurve(
      FloatCurve=DataInterface() {propertyValues {Curve {keys [<at 0 0.8>,<at 1 1>]}}}
      CurveIndex=&NormalizeFloat(Value=&VectorLength(Vector=$Engine.Owner.Velocity), Min=100, Max=500)
     )
    )
    ApplyForce=false
    ApplyPosition=true
    ApplyPositionFalloff=false
   )
   &GravityForce(Gravity=<0 0 50>)
   &CurlNoiseForce(Noise Frequency=100, Pan Noise=false)
   &Drag(Use Linear Drag=true, Use Rotational Drag=false)
   &SolveForcesAndVelocity(Clamp Velocity=false, Limit Acceleration=false)
   &ScaleMeshSize(Scale Factor=&VectorFromFloat(Value=&FloatFromCurve(FloatCurve=DataInterface() {propertyValues {Curve {keys [
          {arriveTangent=9.846, interpMode=RCIM_Cubic, leaveTangent=9.846, tangentMode=RCTM_User, time=0, value=0}
          {arriveTangent=-2.8092, interpMode=RCIM_Cubic, leaveTangent=-2.8092, tangentMode=RCTM_User, time=0.1, value=1}
          {interpMode=RCIM_Cubic, tangentMode=RCTM_User, time=1, value=0}
         ]}}})))
   &ScaleColor(
    Scale RGB=&MakeVectorFromLinearColorRGB(Color=&ColorFromCurve(
      ColorCurve=DataInterface() {propertyValues {
        AlphaCurve {keys [<at 0 1 RCIM_Cubic>]}
        BlueCurve {keys [<at 0 1 RCIM_Cubic>,<at 1 6 RCIM_Cubic>]}
        GreenCurve {keys [<at 0 1 RCIM_Cubic>,<at 1 0 RCIM_Cubic>]}
        RedCurve {keys [<at 0 1 RCIM_Cubic>,<at 1 1.8052 RCIM_Cubic>]}
       }}
      CurveIndex=$System.Thrive_NormalizedLevel
     ))
    ScaleA=true
    ScaleRGB=true
   )
   /SetVariables(Particles.VisibilityTag=&MakeCustomIntFromBool(Bool=$User.bisHandHidden, True INT=1, False INT=0)) {isSetParameters=true, module=SetVariables_#4}
  ]
  rendererClasses [&NiagaraMeshRendererProperties]
  renderers [{rendererClass=&NiagaraMeshRendererProperties, rendererIndex=0}]
 }
}
summary {
 systemName=NS_Thrive_Pink_Body
 emitters [
  {emitterName=BoneFlow_Sleeves, rendererClasses [&NiagaraRibbonRendererProperties]}
  {emitterName=Blobs_Arms, rendererClasses [&NiagaraSpriteRendererProperties, &NiagaraMeshRendererProperties]}
  {
   emitterName=HipRibbon_Blobs
   rendererClasses [&NiagaraSpriteRendererProperties, &NiagaraMeshRendererProperties]
  }
  {emitterName=HipRibbon_Ribbons, rendererClasses [&NiagaraRibbonRendererProperties]}
  {bEnabled=false, emitterName=ReactivityDEBUG, rendererClasses [&NiagaraSpriteRendererProperties]}
  {emitterName=Necklace_Glows, rendererClasses [&NiagaraSpriteRendererProperties]}
  {emitterName=Sleeves_Blobas, rendererClasses [&NiagaraMeshRendererProperties]}
  {emitterName=FN_AttachedRibbon, rendererClasses [&NiagaraRibbonRendererProperties]}
  {emitterName=Light, rendererClasses [&NiagaraLightRendererProperties, &NiagaraSpriteRendererProperties]}
 ]
 userVariables [
  User.bisHandHidden {type=NiagaraBool}
  User.BoneFlowSmallSpawn {type=NiagaraInt32}
  User.Color_Tint {type=LinearColor, description="", defaultValue {struct=LinearColor, value=<rgba 1 1 1 1>}}
  User.ColorChange {type=LinearColor, description="", defaultValue {struct=LinearColor, value=<rgba 1 1 1 1>}}
  User.Debug_Reactive_Level {type=NiagaraFloat, defaultValue=0, description=""}
  User.EmissiveIntensity {type=NiagaraFloat, defaultValue=0, description=""}
  User.FE_Preview {type=NiagaraFloat, defaultValue=0, description=""}
  User.RibbonIDOffset {type=NiagaraInt32}
  User.SkeletalMesh {
   type=NiagaraDataInterfaceSkeletalMesh
   defaultValue {
    struct=NiagaraExt_VariableValue_DataInterface
    value {dataInterface=&self:NiagaraDataInterfaceSkeletalMesh_2}
   }
  }
  User.SkeletalMeshArms {
   type=NiagaraDataInterfaceSkeletalMesh
   defaultValue {
    struct=NiagaraExt_VariableValue_DataInterface
    value {dataInterface=&self:NiagaraDataInterfaceSkeletalMesh_5}
   }
  }
  User.SkeletalMeshArmsRibbons {
   type=NiagaraDataInterfaceSkeletalMesh
   defaultValue {
    struct=NiagaraExt_VariableValue_DataInterface
    value {dataInterface=&self:NiagaraDataInterfaceSkeletalMesh_7}
   }
  }
  User.SkeletalMeshBow {
   type=NiagaraDataInterfaceSkeletalMesh
   defaultValue {
    struct=NiagaraExt_VariableValue_DataInterface
    value {dataInterface=&self:NiagaraDataInterfaceSkeletalMesh_1}
   }
  }
  User.SkeletalMeshCenter {
   type=NiagaraDataInterfaceSkeletalMesh
   defaultValue {
    struct=NiagaraExt_VariableValue_DataInterface
    value {dataInterface=&self:NiagaraDataInterfaceSkeletalMesh_9}
   }
  }
  User.SkeletalMeshNecklace {
   type=NiagaraDataInterfaceSkeletalMesh
   defaultValue {
    struct=NiagaraExt_VariableValue_DataInterface
    value {dataInterface=&self:NiagaraDataInterfaceSkeletalMesh_3}
   }
  }
  User.SkeletalMeshSleves {
   type=NiagaraDataInterfaceSkeletalMesh
   defaultValue {
    struct=NiagaraExt_VariableValue_DataInterface
    value {dataInterface=&self:NiagaraDataInterfaceSkeletalMesh_0}
   }
  }
  User.Thrive_EffectDecayRate {type=NiagaraFloat, defaultValue=0, description=""}
  User.Thrive_EffectLevel {type=NiagaraFloat, defaultValue=0, description=""}
  User.Thrive_GameTimeFromBP {type=NiagaraFloat, defaultValue=0, description=""}
  User.Thrive_MaxLevel {type=NiagaraFloat, defaultValue=0, description=""}
  User.Thrive_Previous_Intensity {type=NiagaraFloat, defaultValue=0, description=""}
  User.Thrive_Target_Intensity {type=NiagaraFloat, defaultValue=0, description=""}
 ]
}
