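# Niagara system for the pink Thrive character's hair and earring effects.
# HairDropsUp emits mesh drops from head bones that drift with noise, curved gravity and inherited motion; HairEndDrops keeps drops on hair-tip bones, sliding them along per-group offsets.
# Both drop emitters take pink or teal colors by spawn group and time of day.
# EaringGlow pins two sprites to earrings, fading by normalized Thrive level; all emitters set a head-hidden visibility tag.
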NiagaraSystem /BRCosmetics/Characters/Player/Female/Medium/Bodies/F_MED_Thrive/FX/NS_Thrive_Pink_Hair
dependencies {
 usedDataInterfaces [
  &NiagaraDataInterfaceArrayColor
  &NiagaraDataInterfaceMeshRendererInfo
  &NiagaraDataInterfaceVectorCurve
  &NiagaraDataInterfaceArrayFloat3
  &NiagaraDataInterfaceCurve
  &NiagaraDataInterfaceColorCurve
  &NiagaraDataInterfaceVector2DCurve
  &NiagaraDataInterfaceArrayFloat
  &NiagaraDataInterfaceArrayFloat2
  &NiagaraDataInterfaceArrayInt32
 ]
 usedDynamicInputs [
  &Add_Float
  &RandomRangeFloat
  &Multiply_Float
  &OneMinusFloat
  &ClampFloat
  &Divide_Float
  &Subtract_Float
  &Lerp_Float
 ]
 usedModules [
  &SystemState
  &Timeline
  &self:SystemSpawnScript.SystemScriptSource.SystemScriptGraph.NiagaraNodeAssignment_0.SetVariables_#1
  &self:SystemSpawnScript.SystemScriptSource.SystemScriptGraph.NiagaraNodeAssignment_1.SetVariables_#2
 ]
}
emitters {
 EaringGlow {
  emitterUpdateScript [
   &EmitterState(Life Cycle Mode=::NewEnumerator1, Loop Behavior=::NewEnumerator1, Loop Duration Mode=::NewEnumerator1, UseLoopDelay=false)
   &SpawnBurst_Instantaneous(Spawn Count=2, Use Spawn Probability=false, UseLoopCountLimit=false)
  ]
  particleSpawnScript [
   &InitializeParticle(
    Color=<rgba 0.842 0.1067 0.2774 0.5>
    Lifetime=5
    Sprite Size Mode=::NewEnumerator3
    Uniform Sprite Size=9
    UsePositionOffset=false
    Write Material Random=true
   )
   &FN_SetSpawnGroup(Number of Spawn Groups=2)
   &FN_Sprite_Attributes(
    Color=<rgba 1 0.166 0.6334 1>
    Create Color Attribute=true
    Create SpriteSize Attribute=true
    SpriteSize=&Multiply_Vector2D_ByFloat(Float=0.5, Vector2D=$Particles.SpriteSize)
    Write to Module Namespace=true
   )
  ]
  particleUpdateScript [
   &ParticleState(Kill Particles When Lifetime Has Elapsed=false, Let Infinitely Lived Particles Die When Emitter Deactivates=true)
   &SkeletalMeshLocation(
    Bone / Socket Index=$Particles.SpawnGroup
    Bone Sampling Mode=::NewEnumerator4
    MaskSpawnGroup=false
    Sampled Position Offset=&MaskVectorBySpawnGroup(Masked Vector=<5 0 0>, Passthrough Vector=<-5 0 0>)
    Skeletal Mesh=$User.SkeletalMeshEarings
   )
   &ScaleColor(
    Scale Alpha=&FloatFromCurve(
     FloatCurve=DataInterface() {propertyValues {Curve {keys [
         {arriveTangent=1.9311, interpMode=RCIM_Cubic, leaveTangent=1.9311, tangentMode=RCTM_User, time=0, value=0}
         {interpMode=RCIM_Cubic, tangentMode=RCTM_User, time=1, value=1}
        ]}}}
     CurveIndex=$System.Thrive_NormalizedLevel
    )
    ScaleA=true
    ScaleRGB=false
   )
   /SetVariables(Particles.FN_Sprite_Attributes.Color:LinearColor=&MultiplyLinearColor(
     Linear Color=$Particles.FN_Sprite_Attributes.Color
     Scale Factor=&MakeVector4_FromLinearColor(LinearColor=$Transient.ParticleColorScaleFactor)
    )) {isSetParameters=true}
   &CameraOffset(Camera Offset Amount=1)
   /SetVariables(Particles.VisibilityTag=&MakeCustomIntFromBool(Bool=$User.bisHeadHidden, True INT=1, False INT=0)) {isSetParameters=true, module=SetVariables_#3}
  ]
  rendererClasses [&NiagaraSpriteRendererProperties]
  renderers [
   {rendererClass=&NiagaraSpriteRendererProperties}
   {rendererClass=&NiagaraSpriteRendererProperties, rendererIndex=1}
  ]
 }
 HairDropsUp {
  emitterUpdateScript [
   &EmitterState()
   &SpawnRate(SpawnRate=2, Use Spawn Probability=false)
  ]
  particleSpawnScript [
   &FN_SetSpawnGroup(Number of Spawn Groups=5)
   &InitializeParticle(
    Color=&SelectLinearColorFromArray(
     Array Sampling Mode=::NewEnumerator1
     Direct Array Index=$Particles.SpawnGroup
     Color Selection Array=DataInterface() {propertyValues {ColorData [
        <rgba 0.3108 0.8906 0.8727 1>
        <rgba 0.3095 0.8879 0.8714 1>
        <rgba 0.3095 0.8879 0.8714 1>
        <rgba 1 0.166 0.6334 1>
        <rgba 1 0.1651 0.6308 1>
       ]}}
     Random Seed=?
    )
    Lifetime Max=3
    Lifetime Min=2
    Lifetime Mode=::NewEnumerator1
    Mesh Scale Mode=::NewEnumerator3
    Mesh Uniform Scale=0.025
    Sprite Size Mode=::NewEnumerator3
    UsePositionOffset=false
    Write Material Random=true
   ) {enabled=false}
   &InitializeParticle(
    Lifetime Max=3
    Lifetime Min=2
    Lifetime Mode=::NewEnumerator1
    Mesh Scale Mode=::NewEnumerator3
    Mesh Uniform Scale=0.025
    Sprite Size Mode=::NewEnumerator3
    UsePositionOffset=false
    Write Material Random=true
   ) {module=InitializeParticle001}
   &SkeletalMeshLocation(Bone / Socket Index=$Particles.SpawnGroup, Bone Sampling Mode=::NewEnumerator4, MaskSpawnGroup=false, Skeletal Mesh=$User.HeadBones)
   &AddVelocity(
    Velocity=&RotateVectorByQuaternion(Vector To Rotate=<1 0 0>, Quaternion=$Particles.SkeletalMeshLocation.SampledRotation)
    Velocity Speed Scale=10
   )
   /BRCosmetics/Weapons/FORT_Melee/Pickaxe_Pumpkin_Spice_1h/FX/ColorFromTOD(
    Afternoon:LinearColor=&MaskLinearColorBySpawnGroup(Masked Linear Color=<rgba 0.1727 1 0.9635 0.5>, Passthrough Linear Color=<rgba 1 0.2578 0.4492 0.5>, Spawn Group Mask=2)
    Evening:LinearColor=&MaskLinearColorBySpawnGroup(Masked Linear Color=<rgba 0.1655 0.3767 0.3289 0.5>, Passthrough Linear Color=<rgba 1 0.1248 0.2748 0.5>, Spawn Group Mask=2)
    FortniteNPC_FortniteActiveTimeOfDay:LinearColor=$NPC.FortniteNPC.FortniteActiveTimeOfDay
    Morning:LinearColor=&MaskLinearColorBySpawnGroup(Masked Linear Color=<rgba 0.1024 0.8288 1 0.5>, Passthrough Linear Color=<rgba 1 0.2038 0.4641 0.5>, Spawn Group Mask=2)
    Night:LinearColor=&MaskLinearColorBySpawnGroup(Masked Linear Color=<rgba 0.0405 0.0987 0.2101 0.5>, Passthrough Linear Color=<rgba 0.5608 0.1801 0.4529 0.5>, Spawn Group Mask=2)
   )
  ]
  particleUpdateScript [
   &ParticleState()
   &InheritSourceMovement(
    Applied Position Scale=&Multiply_VectorByFloat(
     Vector=<0.99 0.99 0.99>
     Float=&FloatFromCurve(
      FloatCurve=DataInterface() {propertyValues {Curve {keys [<at 0 0.8>,<at 1 1>]}}}
      CurveIndex=&NormalizeFloat(Value=&VectorLength(Vector=$Engine.Owner.Velocity), Min=100, Max=500)
     )
    )
    ApplyForce=false
    ApplyPosition=true
    ApplyPositionFalloff=false
   )
   &GravityForce(Gravity=&MakeVector(
     X=0
     Y=0
     Z=&FloatFromCurve(
      FloatCurve=DataInterface() {propertyValues {Curve {keys [
          {arriveTangent=3.0184, interpMode=RCIM_Cubic, leaveTangent=3.0184, tangentMode=RCTM_User, time=0, value=0}
          {interpMode=RCIM_Cubic, tangentMode=RCTM_User, time=0.4, value=0.5}
          {tangentMode=RCTM_User, time=0.7, value=-0.5}
         ]}}}
      Scale Curve=20
     )
    ))
   &CurlNoiseForce(Noise Frequency=100, Pan Noise=false)
   &Drag(Use Linear Drag=true, Use Rotational Drag=false)
   &SolveForcesAndVelocity(Clamp Velocity=false, Limit Acceleration=false)
   &ScaleMeshSize(Scale Factor=&VectorFromFloat(Value=&FloatFromCurve(FloatCurve=DataInterface() {propertyValues {Curve {keys [
          {arriveTangent=9.846, interpMode=RCIM_Cubic, leaveTangent=9.846, tangentMode=RCTM_User, time=0, value=0}
          <at 0.1 1>
          {interpMode=RCIM_Cubic, tangentMode=RCTM_User, time=1, value=0}
         ]}}})))
   &ScaleMeshSizeBySpeed(Max Scale Factor=<1 5 1>, Speed Threshold=100) {enabled=false}
   /SetVariables(Particles.VisibilityTag=&MakeCustomIntFromBool(Bool=$User.bisHeadHidden, True INT=1, False INT=0)) {isSetParameters=true, module=SetVariables_#3}
  ]
  rendererClasses [&NiagaraMeshRendererProperties]
  renderers [{rendererClass=&NiagaraMeshRendererProperties, rendererIndex=0}]
 }
 HairEndDrops {
  emitterUpdateScript [
   &EmitterState()
   &SpawnRate(SpawnRate=2, Use Spawn Probability=false)
  ]
  particleSpawnScript [
   &FN_SetSpawnGroup(Number of Spawn Groups=6)
   &InitializeParticle(
    Color=&SelectLinearColorFromArray(
     Array Sampling Mode=::NewEnumerator1
     Direct Array Index=$Particles.SpawnGroup
     Color Selection Array=DataInterface() {propertyValues {ColorData [
        <rgba 0.3108 0.8906 0.8727 1>
        <rgba 0.3108 0.8906 0.8727 1>
        <rgba 0.3108 0.8906 0.8727 1>
        <rgba 0.3108 0.8906 0.8727 1>
        <rgba 1 0.166 0.6334 1>
        <rgba 1 0.166 0.6334 1>
       ]}}
     Random Seed=?
    )
    Lifetime Max=3
    Lifetime Min=2
    Lifetime Mode=::NewEnumerator1
    Mesh Scale Mode=::NewEnumerator3
    Mesh Uniform Scale=0.02
    UsePositionOffset=false
    Write Material Random=true
   ) {enabled=false}
   &InitializeParticle(
    Lifetime Max=3
    Lifetime Min=2
    Lifetime Mode=::NewEnumerator1
    Mesh Scale Mode=::NewEnumerator3
    Mesh Uniform Scale=0.02
    UsePositionOffset=false
    Write Material Random=true
   ) {module=InitializeParticle001}
   &ColorFromTOD(
    Afternoon:LinearColor=&MaskLinearColorBySpawnGroup(Masked Linear Color=<rgba 0.1727 1 0.9635 0.5>, Passthrough Linear Color=<rgba 1 0.2578 0.4492 0.5>, Spawn Group Mask=3)
    Evening:LinearColor=&MaskLinearColorBySpawnGroup(Masked Linear Color=<rgba 0.1655 0.3767 0.3289 0.5>, Passthrough Linear Color=<rgba 1 0.1248 0.2748 0.5>, Spawn Group Mask=3)
    FortniteNPC_FortniteActiveTimeOfDay:LinearColor=$NPC.FortniteNPC.FortniteActiveTimeOfDay
    Morning:LinearColor=&MaskLinearColorBySpawnGroup(Masked Linear Color=<rgba 0.1024 0.8288 1 0.5>, Passthrough Linear Color=<rgba 1 0.2038 0.4641 0.5>, Spawn Group Mask=3)
    Night:LinearColor=&MaskLinearColorBySpawnGroup(Masked Linear Color=<rgba 0.0405 0.0987 0.2101 0.5>, Passthrough Linear Color=<rgba 0.5608 0.1801 0.4529 0.5>, Spawn Group Mask=3)
   )
  ]
  particleUpdateScript [
   &SkeletalMeshLocation(
    Bone / Socket Index=$Particles.SpawnGroup
    Bone Sampling Mode=::NewEnumerator4
    MaskSpawnGroup=false
    Sampled Position Offset=/Niagara/DynamicInputs/Lerp/Lerp_VectorComponents(
     Lerp Component Alpha:Vector3f=&VectorFromCurve(VectorCurve=DataInterface() {propertyValues {
        XCurve {keys [
          {arriveTangent=1.2941, interpMode=RCIM_Cubic, leaveTangent=1.2941, tangentMode=RCTM_User, time=0, value=0}
          {arriveTangent=1.4286, interpMode=RCIM_Cubic, leaveTangent=1.4286, time=0.4, value=0.5}
          {arriveTangent=0.4404, interpMode=RCIM_Cubic, leaveTangent=0.4404, tangentMode=RCTM_User, time=0.7, value=1}
         ]}
        YCurve {keys [
          {arriveTangent=0.919, interpMode=RCIM_Cubic, leaveTangent=0.919, tangentMode=RCTM_User, time=0.3, value=0}
          {arriveTangent=2.2222, interpMode=RCIM_Cubic, leaveTangent=2.2222, time=0.5, value=0.3}
          {arriveTangent=3.3657, interpMode=RCIM_Cubic, leaveTangent=3.3657, tangentMode=RCTM_User, time=0.75, value=1}
         ]}
        ZCurve {keys [
          {arriveTangent=0.5832, interpMode=RCIM_Cubic, leaveTangent=0.5832, tangentMode=RCTM_User, time=0.3, value=0}
          {arriveTangent=2.1909, interpMode=RCIM_Cubic, leaveTangent=2.1909, tangentMode=RCTM_User, time=0.5, value=0.3}
          {arriveTangent=4.9832, interpMode=RCIM_Cubic, leaveTangent=4.9832, tangentMode=RCTM_User, time=0.7, value=1}
         ]}
       }})
     A:Vector3f=<2 0 0>
     B:Vector3f=&SelectVectorFromArray(
      Direct Array Index=$Particles.SpawnGroup
      Vector Selection Array=DataInterface() {propertyValues {FloatData [<4 2.5 0>, <6.9 0 3.5>, <7.5 -0.7 0>, <3.3 -0.5 1>, <5 0.5 0>, <2 0.7 1>]}}
     )
    )
    Skeletal Mesh=$User.HairTips
   ) {module=SkeletalMeshLocation001}
   &ParticleState()
   &ScaleMeshSize(Scale Factor=&VectorFromFloat(Value=&FloatFromCurve(FloatCurve=DataInterface() {propertyValues {Curve {keys [
          {arriveTangent=0.5525, interpMode=RCIM_Cubic, leaveTangent=0.5525, tangentMode=RCTM_User, time=0, value=0}
          {arriveTangent=1.25, interpMode=RCIM_Cubic, leaveTangent=1.25, time=0.6, value=0.95}
          {arriveTangent=-0.4749, interpMode=RCIM_Cubic, leaveTangent=-0.4749, tangentMode=RCTM_User, time=0.8, value=1}
          <at 1 0 RCIM_Cubic>
         ]}}})))
   &UpdateMeshOrientation(Rotation Rate=2)
   /SetVariables(Particles.VisibilityTag=&MakeCustomIntFromBool(Bool=$User.bisHeadHidden, True INT=1, False INT=0)) {isSetParameters=true, module=SetVariables_#3}
  ]
  rendererClasses [&NiagaraMeshRendererProperties]
  renderers [{rendererClass=&NiagaraMeshRendererProperties, rendererIndex=0}]
 }
}
summary {
 systemName=NS_Thrive_Pink_Hair
 emitters [
  {emitterName=HairEndDrops, rendererClasses [&NiagaraMeshRendererProperties]}
  {emitterName=HairDropsUp, rendererClasses [&NiagaraMeshRendererProperties]}
  {emitterName=EaringGlow, rendererClasses [&NiagaraSpriteRendererProperties]}
 ]
 userVariables [
  User.bisFrontEnd {type=NiagaraBool}
  User.bIsFrontEndPreview {type=NiagaraBool}
  User.bisHeadHidden {type=NiagaraBool}
  User.Color_Tint {type=LinearColor, description="", defaultValue {struct=LinearColor, value=<rgba 1 1 1 1>}}
  User.Debug_Reactive_Level {type=NiagaraFloat, defaultValue=0, description=""}
  User.EmissiveIntensity {type=NiagaraFloat, defaultValue=0, description=""}
  User.FE_Preview {type=NiagaraFloat, defaultValue=0, description=""}
  User.GlobalDelay {type=NiagaraFloat, defaultValue=0, description=""}
  User.HairTips {
   type=NiagaraDataInterfaceSkeletalMesh
   defaultValue {
    struct=NiagaraExt_VariableValue_DataInterface
    value {dataInterface=&self:NiagaraDataInterfaceSkeletalMesh_4}
   }
  }
  User.HeadBones {
   type=NiagaraDataInterfaceSkeletalMesh
   defaultValue {
    struct=NiagaraExt_VariableValue_DataInterface
    value {dataInterface=&self:NiagaraDataInterfaceSkeletalMesh_3}
   }
  }
  User.RibbonIDOffset {type=NiagaraInt32}
  User.SkeletalMeshEarings {
   type=NiagaraDataInterfaceSkeletalMesh
   defaultValue {
    struct=NiagaraExt_VariableValue_DataInterface
    value {dataInterface=&self:NiagaraDataInterfaceSkeletalMesh_5}
   }
  }
  User.Thrive_EffectDecayRate {type=NiagaraFloat, defaultValue=0, description=""}
  User.Thrive_EffectLevel {type=NiagaraFloat, defaultValue=0, description=""}
  User.Thrive_GameTimeFromBP {type=NiagaraFloat, defaultValue=0, description=""}
  User.Thrive_MaxLevel {type=NiagaraFloat, defaultValue=0, description=""}
  User.Thrive_Previous_Intensity {type=NiagaraFloat, defaultValue=0, description=""}
  User.Thrive_Target_Intensity {type=NiagaraFloat, defaultValue=0, description=""}
 ]
}
