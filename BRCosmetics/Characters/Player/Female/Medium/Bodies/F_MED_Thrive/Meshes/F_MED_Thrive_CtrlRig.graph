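# Control Rig for the F_MED_Thrive female body that applies procedural corrective motion to the upper arm and skirt bones.
# Right upper-arm twist, isolated by swing-twist, partly counter-rotates its deform bone, and a parent constraint blends center lower skirt between both front panels.
# Spherical pose readers on the thighs weight fixed local transforms onto skirt masters: left drives left-front and center, right drives right-front.
# An additive bone-offset helper function is defined but unused.
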ControlRigBlueprint /BRCosmetics/Characters/Player/Female/Medium/Bodies/F_MED_Thrive/Meshes/F_MED_Thrive_CtrlRig
graphs [
 ContainedGraph {nodes [
   n1 = RigVMFunctionEntryNode()
   RigUnit_GetTransform = &RigUnit_GetTransform(Space=LocalSpace, bInitial=false, Item=@n1.Joint)
   n2 = &RigVMFunction_MathQuaternionToEuler(RotationOrder=ZYX, Value=@RigUnit_GetTransform.Transform.Rotation)
   > RigUnit_SetTransform = &RigUnit_SetTransform(
    Space=LocalSpace
    Value=(Rotation=(X=0.0132,Y=0.4099,Z=-0.0222,W=0.9118),Translation=(X=23.6045,Y=-6.37,Z=140.1115),Scale3D=(X=1,Y=1,Z=1))
    bInitial=false
    Item=@n1.Joint
    Weight=@n1.Weight
    Value.Translation.Z=&RigVMFunction_MathFloatAdd(B=@n1.Translate Z, A=@RigUnit_GetTransform.Transform.Translation.Z) {template="Add::Execute(in A,in B,out Result)"}.Result
    Value.Scale3D.Z=&RigVMFunction_MathFloatAdd(B=@n1.Scale Z, A=@RigUnit_GetTransform.Transform.Scale3D.Z) {template="Add::Execute(in A,in B,out Result)"}.Result
    Value.Translation.Y=&RigVMFunction_MathFloatAdd(B=@n1.Translate Y, A=@RigUnit_GetTransform.Transform.Translation.Y) {template="Add::Execute(in A,in B,out Result)"}.Result
    Value.Scale3D.Y=&RigVMFunction_MathFloatAdd(B=@n1.Scale Y, A=@RigUnit_GetTransform.Transform.Scale3D.Y) {template="Add::Execute(in A,in B,out Result)"}.Result
    Value.Translation.X=&RigVMFunction_MathFloatAdd(B=@n1.Translate X, A=@RigUnit_GetTransform.Transform.Translation.X) {template="Add::Execute(in A,in B,out Result)"}.Result
    Value.Scale3D.X=&RigVMFunction_MathFloatAdd(B=@n1.Scale X, A=@RigUnit_GetTransform.Transform.Scale3D.X) {template="Add::Execute(in A,in B,out Result)"}.Result
    Value.Rotation=&RigVMFunction_MathQuaternionFromEuler(
     Euler=(X=0,Y=0,Z=0)
     Euler.X=&RigVMFunction_MathFloatAdd(B=@n1.Rotate X, A=@n2.Result.X) {template="Add::Execute(in A,in B,out Result)"}.Result
     Euler.Y=&RigVMFunction_MathFloatAdd(B=@n1.Rotate Y, A=@n2.Result.Y) {template="Add::Execute(in A,in B,out Result)"}.Result
     Euler.Z=&RigVMFunction_MathFloatAdd(B=@n1.Rotate Z, A=@n2.Result.Z) {template="Add::Execute(in A,in B,out Result)"}.Result
    ).Result
   )
   > RigVMFunctionReturnNode()
  ]}
 RigVMFunctionLibrary {nodes [Set Transforms for Bone = RigVMCollapseNode(
    ExecuteContext=()
    Joint=()
    Rotate X=0
    Rotate Y=0
    Rotate Z=0
    Scale X=0
    Scale Y=0
    Scale Z=0
    Translate X=0
    Translate Y=0
    Translate Z=0
    Weight=1
   )]}
 RigVMModel {nodes [
   RigUnit_BeginExecution = &RigUnit_BeginExecution()
   n1 = /Script/RigVM.RigVMFunction_MathQuaternionToAxisAndAngle(Value=&RigVMFunction_MathQuaternionSwingTwist(
     TwistAxis=(X=1,Y=0,Z=0)
     Input=&RigUnit_GetTransform(Item=(Type=Bone,Name="upperarm_r"), Space=LocalSpace).Transform.Rotation
    ).Twist) {is=struct RigVMUnitNode, template="ToAxisAndAngle::Execute(in Value,out Axis,out Angle)"}
   &RigUnit_SetRotation(
    Item=(Type=Bone,Name="upperarm_deformTwist_r")
    Weight=1
    Value=/Script/RigVM.RigVMFunction_MathQuaternionFromAxisAndAngle(
     Angle=&RigVMFunction_MathFloatMul(B=-0.4, A=@n1.Angle) {template="Multiply::Execute(in A,in B,out Result)"}.Result
     Axis=@n1.Axis
    ) {is=struct RigVMUnitNode, template="FromAxisAndAngle::Execute(in Axis,in Angle,out Result)"}.Result
    ExecutePin=@RigUnit_BeginExecution.ExecutePin
   )
   > &RigUnit_ParentConstraint(
    AdvancedSettings=(RotationOrderForFilter=XZY)
    Child=(Type=Bone,Name="dyn_center_lowerSkirt_2")
    Parents=((Item=(Type=Bone,Name="dyn_L_front_LowerSkirt_2"),Weight=0.5),(Item=(Type=Bone,Name="dyn_R_front_LowerSkirt_2"),Weight=0.5))
    Weight=0.6
   )
   > n2 = &RigUnit_SphericalPoseReader(
    ActiveRegionScaleFactors=(PositiveWidth=1.1,NegativeWidth=1.1,PositiveHeight=1.1,NegativeHeight=1.1)
    ActiveRegionSize=0.1
    Debug=(bDrawDebug=false,bDraw2D=false,bDrawLocalAxes=false,DebugScale=25,DebugSegments=20,DebugThickness=0.4)
    DriverAxis=(X=-1,Y=0,Z=0)
    DriverItem=(Type=Bone,Name="thigh_l")
    FalloffRegionScaleFactors=(PositiveWidth=2,NegativeWidth=2,PositiveHeight=2,NegativeHeight=2)
    FalloffSize=0.2
    FlipHeightScaling=false
    FlipWidthScaling=false
    OptionalParentItem=(Type=Bone,Name="spine_01")
    RotationOffset=(X=120,Y=20,Z=-15)
   )
   > &RigUnit_SetTransform(
    Item=(Type=Bone,Name="dyn_L_front_LowerSkirt_master")
    Space=LocalSpace
    Value=(Rotation=(X=0,Y=0,Z=0,W=1),Translation=(X=-30,Y=0,Z=-20),Scale3D=(X=1,Y=1,Z=1))
    Weight=@n2.OutputParam
   )
   n3 = &RigVMFunction_MathQuaternionFromEuler(Euler=(X=0,Y=50,Z=0))
   > &RigUnit_SetTransform(
    Item=(Type=Bone,Name="dyn_center_lowerSkirt_master")
    Space=LocalSpace
    Value=(Rotation=(X=0,Y=0,Z=0,W=1),Translation=(X=9,Y=0,Z=5),Scale3D=(X=1,Y=1,Z=1))
    Value.Rotation=@n3.Result
    Weight=@n2.OutputParam
   )
   > &RigUnit_SetTransform(
    Item=(Type=Bone,Name="dyn_L_front_UpperSkirt_master")
    Space=LocalSpace
    Value=(Rotation=(X=0,Y=0,Z=0,W=1),Translation=(X=-30,Y=0,Z=-20),Scale3D=(X=1,Y=1,Z=1))
    Weight=@n2.OutputParam
   )
   > &RigUnit_SetTransform(
    Item=(Type=Bone,Name="dyn_center_upperSkirt_master")
    Space=LocalSpace
    Value=(Rotation=(X=0,Y=0,Z=0,W=1),Translation=(X=9,Y=0,Z=5),Scale3D=(X=1,Y=1,Z=1))
    Value.Rotation=@n3.Result
    Weight=@n2.OutputParam
   )
   > n4 = &RigUnit_SphericalPoseReader(
    ActiveRegionScaleFactors=(PositiveWidth=1,NegativeWidth=1,PositiveHeight=1,NegativeHeight=1)
    ActiveRegionSize=0.25
    Debug=(bDrawDebug=false,bDraw2D=false,bDrawLocalAxes=false,DebugScale=25,DebugSegments=20,DebugThickness=0.4)
    DriverAxis=(X=1,Y=0,Z=0)
    DriverItem=(Type=Bone,Name="thigh_r")
    FalloffRegionScaleFactors=(PositiveWidth=2,NegativeWidth=2,PositiveHeight=2,NegativeHeight=2)
    FalloffSize=0.3
    FlipHeightScaling=false
    FlipWidthScaling=false
    OptionalParentItem=(Type=Bone,Name="spine_01")
    RotationOffset=(X=-90,Y=-20,Z=-15)
   )
   n5 = &RigVMFunction_MathQuaternionFromEuler(Euler=(X=0,Y=0,Z=-70))
   > &RigUnit_SetTransform(
    Item=(Type=Bone,Name="dyn_R_front_LowerSkirt_master")
    Space=LocalSpace
    Value=(Rotation=(X=0,Y=0,Z=0,W=1),Translation=(X=0,Y=0,Z=0),Scale3D=(X=1,Y=1,Z=1))
    bInitial=false
    Value.Rotation=@n5.Result
    Weight=@n4.OutputParam
   )
   > &RigUnit_SetTransform(
    Item=(Type=Bone,Name="dyn_R_front_UpperSkirt_master")
    Space=LocalSpace
    Value=(Rotation=(X=0,Y=0,Z=0,W=1),Translation=(X=0,Y=0,Z=0),Scale3D=(X=1,Y=1,Z=1))
    bInitial=false
    Value.Rotation=@n5.Result
    Weight=@n4.OutputParam
   )
  ]}
]
hierarchy [
 root {loc=<0 0 0>, rot=<0 0 0 1>}
 pelvis {parent=root, loc=<0 -0.5301 98.0571>, rot=<0.0127 0.707 -0.0127 -0.707>}
 spine_01 {parent=pelvis, loc=<6.1208 -0.9328 0>, rot=<0 0 0.034 -0.9994>}
 spine_02 {parent=spine_01, loc=<6.5996 -0.2095 0>, rot=<0 0 0.0238 -0.9997>}
 spine_03 {parent=spine_02, loc=<6.4914 -0.1237 0>, rot=<0 0 0.0149 -0.9999>}
 spine_04 {parent=spine_03, loc=<6.9598 0.0794 0>, rot=<0 0 -0.0132 -0.9999>}
 spine_05 {parent=spine_04, loc=<6.9987 0 0>, rot=<0 0 -0.0015 -1>}
 clavicle_l {parent=spine_05, loc=<5.4831 -0.1053 -3.2946>, rot=<0.0465 -0.7289 -0.009 -0.6829>}
 upperarm_l {parent=clavicle_l, loc=<9.8658 -1.2136 -0.4105>, rot=<0.0164 -0.3245 -0.0269 -0.9454>}
 lowerarm_l {parent=upperarm_l, loc=<26.465 -0.8345 0.259>, rot=<0.0036 0.0161 -0.2298 -0.9731>}
 hand_l {parent=lowerarm_l, loc=<22.7355 -0.3203 -0.6105>, rot=<-0.6784 -0.0223 0.0735 -0.7307>}
 index_metacarpal_l {parent=hand_l, loc=<1.6535 0.0183 -2.96>, rot=<0.0158 -0.0124 -0.0173 -0.9996>}
 index_01_l {parent=index_metacarpal_l, loc=<8.3008 0.0357 -0.0524>, rot=<-0.0044 0.0485 0.0924 -0.9945>}
 index_02_l {parent=index_01_l, loc=<3.3587 -0.3004 0.0452>, rot=<-0.002 0.0027 0.3225 -0.9466>}
 index_03_l {parent=index_02_l, loc=<2.5578 -0.123 -0.0252>, rot=<-0.002 0.0067 0.2558 -0.9667>}
 middle_metacarpal_l {parent=hand_l, loc=<2.1856 -0.1725 -0.5984>, rot=<0.006 0.0168 -0.0329 -0.9993>}
 middle_01_l {parent=middle_metacarpal_l, loc=<7.3275 -0.2447 0.0129>, rot=<-0.0208 0.0042 0.0858 -0.9961>}
 middle_02_l {parent=middle_01_l, loc=<4.0123 0.1401 0.1175>, rot=<-0.0218 0.0258 0.3512 -0.9357>}
 middle_03_l {parent=middle_02_l, loc=<2.6077 0.0534 0.1373>, rot=<0.0082 0.0188 0.1708 -0.9851>}
 pinky_metacarpal_l {parent=hand_l, loc=<1.7616 -0.9753 3.3814>, rot=<-0.0397 0.0814 -0.069 -0.9935>}
 pinky_01_l {parent=pinky_metacarpal_l, loc=<7.1709 -0.4831 -0.1094>, rot=<-0.0654 -0.0035 0.2242 -0.9723>}
 pinky_02_l {parent=pinky_01_l, loc=<2.8274 0.0893 -0.1043>, rot=<0.0037 0.0011 0.2854 -0.9584>}
 pinky_03_l {parent=pinky_02_l, loc=<2.2046 0.0307 -0.0078>, rot=<-0.0002 0.0098 0.1445 -0.9895>}
 ring_metacarpal_l {parent=hand_l, loc=<2.2074 -0.3625 1.3638>, rot=<-0.0284 0.0722 -0.0458 -0.9959>}
 ring_01_l {parent=ring_metacarpal_l, loc=<7.0295 -0.6197 -0.1446>, rot=<-0.0489 -0.0252 0.1892 -0.9804>}
 ring_02_l {parent=ring_01_l, loc=<3.6174 0.4647 -0.1027>, rot=<0 0 0.2806 -0.9598>}
 ring_03_l {parent=ring_02_l, loc=<2.4152 0.114 0.0199>, rot=<0 0 0.1189 -0.9929>}
 thumb_01_l {parent=hand_l, loc=<1.3118 -0.3191 -2.459>, rot=<0.6546 -0.4042 0.0664 -0.6354>}
 thumb_02_l {parent=thumb_01_l, loc=<5.9336 -0.0448 0.0401>, rot=<-0.0037 0.0213 0.298 -0.9543>}
 thumb_03_l {parent=thumb_02_l, loc=<2.2223 0.3682 -0.0907>, rot=<0.0118 -0.013 0.0066 -0.9998>}
 weapon_l {parent=hand_l, loc=<-0.1974 -3.4111 0.3811>, rot=<0 0 0 1>}
 lowerarm_twist_01_l {parent=lowerarm_l, loc=<18.316 0 0>, rot=<0 0 0 1>}
 lowerarm_twist_02_l {parent=lowerarm_l, loc=<10.6707 0 0>, rot=<0 0 0 1>}
 upperarm_twist_01_l {parent=upperarm_l, loc=<7.2135 -0.1134 0>, rot=<0 0 0 1>}
 upperarm_twist_02_l {parent=upperarm_l, loc=<19.7998 -0.1134 0>, rot=<0 0 0 1>}
 dyn_L_sleeve_1 {parent=upperarm_l, loc=<22.938 -3.5292 4.0483>, rot=<-0.2315 -0.475 0.8489 0.0147>}
 dyn_L_sleeve_2 {parent=dyn_L_sleeve_1, loc=<-6.5338 0 0>, rot=<-0.0848 -0.2195 0.0055 -0.9719>}
 dyn_L_sleeve_3 {parent=dyn_L_sleeve_2, loc=<-5.8008 0.0001 0>, rot=<0 0 0 1>}
 dyn_L_sleeve_END {parent=dyn_L_sleeve_3, loc=<-4.974 0 0.8395>, rot=<0 0 0 1>}
 dyn_L_underSleeve {parent=upperarm_l, loc=<20.637 0.6176 -5.0568>, rot=<0.0876 -0.0076 0.9961 0.0006>}
 upperarm_deformTwist_l {parent=upperarm_l, loc=<-0.0003 0 0.0003>, rot=<0 0 0 1>}
 dyn_L_armClothFix {parent=upperarm_l, loc=<18.7666 3.2927 -0.1225>, rot=<0.9979 0.0648 -0.0082 -0.0056>}
 clavicle_r {parent=spine_05, loc=<5.4831 -0.1053 3.2946>, rot=<0.7289 0.0465 -0.6829 0.009>}
 upperarm_r {parent=clavicle_r, loc=<-9.8658 1.2136 0.4104>, rot=<0.0164 -0.3245 -0.0269 -0.9454>}
 lowerarm_r {parent=upperarm_r, loc=<-26.4651 0.8345 -0.259>, rot=<0.0036 0.0161 -0.2298 -0.9731>}
 hand_r {parent=lowerarm_r, loc=<-22.7356 0.3203 0.6106>, rot=<-0.6784 -0.0223 0.0735 -0.7307>}
 index_metacarpal_r {parent=hand_r, loc=<-1.6534 -0.0181 2.9601>, rot=<0.0158 -0.0124 -0.0173 -0.9996>}
 index_01_r {parent=index_metacarpal_r, loc=<-8.3008 -0.0357 0.0524>, rot=<-0.0044 0.0485 0.0924 -0.9945>}
 index_02_r {parent=index_01_r, loc=<-3.3589 0.3003 -0.0453>, rot=<-0.002 0.0027 0.3225 -0.9466>}
 index_03_r {parent=index_02_r, loc=<-2.5578 0.123 0.0252>, rot=<-0.002 0.0067 0.2558 -0.9667>}
 middle_metacarpal_r {parent=hand_r, loc=<-2.1856 0.1726 0.5984>, rot=<0.006 0.0168 -0.0329 -0.9993>}
 middle_01_r {parent=middle_metacarpal_r, loc=<-7.3273 0.2447 -0.0129>, rot=<-0.0208 0.0042 0.0858 -0.9961>}
 middle_02_r {parent=middle_01_r, loc=<-4.0124 -0.14 -0.1175>, rot=<-0.0218 0.0258 0.3512 -0.9357>}
 middle_03_r {parent=middle_02_r, loc=<-2.6077 -0.0533 -0.1373>, rot=<0.0082 0.0188 0.1708 -0.9851>}
 pinky_metacarpal_r {parent=hand_r, loc=<-1.7617 0.9753 -3.3814>, rot=<-0.0397 0.0814 -0.069 -0.9935>}
 pinky_01_r {parent=pinky_metacarpal_r, loc=<-7.1707 0.4832 0.1094>, rot=<-0.0654 -0.0035 0.2242 -0.9723>}
 pinky_02_r {parent=pinky_01_r, loc=<-2.8274 -0.0892 0.1043>, rot=<0.0037 0.0011 0.2854 -0.9584>}
 pinky_03_r {parent=pinky_02_r, loc=<-2.2045 -0.031 0.0078>, rot=<-0.0002 0.0098 0.1445 -0.9895>}
 ring_metacarpal_r {parent=hand_r, loc=<-2.2073 0.3627 -1.3638>, rot=<-0.0284 0.0722 -0.0458 -0.9959>}
 ring_01_r {parent=ring_metacarpal_r, loc=<-7.0297 0.6197 0.1446>, rot=<-0.0489 -0.0252 0.1892 -0.9804>}
 ring_02_r {parent=ring_01_r, loc=<-3.6171 -0.4648 0.1027>, rot=<0 0 0.2806 -0.9598>}
 ring_03_r {parent=ring_02_r, loc=<-2.4153 -0.1141 -0.0199>, rot=<0 0 0.1189 -0.9929>}
 thumb_01_r {parent=hand_r, loc=<-1.3119 0.3192 2.459>, rot=<0.6546 -0.4042 0.0664 -0.6354>}
 thumb_02_r {parent=thumb_01_r, loc=<-5.9335 0.0447 -0.0401>, rot=<-0.0037 0.0213 0.298 -0.9543>}
 thumb_03_r {parent=thumb_02_r, loc=<-2.2223 -0.3682 0.0907>, rot=<0.0118 -0.013 0.0066 -0.9998>}
 weapon_r {parent=hand_r, loc=<0.1975 3.4112 -0.3811>, rot=<0 0 0 -1>}
 lowerarm_twist_01_r {parent=lowerarm_r, loc=<-18.3161 0 0>, rot=<0 0 0 1>}
 lowerarm_twist_02_r {parent=lowerarm_r, loc=<-10.6707 0 0>, rot=<0 0 0 1>}
 upperarm_twist_01_r {parent=upperarm_r, loc=<-7.2135 0.1134 0>, rot=<0 0 0 1>}
 upperarm_twist_02_r {parent=upperarm_r, loc=<-19.7998 0.1134 0>, rot=<0 0 0 1>}
 dyn_R_sleeve_1 {parent=upperarm_r, loc=<-22.9379 3.5293 -4.0484>, rot=<-0.2315 -0.475 0.8489 0.0147>}
 dyn_R_sleeve_2 {parent=dyn_R_sleeve_1, loc=<6.5337 0 0>, rot=<-0.0848 -0.2195 0.0055 -0.9719>}
 dyn_R_sleeve_3 {parent=dyn_R_sleeve_2, loc=<5.8008 0 0>, rot=<0 0 0 1>}
 dyn_R_sleeve_END {parent=dyn_R_sleeve_3, loc=<4.8876 0.514 -0.6107>, rot=<0 0 0 1>}
 dyn_R_underSleeve {parent=upperarm_r, loc=<-20.6372 -0.6176 5.0568>, rot=<0.0876 -0.0076 0.9961 0.0006>}
 upperarm_deformTwist_r {parent=upperarm_r, loc=<0 0 0>, rot=<0 0 0 1>}
 dyn_R_armClothFix {parent=upperarm_r, loc=<-18.7666 -3.2927 0.1224>, rot=<0.9979 0.0648 -0.0082 -0.0056>}
 neck_01 {parent=spine_05, loc=<11.3138 0.1529 0>, rot=<0 0 -0.2003 -0.9797>}
 dyn_necklace {parent=neck_01, loc=<-0.5476 0.0643 -0.0966>, rot=<0 0 0 1>}
 dyn_necklaceBall_1 {parent=dyn_necklace, loc=<0.2013 -2.4569 7.7154>, rot=<0.6742 -0.0529 0.7301 -0.098>}
 dyn_necklaceBall_2 {parent=dyn_necklace, loc=<0.5668 -8.0815 -0.4447>, rot=<0.449 -0.4894 0.504 -0.5522>}
 dyn_necklaceBall_3 {parent=dyn_necklace, loc=<0.0194 -1.8442 -7.7261>, rot=<-0.0641 -0.7042 0.1628 -0.6881>}
 dyn_necklaceBall_4 {parent=dyn_necklace, loc=<-0.3396 6.2155 -4.4332>, rot=<-0.4545 -0.5417 -0.259 -0.658>}
 dyn_necklaceBall_5 {parent=dyn_necklace, loc=<-0.4492 6.8234 3.8554>, rot=<0.6529 0.2716 0.5328 0.4648>}
 neck_02 {parent=neck_01, loc=<4.5627 0 0>, rot=<0 0 0.0504 -0.9987>}
 head {parent=neck_02, loc=<5.3914 0 0>, rot=<0 0 0.115 -0.9934>}
 dyn_socketAttach {parent=spine_05, loc=<0.7005 -9.9792 0>, rot=<0.0284 -0.7065 -0.0284 -0.7065>}
 dyn_cape_socket {parent=dyn_socketAttach, loc=<0 0.775 7.0186>, rot=<0 0 0 1>}
 dyn_backpack_socket {parent=dyn_socketAttach, loc=<0 0 0>, rot=<0 0 0 1>}
 thigh_l {parent=pelvis, loc=<-1.0855 -0.2109 -8.7053>, rot=<0.0009 0.0355 0.0155 -0.9993>}
 calf_l {parent=thigh_l, loc=<-42.5284 -0.0019 -0.7297>, rot=<0.0006 -0.0161 -0.0391 -0.9991>}
 calf_twist_01_l {parent=calf_l, loc=<-35.8483 -1.9197 -0.5794>, rot=<0.0237 0.0095 -0.0035 -0.9997>}
 calf_twist_02_l {parent=calf_l, loc=<-19.1546 -1.7943 -0.2638>, rot=<0.0237 0.0095 -0.0035 -0.9997>}
 foot_l {parent=calf_l, loc=<-46.7402 -2.5014 -0.592>, rot=<0.0427 -0.0238 0.0353 -0.9982>}
 ball_l {parent=foot_l, loc=<-3.5658 10.8066 -0.1544>, rot=<0.0001 -0.0001 -0.7063 -0.7079>}
 dyn_L_shoeTongue {parent=foot_l, loc=<2.1777 7.5838 0.2679>, rot=<0.1498 -0.6853 -0.1346 -0.6998>}
 thigh_twist_01_l {parent=thigh_l, loc=<-27.952 0 0>, rot=<0.0013 0 0 -1>}
 dyn_L_side_LowerSkirt_master {parent=thigh_l, loc=<4.5016 -1.0844 -4.9722>, rot=<0.0542 -0.9782 0.0119 -0.2002>}
 dyn_L_side_LowerSkirt_1 {parent=dyn_L_side_LowerSkirt_master, loc=<0 0 0>, rot=<0 0 0 1>}
 dyn_L_side_LowerSkirt_2 {parent=dyn_L_side_LowerSkirt_1, loc=<6.934 0 0>, rot=<0 0 0 1>}
 dyn_L_side_LowerSkirt_3 {parent=dyn_L_side_LowerSkirt_2, loc=<8.1481 0 0>, rot=<0 0 0 1>}
 dyn_L_front_LowerSkirt_master {parent=thigh_l, loc=<4.4248 8.681 1.3081>, rot=<-0.1059 -0.989 -0.0107 -0.1027>}
 dyn_L_front_LowerSkirt_1 {parent=dyn_L_front_LowerSkirt_master, loc=<0 0 0>, rot=<0 0 0 1>}
 dyn_L_front_LowerSkirt_2 {parent=dyn_L_front_LowerSkirt_1, loc=<6.531 -2.7453 -0.1793>, rot=<0 0 0 1>}
 dyn_L_front_LowerSkirt_3 {parent=dyn_L_front_LowerSkirt_2, loc=<8.0085 -0.6601 0>, rot=<0 0 0 1>}
 thigh_r {parent=pelvis, loc=<-1.0853 -0.2109 8.7053>, rot=<0.0355 -0.0009 0.9993 0.0155>}
 calf_r {parent=thigh_r, loc=<42.5286 0.0019 0.7297>, rot=<0.0006 -0.0161 -0.0391 -0.9991>}
 calf_twist_01_r {parent=calf_r, loc=<35.8485 1.9197 0.5794>, rot=<0.0237 0.0095 -0.0035 -0.9997>}
 calf_twist_02_r {parent=calf_r, loc=<19.1546 1.7943 0.2638>, rot=<0.0237 0.0095 -0.0035 -0.9997>}
 foot_r {parent=calf_r, loc=<46.7403 2.5014 0.592>, rot=<0.0427 -0.0238 0.0353 -0.9982>}
 ball_r {parent=foot_r, loc=<3.5658 -10.8066 0.1544>, rot=<0.0001 -0.0001 -0.7063 -0.7079>}
 dyn_R_shoeTongue {parent=foot_r, loc=<-2.1779 -7.5838 -0.2678>, rot=<0.1498 -0.6853 -0.1346 -0.6998>}
 thigh_twist_01_r {parent=thigh_r, loc=<27.952 0 0>, rot=<0.0013 0 0 -1>}
 dyn_R_front_LowerSkirt_master {parent=thigh_r, loc=<-4.4246 -8.681 -1.3081>, rot=<-0.0107 0.1027 0.1059 -0.989>}
 dyn_R_front_LowerSkirt_1 {parent=dyn_R_front_LowerSkirt_master, loc=<0 0 0>, rot=<0 0 0 1>}
 dyn_R_front_LowerSkirt_2 {parent=dyn_R_front_LowerSkirt_1, loc=<6.531 2.7453 -0.1793>, rot=<0 0 0 1>}
 dyn_R_front_LowerSkirt_3 {parent=dyn_R_front_LowerSkirt_2, loc=<8.0085 0.6601 0>, rot=<0 0 0 1>}
 dyn_R_side_LowerSkirt_master {parent=thigh_r, loc=<-4.5014 1.0844 4.9722>, rot=<0.0119 0.2002 -0.0542 -0.9782>}
 dyn_R_side_LowerSkirt_1 {parent=dyn_R_side_LowerSkirt_master, loc=<0 0 0>, rot=<0 0 0 1>}
 dyn_R_side_LowerSkirt_2 {parent=dyn_R_side_LowerSkirt_1, loc=<6.934 0 0>, rot=<0 0 0 1>}
 dyn_R_side_LowerSkirt_3 {parent=dyn_R_side_LowerSkirt_2, loc=<8.1481 0 0>, rot=<0 0 0 1>}
 dyn_bow_master_jnt {parent=pelvis, loc=<5.2478 -15.3232 -0.3389>, rot=<0 0 0.0179 -0.9998>}
 dyn_bow_C {parent=dyn_bow_master_jnt, loc=<0 0 0>, rot=<0 0 0 1>}
 dyn_bow_L {parent=dyn_bow_C, loc=<0.3226 0.9625 -6.9736>, rot=<0 0.7071 0 -0.7071>}
 dyn_bow_R {parent=dyn_bow_C, loc=<0.3222 0.9625 7.6515>, rot=<0 0.7071 0 -0.7071>}
 dyn_L_bowStrand_1 {parent=dyn_bow_C, loc=<-5.8528 -4.1511 -7.6213>, rot=<0.9363 0.2313 0.2565 0.0634>}
 dyn_L_bowStrand_2 {parent=dyn_L_bowStrand_1, loc=<-14.0012 0 0>, rot=<0.0023 -0.0045 -0.0019 -1>}
 dyn_L_bowStrand_3 {parent=dyn_L_bowStrand_2, loc=<-14.4033 0 0.0001>, rot=<0.0049 -0.0111 -0.0615 -0.998>}
 dyn_L_bowStrand_4 {parent=dyn_L_bowStrand_3, loc=<-12.946 0 0>, rot=<-0.0143 0.0529 -0.0939 -0.9941>}
 dyn_L_bowStrand_5 {parent=dyn_L_bowStrand_4, loc=<-10.4944 0 0.0001>, rot=<-0.3845 0.0481 -0.1144 -0.9148>}
 dyn_R_bowStrand_1 {parent=dyn_bow_C, loc=<-5.8528 -4.1511 8.2992>, rot=<-0.0634 -0.2565 -0.2313 -0.9363>}
 dyn_R_bowStrand_2 {parent=dyn_R_bowStrand_1, loc=<-14.0012 0 0>, rot=<-0.0023 -0.0045 0.0019 -1>}
 dyn_R_bowStrand_3 {parent=dyn_R_bowStrand_2, loc=<-14.4032 0 0>, rot=<-0.0049 -0.0111 0.0615 -0.998>}
 dyn_R_bowStrand_4 {parent=dyn_R_bowStrand_3, loc=<-12.9461 0 0>, rot=<0.0143 0.0529 0.0939 -0.9941>}
 dyn_R_bowStrand_5 {parent=dyn_R_bowStrand_4, loc=<-10.4943 0 0>, rot=<0.3845 0.0481 0.1144 -0.9148>}
 dyn_skirtStar_A {parent=pelvis, loc=<-13.9479 17.7634 0.5087>, rot=<-0.0127 -0.707 0.0127 -0.707>}
 dyn_skirtStar_B {parent=pelvis, loc=<-12.1109 11.0467 -17.0392>, rot=<-0.0127 -0.707 0.0127 -0.707>}
 dyn_skirtStar_C {parent=pelvis, loc=<-10.9642 -1.8415 -22.982>, rot=<-0.0127 -0.707 0.0127 -0.707>}
 dyn_skirtStar_D {parent=pelvis, loc=<-10.3215 -16.1994 -15.1801>, rot=<-0.0127 -0.707 0.0127 -0.707>}
 dyn_skirtStar_E {parent=pelvis, loc=<-11.7534 -20.4258 0.4491>, rot=<-0.0127 -0.707 0.0127 -0.707>}
 dyn_skirtStar_F {parent=pelvis, loc=<-10.9259 -16.5066 13.9575>, rot=<-0.0127 -0.707 0.0127 -0.707>}
 dyn_skirtStar_G {parent=pelvis, loc=<-10.7572 -3.2747 23.7403>, rot=<-0.0127 -0.707 0.0127 -0.707>}
 dyn_skirtStar_H {parent=pelvis, loc=<-12.3085 13.7086 17.0273>, rot=<-0.0127 -0.707 0.0127 -0.707>}
 dyn_L_front_UpperSkirt_master {parent=pelvis, loc=<3.5027 8.3331 -7.0928>, rot=<-0.1206 -0.983 -0.0119 -0.138>}
 dyn_L_front_UpperSkirt_1 {parent=dyn_L_front_UpperSkirt_master, loc=<0 0 0>, rot=<0 0 0 1>}
 dyn_L_front_UpperSkirt_2 {parent=dyn_L_front_UpperSkirt_1, loc=<6.0216 -0.1831 -0.1793>, rot=<0 0 0 1>}
 dyn_L_side_UpperSkirt_master {parent=pelvis, loc=<3.7215 -1.4482 -13.345>, rot=<0.0388 -0.9712 0.0178 -0.2345>}
 dyn_L_side_UpperSkirt_1 {parent=dyn_L_side_UpperSkirt_master, loc=<0 0 0>, rot=<0 0 0 1>}
 dyn_L_side_UpperSkirt_2 {parent=dyn_L_side_UpperSkirt_1, loc=<5.8829 0 0>, rot=<0 0 0 1>}
 dyn_L_back_UpperSkirt_master {parent=pelvis, loc=<3.5607 -9.6629 -7.6166>, rot=<0.2546 -0.9568 0.041 -0.1346>}
 dyn_L_back_UpperSkirt_1 {parent=dyn_L_back_UpperSkirt_master, loc=<0 0 0>, rot=<0 0 0 1>}
 dyn_L_back_UpperSkirt_2 {parent=dyn_L_back_UpperSkirt_1, loc=<6.0967 0 0>, rot=<0 0 0 1>}
 dyn_R_front_UpperSkirt_master {parent=pelvis, loc=<3.5026 8.3331 7.0928>, rot=<0.138 0.0119 0.983 0.1206>}
 dyn_R_front_UpperSkirt_1 {parent=dyn_R_front_UpperSkirt_master, loc=<0 0 0>, rot=<0 0 0 1>}
 dyn_R_front_UpperSkirt_2 {parent=dyn_R_front_UpperSkirt_1, loc=<6.0215 0.1831 -0.1793>, rot=<0 0 0 1>}
 dyn_R_side_UpperSkirt_master {parent=pelvis, loc=<3.7214 -1.4482 13.345>, rot=<0.2345 -0.0178 0.9712 -0.0388>}
 dyn_R_side_UpperSkirt_1 {parent=dyn_R_side_UpperSkirt_master, loc=<0 0 0>, rot=<0 0 0 1>}
 dyn_R_side_UpperSkirt_2 {parent=dyn_R_side_UpperSkirt_1, loc=<5.8827 0 0>, rot=<0 0 0 1>}
 dyn_R_back_UpperSkirt_master {parent=pelvis, loc=<3.561 -9.6628 7.6166>, rot=<0.1346 -0.041 0.9568 -0.2546>}
 dyn_R_back_UpperSkirt_1 {parent=dyn_R_back_UpperSkirt_master, loc=<0 0 0>, rot=<0 0 0 1>}
 dyn_R_back_UpperSkirt_2 {parent=dyn_R_back_UpperSkirt_1, loc=<6.097 -0.0001 -0.0001>, rot=<0 0 0 1>}
 dyn_L_back_LowerSkirt_master {parent=pelvis, loc=<3.561 -9.6628 -7.6166>, rot=<-0.1485 0.0357 0.9685 -0.1967>}
 dyn_L_back_LowerSkirt_1 {parent=dyn_L_back_LowerSkirt_master, loc=<0 0 0>, rot=<0 0 0 1>}
 dyn_L_back_LowerSkirt_2 {parent=dyn_L_back_LowerSkirt_1, loc=<6.8655 -1.0287 0>, rot=<0.0206 0.053 0.0663 -0.9962>}
 dyn_L_back_LowerSkirt_3 {parent=dyn_L_back_LowerSkirt_2, loc=<6.2283 -0.5758 0>, rot=<0 0 0 -1>}
 dyn_R_back_LowerSkirt_master {parent=pelvis, loc=<3.561 -9.6628 7.6166>, rot=<0.1967 -0.9685 -0.0357 0.1485>}
 dyn_R_back_LowerSkirt_1 {parent=dyn_R_back_LowerSkirt_master, loc=<0 0 0>, rot=<0 0 0 1>}
 dyn_R_back_LowerSkirt_2 {parent=dyn_R_back_LowerSkirt_1, loc=<6.8655 1.0287 0>, rot=<-0.0206 0.053 -0.0663 -0.9962>}
 dyn_R_back_LowerSkirt_3 {parent=dyn_R_back_LowerSkirt_2, loc=<6.2283 0.5758 0>, rot=<0 0 0 1>}
 dyn_center_lowerSkirt_master {parent=pelvis, loc=<2.257 8.7278 0>, rot=<-0.1034 -0.9946 0 0>}
 dyn_center_lowerSkirt_1 {parent=dyn_center_lowerSkirt_master, loc=<0 0 0>, rot=<0 0 0 1>}
 dyn_center_upperSkirt_1 {parent=dyn_center_lowerSkirt_1, loc=<0 0 0>, rot=<0 0 0 1>}
 dyn_center_upperSkirt_2 {parent=dyn_center_upperSkirt_1, loc=<6.6378 0.5098 0>, rot=<0 0 -0.0227 -0.9997>}
 dyn_center_lowerSkirt_2 {parent=dyn_center_lowerSkirt_1, loc=<8.792 -0.0725 0>, rot=<0 0 -0.0227 -0.9997>}
 dyn_center_lowerSkirt_3 {parent=dyn_center_lowerSkirt_2, loc=<7.1868 0.1016 0>, rot=<0 0 0 1>}
 dyn_center_upperSkirt_master {parent=pelvis, loc=<2.257 8.7278 0>, rot=<-0.1034 -0.9946 0 0>}
 ik_hand_root {parent=root, loc=<0 0 0>, rot=<0 0 0 1>}
 ik_hand_gun {parent=ik_hand_root, loc=<-48.1744 3.3546 104.4381>, rot=<0.6276 0.1309 0.4183 -0.6434>}
 ik_hand_l {parent=ik_hand_gun, loc=<59.3272 67.6997 34.3524>, rot=<-0.6158 -0.7027 -0.3565 0>}
 ik_hand_r {parent=ik_hand_gun, loc=<0 0 0>, rot=<0 0 0 1>}
 attach {parent=root, loc=<0 0 0>, rot=<0 0 0 1>}
 ik_foot_root {parent=root, loc=<0 0 0>, rot=<0 0 0 1>}
 ik_foot_l {parent=ik_foot_root, loc=<14.8262 -7.3545 8.296>, rot=<0.0357 0.7045 0.0284 -0.7082>}
 ik_foot_r {parent=ik_foot_root, loc=<-14.8262 -7.3546 8.2958>, rot=<0.7082 0.0284 -0.7045 0.0357>}
]
