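# Physics asset for the F_MED_Thrive female cosmetic body, simulating its upper and lower skirt bones with capsule bodies around pelvis and thigh capsules.
# Front lower chains hang from the thighs, while back lower and upper flaps hang from pelvis-locked master bones and a branching center chain hangs from the pelvis.
# Skirt joints use tight swing and twist limits.
# Setup is asymmetric: right thigh constraint is free, left is locked, and two right-front bodies carry extra angled capsules.
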PhysicsAsset /BRCosmetics/Characters/Player/Female/Medium/Bodies/F_MED_Thrive/Meshes/F_MED_Thrive_lowerSkirtPhysics
bodies {
 dyn_L_back_LowerSkirt_1 [{shapeType=Capsule, length=3, radius=3.5, shapeName=dyn_R_back_LowerSkirt_1_capsule, center=<3 0 0>, rotation=<rot 90 0 0>}]
 dyn_L_back_LowerSkirt_2 [{shapeType=Capsule, length=3, radius=3.5, shapeName=dyn_R_back_LowerSkirt_2_capsule, center=<3 0 0>, rotation=<rot 90 0 0>}]
 dyn_L_back_LowerSkirt_3 [{shapeType=Capsule, length=3, radius=3.5, shapeName=dyn_R_back_LowerSkirt_3_capsule, center=<3 0 0>, rotation=<rot 90 0 0>}]
 dyn_L_back_LowerSkirt_master [{shapeType=Capsule, length=0.5, radius=0.5, shapeName=dyn_L_back_LowerSkirt_master_capsule, rotation=<rot 90 0 0>}]
 dyn_L_back_UpperSkirt_2 [{shapeType=Capsule, length=3, radius=3, shapeName=dyn_L_back_UpperSkirt_2_capsule, center=<3 0 0>, rotation=<rot 90 0 0>}]
 dyn_L_back_UpperSkirt_master [{shapeType=Capsule, length=0.5, radius=0.5, shapeName=dyn_L_back_UpperSkirt_master_capsule, rotation=<rot 90 0 0>}]
 dyn_L_front_LowerSkirt_1 [{shapeType=Capsule, length=3, radius=2.5, shapeName=dyn_R_front_LowerSkirt_3_capsule, center=<3 0 0>, rotation=<rot 90 0 0>}]
 dyn_L_front_LowerSkirt_2 [{shapeType=Capsule, length=6, radius=3, shapeName=dyn_L_front_LowerSkirt_2_capsule, center=<3 0 0>, rotation=<rot 90 0 0>}]
 dyn_L_front_LowerSkirt_3 [{shapeType=Capsule, length=6, radius=3, shapeName=dyn_L_front_LowerSkirt_3_capsule, center=<3 0 0>, rotation=<rot 90 0 0>}]
 dyn_L_front_UpperSkirt_2 [{shapeType=Capsule, length=3, radius=2.5, shapeName=dyn_L_front_UpperSkirt_2_capsule, center=<4 2 0>, rotation=<rot 90 0 0>}]
 dyn_L_front_UpperSkirt_master [{shapeType=Capsule, length=0.5, radius=0.5, shapeName=dyn_L_front_UpperSkirt_master_capsule, rotation=<rot 90 0 0>}]
 dyn_R_back_LowerSkirt_1 [{shapeType=Capsule, length=3, radius=3.5, shapeName=dyn_R_back_LowerSkirt_1_capsule, center=<3 0 0>, rotation=<rot 90 0 0>}]
 dyn_R_back_LowerSkirt_2 [{shapeType=Capsule, length=3, radius=3.5, shapeName=dyn_R_back_LowerSkirt_2_capsule, center=<3 0 0>, rotation=<rot 90 0 0>}]
 dyn_R_back_LowerSkirt_3 [{shapeType=Capsule, length=3, radius=3.5, shapeName=dyn_R_back_LowerSkirt_3_capsule, center=<3 0 0>, rotation=<rot 90 0 0>}]
 dyn_R_back_LowerSkirt_master [{shapeType=Capsule, length=0.5, radius=0.5, shapeName=dyn_R_back_LowerSkirt_master_capsule}]
 dyn_R_back_UpperSkirt_2 [{shapeType=Capsule, length=3, radius=3, shapeName=dyn_R_back_UpperSkirt_2_capsule, center=<3 0 0>, rotation=<rot 90 0 0>}]
 dyn_R_back_UpperSkirt_master [{shapeType=Capsule, length=0.5, radius=0.5, shapeName=dyn_R_back_UpperSkirt_master_capsule, rotation=<rot 90 0 0>}]
 dyn_R_front_LowerSkirt_1 [{shapeType=Capsule, length=3, radius=2.5, shapeName=dyn_R_front_LowerSkirt_3_capsule, center=<3 0.9772 0>, rotation=<rot 90 0 0>}]
 dyn_R_front_LowerSkirt_2 [
  {shapeType=Capsule, length=6, radius=3, shapeName=dyn_R_front_LowerSkirt_2_capsule, center=<3 0 0>, rotation=<rot 90 0 0>}
  {shapeType=Capsule, length=9, radius=3, shapeName=dyn_R_front_LowerSkirt_2_capsule, center=<3.2272 0 -6.1519>, rotation=<rot 74.9999 0 0>}
 ]
 dyn_R_front_LowerSkirt_3 [
  {shapeType=Capsule, length=6, radius=3, shapeName=dyn_R_front_LowerSkirt_3_capsule, center=<3 0 0>, rotation=<rot 90 0 0>}
  {shapeType=Capsule, length=9, radius=3, shapeName=dyn_R_front_LowerSkirt_3_capsule, center=<2.6673 0 -8.2417>, rotation=<rot 74.9999 0 0>}
 ]
 dyn_R_front_UpperSkirt_2 [{
   shapeType=Capsule
   length=3
   radius=2.5
   shapeName=dyn_R_front_UpperSkirt_2_capsule
   center=<3.3654 -3.7435 0>
   rotation=<rot 90 165.9638 -174.0361>
  }]
 dyn_R_front_UpperSkirt_master [{shapeType=Capsule, length=0.5, radius=0.5, shapeName=dyn_R_front_UpperSkirt_master_capsule, rotation=<rot 90 0 0>}]
 dyn_center_lowerSkirt_1 [{shapeType=Capsule, length=6, radius=4, shapeName=dyn_center_lowerSkirt_1_capsule, center=<6 0 0>}]
 dyn_center_lowerSkirt_2 [{shapeType=Capsule, length=6, radius=4, shapeName=dyn_center_lowerSkirt_2_capsule, center=<6 0 0>}]
 dyn_center_upperSkirt_2 [{shapeType=Capsule, length=6, radius=4, shapeName=dyn_center_upperSkirt_2_capsule, center=<3 0 0>}]
 pelvis [{shapeType=Capsule, length=5, radius=5, shapeName=pelvis_capsule}]
 thigh_l [{shapeType=Capsule, length=28, radius=8.3, shapeName=thigh_r_capsule, center=<-10 0.0697 1.3694>, rotation=<rot 90 0 0>}]
 thigh_r [{shapeType=Capsule, length=28, radius=8.3, shapeName=thigh_r_capsule, center=<10 0.9521 -1.3695>, rotation=<rot 90 0 0>}]
}
constraints [
 {bone1Name=dyn_R_front_LowerSkirt_2, bone2Name=dyn_R_front_LowerSkirt_1, swing1LimitDegrees=20, swing2LimitDegrees=10, twistLimitDegrees=5}
 {bone1Name=dyn_R_front_LowerSkirt_3, bone2Name=dyn_R_front_LowerSkirt_2, swing1LimitDegrees=20, swing2LimitDegrees=10, twistLimitDegrees=5}
 {bone1Name=dyn_R_back_LowerSkirt_2, bone2Name=dyn_R_back_LowerSkirt_1, swing1LimitDegrees=10, swing2LimitDegrees=10, twistLimitDegrees=5}
 {bone1Name=dyn_R_back_LowerSkirt_3, bone2Name=dyn_R_back_LowerSkirt_2, swing1LimitDegrees=10, swing2LimitDegrees=10, twistLimitDegrees=5}
 {
  bone1Name=thigh_r
  bone2Name=pelvis
  swing1LimitDegrees=45
  swing1Motion=Free
  swing2LimitDegrees=45
  swing2Motion=Free
  twistLimitDegrees=45
  twistMotion=Free
 }
 {bone1Name=dyn_R_front_LowerSkirt_1, bone2Name=thigh_r, swing1LimitDegrees=20, swing2LimitDegrees=0, twistLimitDegrees=5}
 {bone1Name=dyn_R_back_LowerSkirt_1, bone2Name=dyn_R_back_LowerSkirt_master, swing1LimitDegrees=30, swing2LimitDegrees=10, twistLimitDegrees=5}
 {
  bone1Name=dyn_R_back_LowerSkirt_master
  bone2Name=pelvis
  swing1LimitDegrees=45
  swing1Motion=Locked
  swing2LimitDegrees=45
  swing2Motion=Locked
  twistLimitDegrees=5
  twistMotion=Locked
 }
 {bone1Name=dyn_L_front_LowerSkirt_1, bone2Name=thigh_l, swing1LimitDegrees=20, swing2LimitDegrees=10, twistLimitDegrees=5}
 {bone1Name=dyn_L_front_LowerSkirt_2, bone2Name=dyn_L_front_LowerSkirt_1, swing1LimitDegrees=20, swing2LimitDegrees=10, twistLimitDegrees=5}
 {bone1Name=dyn_L_front_LowerSkirt_3, bone2Name=dyn_L_front_LowerSkirt_2, swing1LimitDegrees=20, swing2LimitDegrees=10, twistLimitDegrees=5}
 {
  bone1Name=dyn_L_back_LowerSkirt_master
  bone2Name=pelvis
  swing1LimitDegrees=45
  swing1Motion=Locked
  swing2LimitDegrees=45
  swing2Motion=Locked
  twistLimitDegrees=45
  twistMotion=Locked
 }
 {bone1Name=dyn_L_back_LowerSkirt_1, bone2Name=dyn_L_back_LowerSkirt_master, swing1LimitDegrees=30, swing2LimitDegrees=10, twistLimitDegrees=5}
 {bone1Name=dyn_L_back_LowerSkirt_2, bone2Name=dyn_L_back_LowerSkirt_1, swing1LimitDegrees=10, swing2LimitDegrees=10, twistLimitDegrees=5}
 {bone1Name=dyn_L_back_LowerSkirt_3, bone2Name=dyn_L_back_LowerSkirt_2, swing1LimitDegrees=10, swing2LimitDegrees=10, twistLimitDegrees=5}
 {
  bone1Name=thigh_l
  bone2Name=pelvis
  swing1LimitDegrees=45
  swing1Motion=Locked
  swing2LimitDegrees=45
  swing2Motion=Locked
  twistLimitDegrees=45
  twistMotion=Locked
 }
 {
  bone1Name=dyn_L_back_UpperSkirt_master
  bone2Name=pelvis
  swing1LimitDegrees=45
  swing1Motion=Locked
  swing2LimitDegrees=45
  swing2Motion=Locked
  twistLimitDegrees=45
  twistMotion=Locked
 }
 {bone1Name=dyn_L_back_UpperSkirt_2, bone2Name=dyn_L_back_UpperSkirt_master, swing1LimitDegrees=30, swing2LimitDegrees=10, twistLimitDegrees=5}
 {
  bone1Name=dyn_R_front_UpperSkirt_master
  bone2Name=pelvis
  swing1LimitDegrees=45
  swing1Motion=Locked
  swing2LimitDegrees=45
  swing2Motion=Locked
  twistLimitDegrees=45
  twistMotion=Locked
 }
 {
  bone1Name=dyn_R_front_UpperSkirt_2
  bone2Name=dyn_R_front_UpperSkirt_master
  swing1LimitDegrees=30
  swing2LimitDegrees=10
  twistLimitDegrees=5
 }
 {
  bone1Name=dyn_L_front_UpperSkirt_master
  bone2Name=pelvis
  swing1LimitDegrees=45
  swing1Motion=Locked
  swing2LimitDegrees=45
  swing2Motion=Locked
  twistLimitDegrees=45
  twistMotion=Locked
 }
 {
  bone1Name=dyn_L_front_UpperSkirt_2
  bone2Name=dyn_L_front_UpperSkirt_master
  swing1LimitDegrees=20
  swing2LimitDegrees=10
  twistLimitDegrees=5
 }
 {
  bone1Name=dyn_R_back_UpperSkirt_master
  bone2Name=pelvis
  swing1LimitDegrees=45
  swing1Motion=Locked
  swing2LimitDegrees=45
  swing2Motion=Locked
  twistLimitDegrees=45
  twistMotion=Locked
 }
 {bone1Name=dyn_R_back_UpperSkirt_2, bone2Name=dyn_R_back_UpperSkirt_master, swing1LimitDegrees=30, swing2LimitDegrees=10, twistLimitDegrees=5}
 {bone1Name=dyn_center_upperSkirt_2, bone2Name=dyn_center_lowerSkirt_1, swing1LimitDegrees=20, swing2LimitDegrees=10, twistLimitDegrees=0}
 {bone1Name=dyn_center_lowerSkirt_1, bone2Name=pelvis, swing1LimitDegrees=30, swing2LimitDegrees=10, twistLimitDegrees=0}
 {bone1Name=dyn_center_lowerSkirt_2, bone2Name=dyn_center_lowerSkirt_1, swing1LimitDegrees=30, swing2LimitDegrees=10, twistLimitDegrees=0}
]
