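# Post-process style AnimBlueprint for the F_MED_Thrive cosmetic body, adding secondary motion to skirt, stars, shoe tongues and bow ribbons.
# The parent mesh pose runs through RigidBody skirt physics, a Control Rig, weighted skirt-star constraints and crouch-specific bone corrections.
# The cached preRibbon pose feeds two AnimDynamics bow-and-ribbon chains, selected by whether the backbling is a cape.
# The non-cape chain adds sine-driven strand rotations; the cape chain uses fixed offsets and tighter limits.
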AnimBlueprint /BRCosmetics/Characters/Player/Female/Medium/Bodies/F_MED_Thrive/Meshes/F_MED_Thrive_AnimBP
target_skeleton=&/F_MED_Thrive_Skeleton
graphs [
 AnimGraph {nodes [
   "Constraint - dyn_skirtStar_A (PPPPPPP)"!(
    ComponentPose
    Alpha=1
    ConstraintWeights_0=0.7255
    ConstraintWeights_1:float=0.0235
    ConstraintWeights_2:float=0.0351
    ConstraintWeights_3:float=0.1098
    ConstraintWeights_4:float=0.0183
    ConstraintWeights_5:float=0.0238
    ConstraintWeights_6:float=0.064
   ) -> (Pose) {
    alphaBoolBlend {blendOption=Linear}
    boneToModify {boneName=dyn_skirtStar_A}
    constraintSetup [
     {targetBone {boneName=pelvis}}
     {targetBone {boneName=dyn_R_front_LowerSkirt_1}}
     {targetBone {boneName=dyn_R_front_LowerSkirt_2}}
     {targetBone {boneName=dyn_R_front_LowerSkirt_3}}
     {targetBone {boneName=dyn_L_front_LowerSkirt_1}}
     {targetBone {boneName=dyn_L_front_LowerSkirt_2}}
     {targetBone {boneName=dyn_L_front_LowerSkirt_3}}
    ]
    constraintWeights [1,1,1,1,1,1,1]
   }
   n1 = Get sineRotation3() -> (sineRotation3:Rotator)
   n2 = "Transform (Modify) Bone - Bone: dyn_L_bowStrand_1"(ComponentPose=Local To Component(LocalPose=Use cached pose 'preRibbon'()), Alpha=1, Translation="0,0,0", Rotation=@n1, Scale="1,1,1") {rotationMode=BMM_Additive, alphaBoolBlend {blendOption=Linear}, alphaScaleBiasClamp {scale=0.7}, boneToModify {boneName=dyn_L_bowStrand_1}}
   n3 = Get sineRotation4() -> (sineRotation4:Rotator)
   n4 = "Transform (Modify) Bone - Bone: dyn_L_bowStrand_2"(ComponentPose=@n2, Alpha=1, Translation="0,0,0", Rotation=@n3, Scale="1,1,1") {rotationMode=BMM_Additive, alphaBoolBlend {blendOption=Linear}, alphaScaleBiasClamp {scale=0.7}, boneToModify {boneName=dyn_L_bowStrand_2}}
   n5 = Get sineRotation5() -> (sineRotation5:Rotator)
   n6 = "Transform (Modify) Bone - Bone: dyn_L_bowStrand_3"(ComponentPose=@n4, Alpha=1, Translation="0,0,0", Rotation=@n5, Scale="1,1,1") {rotationMode=BMM_Additive, alphaBoolBlend {blendOption=Linear}, alphaScaleBiasClamp {scale=0.7}, boneToModify {boneName=dyn_L_bowStrand_3}}
   n7 = Get sineRotation6() -> (sineRotation6:Rotator)
   n8 = "Transform (Modify) Bone - Bone: dyn_L_bowStrand_4"(ComponentPose=@n6, Alpha=1, Translation="0,0,0", Rotation=@n7, Scale="1,1,1") {rotationMode=BMM_Additive, alphaBoolBlend {blendOption=Linear}, alphaScaleBiasClamp {scale=0.7}, boneToModify {boneName=dyn_L_bowStrand_4}}
   n9 = Get sineRotation7() -> (sineRotation7:Rotator)
   n10 = "Transform (Modify) Bone - Bone: dyn_L_bowStrand_5"(ComponentPose=@n8, Alpha=1, Translation="0,0,0", Rotation=@n9, Scale="1,1,1") {rotationMode=BMM_Additive, alphaBoolBlend {blendOption=Linear}, alphaScaleBiasClamp {scale=0.7}, boneToModify {boneName=dyn_L_bowStrand_5}}
   n11 = "Transform (Modify) Bone - Bone: dyn_R_bowStrand_1"(ComponentPose=@n10, Alpha=1, Translation="0,0,0", Rotation=@n1, Scale="1,1,1") {rotationMode=BMM_Additive, alphaBoolBlend {blendOption=Linear}, alphaScaleBiasClamp {scale=0.7}, boneToModify {boneName=dyn_R_bowStrand_1}}
   n12 = "Transform (Modify) Bone - Bone: dyn_R_bowStrand_2"(ComponentPose=@n11, Alpha=1, Translation="0,0,0", Rotation=@n3, Scale="1,1,1") {rotationMode=BMM_Additive, alphaBoolBlend {blendOption=Linear}, alphaScaleBiasClamp {scale=0.7}, boneToModify {boneName=dyn_R_bowStrand_2}}
   n13 = "Transform (Modify) Bone - Bone: dyn_R_bowStrand_3"(ComponentPose=@n12, Alpha=1, Translation="0,0,0", Rotation=@n5, Scale="1,1,1") {rotationMode=BMM_Additive, alphaBoolBlend {blendOption=Linear}, alphaScaleBiasClamp {scale=0.7}, boneToModify {boneName=dyn_R_bowStrand_3}}
   n14 = "Transform (Modify) Bone - Bone: dyn_R_bowStrand_4"(ComponentPose=@n13, Alpha=1, Translation="0,0,0", Rotation=@n7, Scale="1,1,1") {rotationMode=BMM_Additive, alphaBoolBlend {blendOption=Linear}, alphaScaleBiasClamp {scale=0.7}, boneToModify {boneName=dyn_R_bowStrand_4}}
   n15 = "Transform (Modify) Bone - Bone: dyn_R_bowStrand_5"(ComponentPose=@n14, Alpha=1, Translation="0,0,0", Rotation=@n9, Scale="1,1,1") {rotationMode=BMM_Additive, alphaBoolBlend {blendOption=Linear}, alphaScaleBiasClamp {scale=0.7}, boneToModify {boneName=dyn_R_bowStrand_5}}
   n16 = "Transform (Modify) Bone - Bone: dyn_bow_master_jnt"(
    ComponentPose=@n15
    Alpha=Get FrontEnd_Idle_On_float() -> (FrontEnd_Idle_On_float:double)
    Translation="0,0,0"
    Rotation="16,0,-10"
    Scale="1,1,1"
   ) {
    rotationMode=BMM_Additive
    rotationSpace=BCS_BoneSpace
    alphaBoolBlend {blendOption=Linear}
    boneToModify {boneName=dyn_bow_master_jnt}
    rotation {pitch=16, roll=-10}
   }
   n17 = Anim Dynamics - Bone: dyn_bow_C(ComponentPose=@n16, Alpha=1, ExternalForce="0,0,0") {
    angularDampingOverride=1
    angularSpringConstant=20000
    bAngularSpring=true
    bOverrideAngularDamping=true
    bUseGravityOverride=true
    gravityScale=0.35
    simulationSpace=RootRelative
    alphaBoolBlend {blendOption=Linear}
    boundBone {boneName=dyn_bow_C}
    componentLinearAccScale=<1 1 1>
    componentLinearVelScale=<1 1 1>
    gravityOverride {z=-980}
    physicsBodyDefinitions [{
      sphereCollisionRadius=0
      boundBone {boneName=dyn_bow_C}
      boxExtents=<30 40 40>
      constraintSetup {angularLimitsMax=<5 7 0>, angularLimitsMin=<-5 -7 0>, angularTarget=<1 0 0>}
      localJointOffset=<4 0 0>
     }]
   }
   n18 = Anim Dynamics - Bone: dyn_bow_L(ComponentPose=@n17, Alpha=1, ExternalForce="0,0,0") {
    angularDampingOverride=1
    angularSpringConstant=20000
    bAngularSpring=true
    bOverrideAngularDamping=true
    bUseGravityOverride=true
    gravityScale=0.35
    simulationSpace=RootRelative
    alphaBoolBlend {blendOption=Linear}
    boundBone {boneName=dyn_bow_L}
    componentLinearAccScale=<1 1 1>
    componentLinearVelScale=<1 1 1>
    gravityOverride {z=-980}
    physicsBodyDefinitions [{
      sphereCollisionRadius=0
      boundBone {boneName=dyn_bow_L}
      boxExtents=<70 80 80>
      constraintSetup {angularLimitsMax=<10 9 10>, angularLimitsMin=<-10 -9 -10>, angularTarget=<1 0 0>}
      localJointOffset=<-4 0 0>
     }]
   }
   n19 = Anim Dynamics - Bone: dyn_bow_R(ComponentPose=@n18, Alpha=1, ExternalForce="0,0,0") {
    angularDampingOverride=1
    angularSpringConstant=20000
    bAngularSpring=true
    bOverrideAngularDamping=true
    bUseGravityOverride=true
    gravityScale=0.35
    simulationSpace=RootRelative
    alphaBoolBlend {blendOption=Linear}
    boundBone {boneName=dyn_bow_R}
    componentLinearAccScale=<1 1 1>
    componentLinearVelScale=<1 1 1>
    gravityOverride {z=-980}
    physicsBodyDefinitions [{
      sphereCollisionRadius=0
      boundBone {boneName=dyn_bow_R}
      boxExtents=<70 80 80>
      constraintSetup {angularLimitsMax=<10 9 10>, angularLimitsMin=<-10 -9 -10>, angularTarget=<1 0 0>}
      localJointOffset=<4 0 0>
     }]
   }
   n20 = Get toneDownRibbonAlpha() -> (toneDownRibbonAlpha:double)
   n21 = Anim Dynamics - Bone: dyn_R_bowStrand_1!(ComponentPose=@n19, Alpha=@n20, GravityOverride:Vector="0,0,-1000", ExternalForce="0,0,0") -> (Pose) {
    angularDampingOverride=0.9999
    angularSpringConstant=5
    bAngularSpring=true
    bOverrideAngularDamping=true
    bUseGravityOverride=true
    gravityScale=0.35
    alphaBoolBlend {blendOption=Linear}
    alphaScaleBiasClamp {bInterpResult=true, interpSpeedDecreasing=5, interpSpeedIncreasing=5}
    boundBone {boneName=dyn_R_bowStrand_1}
    componentLinearAccScale=<1 3 1>
    componentLinearVelScale=<1 3 1>
    physicsBodyDefinitions [{
      sphereCollisionRadius=0
      boundBone {boneName=dyn_R_bowStrand_1}
      boxExtents=<20 20 20>
      constraintSetup {angularLimitsMax=<0 10 10>, angularLimitsMin=<0 -10 -10>, angularTarget=<1 0 0>}
      localJointOffset=<-4 0 0>
     }]
   }
   n22 = Anim Dynamics - Bone: dyn_R_bowStrand_2!(ComponentPose=@n21, Alpha=@n20, GravityOverride:Vector="0,0,-1000", ExternalForce="0,0,0") -> (Pose) {
    angularDampingOverride=0.9999
    angularSpringConstant=5
    bAngularSpring=true
    bOverrideAngularDamping=true
    bUseGravityOverride=true
    gravityScale=0.35
    alphaBoolBlend {blendOption=Linear}
    alphaScaleBiasClamp {bInterpResult=true, interpSpeedDecreasing=5, interpSpeedIncreasing=5}
    boundBone {boneName=dyn_R_bowStrand_2}
    componentLinearAccScale=<1 3 1>
    componentLinearVelScale=<1 3 1>
    physicsBodyDefinitions [{
      sphereCollisionRadius=0
      boundBone {boneName=dyn_R_bowStrand_2}
      boxExtents=<20 20 20>
      constraintSetup {angularLimitsMax=<0 20 20>, angularLimitsMin=<0 -20 -20>, angularTarget=<1 0 0>}
      localJointOffset=<-4 0 0>
     }]
   }
   n23 = Anim Dynamics - Bone: dyn_R_bowStrand_3!(ComponentPose=@n22, Alpha=@n20, GravityOverride:Vector="0,0,-1000", ExternalForce="0,0,0") -> (Pose) {
    angularDampingOverride=0.9999
    angularSpringConstant=5
    bAngularSpring=true
    bOverrideAngularDamping=true
    bUseGravityOverride=true
    gravityScale=0.35
    alphaBoolBlend {blendOption=Linear}
    alphaScaleBiasClamp {bInterpResult=true, interpSpeedDecreasing=5, interpSpeedIncreasing=5}
    boundBone {boneName=dyn_R_bowStrand_3}
    componentLinearAccScale=<1 3 1>
    componentLinearVelScale=<1 3 1>
    physicsBodyDefinitions [{
      sphereCollisionRadius=0
      boundBone {boneName=dyn_R_bowStrand_3}
      boxExtents=<15 15 15>
      constraintSetup {angularLimitsMax=<10 20 20>, angularLimitsMin=<-10 -20 -20>, angularTarget=<1 0 0>}
      localJointOffset=<-4 0 0>
     }]
   }
   n24 = Anim Dynamics - Bone: dyn_R_bowStrand_4!(ComponentPose=@n23, Alpha=@n20, GravityOverride:Vector="0,0,-1000", ExternalForce="0,0,0") -> (Pose) {
    angularDampingOverride=0.9999
    angularSpringConstant=5
    bAngularSpring=true
    bOverrideAngularDamping=true
    bUseGravityOverride=true
    gravityScale=0.35
    alphaBoolBlend {blendOption=Linear}
    alphaScaleBiasClamp {bInterpResult=true, interpSpeedDecreasing=5, interpSpeedIncreasing=5}
    boundBone {boneName=dyn_R_bowStrand_4}
    componentLinearAccScale=<1 3 1>
    componentLinearVelScale=<1 3 1>
    physicsBodyDefinitions [{
      sphereCollisionRadius=0
      boundBone {boneName=dyn_R_bowStrand_4}
      boxExtents=<15 15 15>
      constraintSetup {angularLimitsMax=<10 20 20>, angularLimitsMin=<-10 -20 -20>, angularTarget=<1 0 0>}
      localJointOffset=<-3 0 0>
     }]
   }
   n25 = Anim Dynamics - Bone: dyn_R_bowStrand_5!(ComponentPose=@n24, Alpha=@n20, GravityOverride:Vector="0,0,-1000", ExternalForce="0,0,0") -> (Pose) {
    angularDampingOverride=0.9999
    angularSpringConstant=5
    bAngularSpring=true
    bOverrideAngularDamping=true
    bUseGravityOverride=true
    gravityScale=0.35
    alphaBoolBlend {blendOption=Linear}
    alphaScaleBiasClamp {bInterpResult=true, interpSpeedDecreasing=5, interpSpeedIncreasing=5}
    boundBone {boneName=dyn_R_bowStrand_5}
    componentLinearAccScale=<1 3 1>
    componentLinearVelScale=<1 3 1>
    physicsBodyDefinitions [{
      sphereCollisionRadius=0
      boundBone {boneName=dyn_R_bowStrand_5}
      boxExtents=<15 15 15>
      constraintSetup {angularLimitsMax=<0 20 20>, angularLimitsMin=<0 -20 -20>, angularTarget=<1 0 0>}
      localJointOffset=<-3 0 0>
     }]
   }
   n26 = Anim Dynamics - Bone: dyn_L_bowStrand_1!(ComponentPose=@n25, Alpha=@n20, GravityOverride:Vector="0,0,-1000", ExternalForce="0,0,0") -> (Pose) {
    angularDampingOverride=0.9999
    angularSpringConstant=5
    bAngularSpring=true
    bOverrideAngularDamping=true
    bUseGravityOverride=true
    gravityScale=0.35
    alphaBoolBlend {blendOption=Linear}
    alphaScaleBiasClamp {bInterpResult=true, interpSpeedDecreasing=5, interpSpeedIncreasing=5}
    boundBone {boneName=dyn_L_bowStrand_1}
    componentLinearAccScale=<1 3 1>
    componentLinearVelScale=<1 3 1>
    physicsBodyDefinitions [{
      sphereCollisionRadius=0
      boundBone {boneName=dyn_L_bowStrand_1}
      boxExtents=<20 20 20>
      constraintSetup {angularLimitsMax=<0 10 10>, angularLimitsMin=<0 -10 -10>, angularTarget=<1 0 0>}
      localJointOffset=<-4 0 0>
     }]
   }
   n27 = Anim Dynamics - Bone: dyn_L_bowStrand_2!(ComponentPose=@n26, Alpha=@n20, GravityOverride:Vector="0,0,-1000", ExternalForce="0,0,0") -> (Pose) {
    angularDampingOverride=0.9999
    angularSpringConstant=5
    bAngularSpring=true
    bOverrideAngularDamping=true
    bUseGravityOverride=true
    gravityScale=0.35
    alphaBoolBlend {blendOption=Linear}
    alphaScaleBiasClamp {bInterpResult=true, interpSpeedDecreasing=5, interpSpeedIncreasing=5}
    boundBone {boneName=dyn_L_bowStrand_2}
    componentLinearAccScale=<1 3 1>
    componentLinearVelScale=<1 3 1>
    physicsBodyDefinitions [{
      sphereCollisionRadius=0
      boundBone {boneName=dyn_L_bowStrand_2}
      boxExtents=<20 20 20>
      constraintSetup {angularLimitsMax=<0 20 20>, angularLimitsMin=<0 -20 -20>, angularTarget=<1 0 0>}
      localJointOffset=<-4 0 0>
     }]
   }
   n28 = Anim Dynamics - Bone: dyn_L_bowStrand_3!(ComponentPose=@n27, Alpha=@n20, GravityOverride:Vector="0,0,-1000", ExternalForce="0,0,0") -> (Pose) {
    angularDampingOverride=0.9999
    angularSpringConstant=5
    bAngularSpring=true
    bOverrideAngularDamping=true
    bUseGravityOverride=true
    gravityScale=0.35
    alphaBoolBlend {blendOption=Linear}
    alphaScaleBiasClamp {bInterpResult=true, interpSpeedDecreasing=5, interpSpeedIncreasing=5}
    boundBone {boneName=dyn_L_bowStrand_3}
    componentLinearAccScale=<1 3 1>
    componentLinearVelScale=<1 3 1>
    physicsBodyDefinitions [{
      sphereCollisionRadius=0
      boundBone {boneName=dyn_L_bowStrand_3}
      boxExtents=<15 15 15>
      constraintSetup {angularLimitsMax=<10 20 20>, angularLimitsMin=<-10 -20 -20>, angularTarget=<1 0 0>}
      localJointOffset=<-4 0 0>
     }]
   }
   n29 = Anim Dynamics - Bone: dyn_L_bowStrand_4!(ComponentPose=@n28, Alpha=@n20, GravityOverride:Vector="0,0,-1000", ExternalForce="0,0,0") -> (Pose) {
    angularDampingOverride=0.9999
    angularSpringConstant=5
    bAngularSpring=true
    bOverrideAngularDamping=true
    bUseGravityOverride=true
    gravityScale=0.35
    alphaBoolBlend {blendOption=Linear}
    alphaScaleBiasClamp {bInterpResult=true, interpSpeedDecreasing=5, interpSpeedIncreasing=5}
    boundBone {boneName=dyn_L_bowStrand_4}
    componentLinearAccScale=<1 3 1>
    componentLinearVelScale=<1 3 1>
    physicsBodyDefinitions [{
      sphereCollisionRadius=0
      boundBone {boneName=dyn_L_bowStrand_4}
      boxExtents=<15 15 15>
      constraintSetup {angularLimitsMax=<10 20 20>, angularLimitsMin=<-10 -20 -20>, angularTarget=<1 0 0>}
      localJointOffset=<-3 0 0>
     }]
   }
   n30 = Anim Dynamics - Bone: dyn_L_bowStrand_5!(ComponentPose=@n29, Alpha=@n20, GravityOverride:Vector="0,0,-1000", ExternalForce="0,0,0") -> (Pose) {
    angularDampingOverride=0.9999
    angularSpringConstant=5
    bAngularSpring=true
    bOverrideAngularDamping=true
    bUseGravityOverride=true
    gravityScale=0.35
    alphaBoolBlend {blendOption=Linear}
    alphaScaleBiasClamp {bInterpResult=true, interpSpeedDecreasing=5, interpSpeedIncreasing=5}
    boundBone {boneName=dyn_L_bowStrand_5}
    componentLinearAccScale=<1 3 1>
    componentLinearVelScale=<1 3 1>
    physicsBodyDefinitions [{
      sphereCollisionRadius=0
      boundBone {boneName=dyn_L_bowStrand_5}
      boxExtents=<15 15 15>
      constraintSetup {angularLimitsMax=<0 20 20>, angularLimitsMin=<0 -20 -20>, angularTarget=<1 0 0>}
      localJointOffset=<-3 0 0>
     }]
   }
   n31 = "Transform (Modify) Bone - Bone: dyn_L_bowStrand_1"(ComponentPose=Local To Component(LocalPose=Use cached pose 'preRibbon'()), Alpha=1, Translation="0,0,0", Rotation="10,0,-30", Scale="1,1,1") {
    rotationMode=BMM_Additive
    alphaBoolBlend {blendOption=Linear}
    alphaScaleBiasClamp {scale=0.7}
    boneToModify {boneName=dyn_L_bowStrand_1}
    rotation {pitch=10, roll=-30}
   }
   n32 = "Transform (Modify) Bone - Bone: dyn_L_bowStrand_4"(ComponentPose=@n31, Alpha=1, Translation="0,0,0", Rotation="20,0,40", Scale="1,1,1") {
    rotationMode=BMM_Additive
    alphaBoolBlend {blendOption=Linear}
    alphaScaleBiasClamp {scale=0.7}
    boneToModify {boneName=dyn_L_bowStrand_4}
    rotation {pitch=20, roll=40}
   }
   n33 = "Transform (Modify) Bone - Bone: dyn_L_bowStrand_5"(ComponentPose=@n32, Alpha=1, Translation="0,0,0", Rotation="0,-20,60", Scale="1,1,1") {
    rotationMode=BMM_Additive
    alphaBoolBlend {blendOption=Linear}
    alphaScaleBiasClamp {scale=0.7}
    boneToModify {boneName=dyn_L_bowStrand_5}
    rotation {roll=60, yaw=-20}
   }
   n34 = "Transform (Modify) Bone - Bone: dyn_R_bowStrand_1"(ComponentPose=@n33, Alpha=1, Translation="0,0,0", Rotation="-10,0,-30", Scale="1,1,1") {
    rotationMode=BMM_Additive
    alphaBoolBlend {blendOption=Linear}
    alphaScaleBiasClamp {scale=0.7}
    boneToModify {boneName=dyn_R_bowStrand_1}
    rotation {pitch=-10, roll=-30}
   }
   n35 = "Transform (Modify) Bone - Bone: dyn_R_bowStrand_4"(ComponentPose=@n34, Alpha=1, Translation="0,0,0", Rotation="-20,0,40", Scale="1,1,1") {
    rotationMode=BMM_Additive
    alphaBoolBlend {blendOption=Linear}
    alphaScaleBiasClamp {scale=0.7}
    boneToModify {boneName=dyn_R_bowStrand_4}
    rotation {pitch=-20, roll=40}
   }
   n36 = "Transform (Modify) Bone - Bone: dyn_R_bowStrand_5"(ComponentPose=@n35, Alpha=1, Translation="0,0,0", Rotation="0,20,60", Scale="1,1,1") {
    rotationMode=BMM_Additive
    alphaBoolBlend {blendOption=Linear}
    alphaScaleBiasClamp {scale=0.7}
    boneToModify {boneName=dyn_R_bowStrand_5}
    rotation {roll=60, yaw=20}
   }
   n37 = Anim Dynamics - Bone: dyn_bow_C(ComponentPose=@n36, Alpha=1, ExternalForce="0,0,0") {
    angularDampingOverride=1
    angularSpringConstant=20000
    bAngularSpring=true
    bOverrideAngularDamping=true
    bUseGravityOverride=true
    gravityScale=0.35
    simulationSpace=RootRelative
    alphaBoolBlend {blendOption=Linear}
    boundBone {boneName=dyn_bow_C}
    componentLinearAccScale=<1 1 1>
    componentLinearVelScale=<1 1 1>
    gravityOverride {z=-980}
    physicsBodyDefinitions [{
      sphereCollisionRadius=0
      boundBone {boneName=dyn_bow_C}
      boxExtents=<30 40 40>
      constraintSetup {angularLimitsMax=<5 7 0>, angularLimitsMin=<-5 -7 0>, angularTarget=<1 0 0>}
      localJointOffset=<4 0 0>
     }]
   }
   n38 = Anim Dynamics - Bone: dyn_bow_L(ComponentPose=@n37, Alpha=1, ExternalForce="0,0,0") {
    angularDampingOverride=1
    angularSpringConstant=20000
    bAngularSpring=true
    bOverrideAngularDamping=true
    bUseGravityOverride=true
    gravityScale=0.35
    simulationSpace=RootRelative
    alphaBoolBlend {blendOption=Linear}
    boundBone {boneName=dyn_bow_L}
    componentLinearAccScale=<1 1 1>
    componentLinearVelScale=<1 1 1>
    gravityOverride {z=-980}
    physicsBodyDefinitions [{
      sphereCollisionRadius=0
      boundBone {boneName=dyn_bow_L}
      boxExtents=<70 80 80>
      constraintSetup {angularLimitsMax=<10 9 10>, angularLimitsMin=<-10 -9 -10>, angularTarget=<1 0 0>}
      localJointOffset=<-4 0 0>
     }]
   }
   n39 = Anim Dynamics - Bone: dyn_bow_R(ComponentPose=@n38, Alpha=1, ExternalForce="0,0,0") {
    angularDampingOverride=1
    angularSpringConstant=20000
    bAngularSpring=true
    bOverrideAngularDamping=true
    bUseGravityOverride=true
    gravityScale=0.35
    simulationSpace=RootRelative
    alphaBoolBlend {blendOption=Linear}
    boundBone {boneName=dyn_bow_R}
    componentLinearAccScale=<1 1 1>
    componentLinearVelScale=<1 1 1>
    gravityOverride {z=-980}
    physicsBodyDefinitions [{
      sphereCollisionRadius=0
      boundBone {boneName=dyn_bow_R}
      boxExtents=<70 80 80>
      constraintSetup {angularLimitsMax=<10 9 10>, angularLimitsMin=<-10 -9 -10>, angularTarget=<1 0 0>}
      localJointOffset=<4 0 0>
     }]
   }
   n40 = Get toneDownRibbonAlpha_capeOn() -> (toneDownRibbonAlpha_capeOn:double)
   n41 = Anim Dynamics - Bone: dyn_R_bowStrand_1(ComponentPose=@n39, Alpha=@n40, GravityOverride="0,0,-1000", ExternalForce="0,0,0") {
    angularDampingOverride=0.9999
    angularSpringConstant=5
    bAngularSpring=true
    bOverrideAngularDamping=true
    bUseGravityOverride=true
    gravityScale=0.35
    alphaBoolBlend {blendOption=Linear}
    alphaScaleBiasClamp {bInterpResult=true, interpSpeedDecreasing=5, interpSpeedIncreasing=5}
    boundBone {boneName=dyn_R_bowStrand_1}
    componentLinearAccScale=<1 3 1>
    componentLinearVelScale=<1 3 1>
    physicsBodyDefinitions [{
      sphereCollisionRadius=0
      boundBone {boneName=dyn_R_bowStrand_1}
      boxExtents=<20 20 20>
      constraintSetup {angularLimitsMax=<0 10 0>, angularLimitsMin=<0 -10 0>, angularTarget=<1 0 0>}
      localJointOffset=<-4 0 0>
     }]
   }
   n42 = Anim Dynamics - Bone: dyn_R_bowStrand_2(ComponentPose=@n41, Alpha=@n40, GravityOverride="0,0,-1000", ExternalForce="0,0,0") {
    angularDampingOverride=0.9999
    angularSpringConstant=5
    bAngularSpring=true
    bOverrideAngularDamping=true
    bUseGravityOverride=true
    gravityScale=0.35
    alphaBoolBlend {blendOption=Linear}
    alphaScaleBiasClamp {bInterpResult=true, interpSpeedDecreasing=5, interpSpeedIncreasing=5}
    boundBone {boneName=dyn_R_bowStrand_2}
    componentLinearAccScale=<1 3 1>
    componentLinearVelScale=<1 3 1>
    physicsBodyDefinitions [{
      sphereCollisionRadius=0
      boundBone {boneName=dyn_R_bowStrand_2}
      boxExtents=<20 20 20>
      constraintSetup {angularLimitsMax=<0 20 0>, angularLimitsMin=<0 -20 0>, angularTarget=<1 0 0>}
      localJointOffset=<-4 0 0>
     }]
   }
   n43 = Anim Dynamics - Bone: dyn_R_bowStrand_3(ComponentPose=@n42, Alpha=@n40, GravityOverride="0,0,-1000", ExternalForce="0,0,0") {
    angularDampingOverride=0.9999
    angularSpringConstant=5
    bAngularSpring=true
    bOverrideAngularDamping=true
    bUseGravityOverride=true
    gravityScale=0.35
    alphaBoolBlend {blendOption=Linear}
    alphaScaleBiasClamp {bInterpResult=true, interpSpeedDecreasing=5, interpSpeedIncreasing=5}
    boundBone {boneName=dyn_R_bowStrand_3}
    componentLinearAccScale=<1 3 1>
    componentLinearVelScale=<1 3 1>
    physicsBodyDefinitions [{
      sphereCollisionRadius=0
      boundBone {boneName=dyn_R_bowStrand_3}
      boxExtents=<15 15 15>
      constraintSetup {angularLimitsMax=<0 20 0>, angularLimitsMin=<0 -20 0>, angularTarget=<1 0 0>}
      localJointOffset=<-4 0 0>
     }]
   }
   n44 = Anim Dynamics - Bone: dyn_R_bowStrand_4(ComponentPose=@n43, Alpha=@n40, GravityOverride="0,0,-1000", ExternalForce="0,0,0") {
    angularDampingOverride=0.9999
    angularSpringConstant=5
    bAngularSpring=true
    bOverrideAngularDamping=true
    bUseGravityOverride=true
    gravityScale=0.35
    alphaBoolBlend {blendOption=Linear}
    alphaScaleBiasClamp {bInterpResult=true, interpSpeedDecreasing=5, interpSpeedIncreasing=5}
    boundBone {boneName=dyn_R_bowStrand_4}
    componentLinearAccScale=<1 3 1>
    componentLinearVelScale=<1 3 1>
    physicsBodyDefinitions [{
      sphereCollisionRadius=0
      boundBone {boneName=dyn_R_bowStrand_4}
      boxExtents=<15 15 15>
      constraintSetup {angularLimitsMax=<0 20 0>, angularLimitsMin=<0 -20 0>, angularTarget=<1 0 0>}
      localJointOffset=<-3 0 0>
     }]
   }
   n45 = Anim Dynamics - Bone: dyn_R_bowStrand_5(ComponentPose=@n44, Alpha=@n40, GravityOverride="0,0,-1000", ExternalForce="0,0,0") {
    angularDampingOverride=0.9999
    angularSpringConstant=5
    bAngularSpring=true
    bOverrideAngularDamping=true
    bUseGravityOverride=true
    gravityScale=0.35
    alphaBoolBlend {blendOption=Linear}
    alphaScaleBiasClamp {bInterpResult=true, interpSpeedDecreasing=5, interpSpeedIncreasing=5}
    boundBone {boneName=dyn_R_bowStrand_5}
    componentLinearAccScale=<1 3 1>
    componentLinearVelScale=<1 3 1>
    physicsBodyDefinitions [{
      sphereCollisionRadius=0
      boundBone {boneName=dyn_R_bowStrand_5}
      boxExtents=<15 15 15>
      constraintSetup {angularLimitsMax=<0 20 0>, angularLimitsMin=<0 -20 0>, angularTarget=<1 0 0>}
      localJointOffset=<-3 0 0>
     }]
   }
   n46 = Anim Dynamics - Bone: dyn_L_bowStrand_1(ComponentPose=@n45, Alpha=@n40, GravityOverride="0,0,-1000", ExternalForce="0,0,0") {
    angularDampingOverride=0.9999
    angularSpringConstant=5
    bAngularSpring=true
    bOverrideAngularDamping=true
    bUseGravityOverride=true
    gravityScale=0.35
    alphaBoolBlend {blendOption=Linear}
    alphaScaleBiasClamp {bInterpResult=true, interpSpeedDecreasing=5, interpSpeedIncreasing=5}
    boundBone {boneName=dyn_L_bowStrand_1}
    componentLinearAccScale=<1 3 1>
    componentLinearVelScale=<1 3 1>
    physicsBodyDefinitions [{
      sphereCollisionRadius=0
      boundBone {boneName=dyn_L_bowStrand_1}
      boxExtents=<20 20 20>
      constraintSetup {angularLimitsMax=<0 10 0>, angularLimitsMin=<0 -10 0>, angularTarget=<1 0 0>}
      localJointOffset=<-4 0 0>
     }]
   }
   n47 = Anim Dynamics - Bone: dyn_L_bowStrand_2(ComponentPose=@n46, Alpha=@n40, GravityOverride="0,0,-1000", ExternalForce="0,0,0") {
    angularDampingOverride=0.9999
    angularSpringConstant=5
    bAngularSpring=true
    bOverrideAngularDamping=true
    bUseGravityOverride=true
    gravityScale=0.35
    alphaBoolBlend {blendOption=Linear}
    alphaScaleBiasClamp {bInterpResult=true, interpSpeedDecreasing=5, interpSpeedIncreasing=5}
    boundBone {boneName=dyn_L_bowStrand_2}
    componentLinearAccScale=<1 3 1>
    componentLinearVelScale=<1 3 1>
    physicsBodyDefinitions [{
      sphereCollisionRadius=0
      boundBone {boneName=dyn_L_bowStrand_2}
      boxExtents=<20 20 20>
      constraintSetup {angularLimitsMax=<0 20 0>, angularLimitsMin=<0 -20 0>, angularTarget=<1 0 0>}
      localJointOffset=<-4 0 0>
     }]
   }
   n48 = Anim Dynamics - Bone: dyn_L_bowStrand_3(ComponentPose=@n47, Alpha=@n40, GravityOverride="0,0,-1000", ExternalForce="0,0,0") {
    angularDampingOverride=0.9999
    angularSpringConstant=5
    bAngularSpring=true
    bOverrideAngularDamping=true
    bUseGravityOverride=true
    gravityScale=0.35
    alphaBoolBlend {blendOption=Linear}
    alphaScaleBiasClamp {bInterpResult=true, interpSpeedDecreasing=5, interpSpeedIncreasing=5}
    boundBone {boneName=dyn_L_bowStrand_3}
    componentLinearAccScale=<1 3 1>
    componentLinearVelScale=<1 3 1>
    physicsBodyDefinitions [{
      sphereCollisionRadius=0
      boundBone {boneName=dyn_L_bowStrand_3}
      boxExtents=<15 15 15>
      constraintSetup {angularLimitsMax=<0 20 0>, angularLimitsMin=<0 -20 0>, angularTarget=<1 0 0>}
      localJointOffset=<-4 0 0>
     }]
   }
   n49 = Anim Dynamics - Bone: dyn_L_bowStrand_4(ComponentPose=@n48, Alpha=@n40, GravityOverride="0,0,-1000", ExternalForce="0,0,0") {
    angularDampingOverride=0.9999
    angularSpringConstant=5
    bAngularSpring=true
    bOverrideAngularDamping=true
    bUseGravityOverride=true
    gravityScale=0.35
    alphaBoolBlend {blendOption=Linear}
    alphaScaleBiasClamp {bInterpResult=true, interpSpeedDecreasing=5, interpSpeedIncreasing=5}
    boundBone {boneName=dyn_L_bowStrand_4}
    componentLinearAccScale=<1 3 1>
    componentLinearVelScale=<1 3 1>
    physicsBodyDefinitions [{
      sphereCollisionRadius=0
      boundBone {boneName=dyn_L_bowStrand_4}
      boxExtents=<15 15 15>
      constraintSetup {angularLimitsMax=<0 20 0>, angularLimitsMin=<0 -20 0>, angularTarget=<1 0 0>}
      localJointOffset=<-3 0 0>
     }]
   }
   n50 = Anim Dynamics - Bone: dyn_L_bowStrand_5(ComponentPose=@n49, Alpha=@n40, GravityOverride="0,0,-1000", ExternalForce="0,0,0") {
    angularDampingOverride=0.9999
    angularSpringConstant=5
    bAngularSpring=true
    bOverrideAngularDamping=true
    bUseGravityOverride=true
    gravityScale=0.35
    alphaBoolBlend {blendOption=Linear}
    alphaScaleBiasClamp {bInterpResult=true, interpSpeedDecreasing=5, interpSpeedIncreasing=5}
    boundBone {boneName=dyn_L_bowStrand_5}
    componentLinearAccScale=<1 3 1>
    componentLinearVelScale=<1 3 1>
    physicsBodyDefinitions [{
      sphereCollisionRadius=0
      boundBone {boneName=dyn_L_bowStrand_5}
      boxExtents=<15 15 15>
      constraintSetup {angularLimitsMax=<0 20 0>, angularLimitsMin=<0 -20 0>, angularTarget=<1 0 0>}
      localJointOffset=<-3 0 0>
     }]
   }
   Output Pose(Result=Blend Poses by bool(
     BlendPose_0=Component To Local(ComponentPose=@n50)
     BlendPose_1=Component To Local(ComponentPose=@n30)
     bActiveValue=Get Is Backbling A Cape() -> (Is Backbling A Cape:bool)
    ) {blendType=Linear}) {layerGroup=None}
   n51 = RigidBody!(
    ComponentPose=Local To Component(LocalPose=Copy Pose From Mesh() {bUseAttachedParent=true})
    Alpha
    OverrideWorldGravity:Vector=Get gravityOverride() -> (gravityOverride:Vector)
    ExternalForce
   ) -> (Pose) {
    bClampLinearTranslationLimitToRefPose=true
    bOverrideWorldGravity=true
    overridePhysicsAsset=&/F_MED_Thrive_lowerSkirtPhysics
    simulationSpace=BaseBoneSpace
    alphaBoolBlend {blendOption=Linear}
    alphaScaleBiasClamp {bInterpResult=true}
    baseBoneRef {boneName=pelvis}
    componentLinearAccScale=<0.3 0.3 0.3>
    componentLinearVelScale=<0.4 0.4 0.4>
    overrideWorldGravity {z=-980}
    simSpaceSettings {maxAngularAcceleration=9, maxAngularVelocity=9, maxLinearAcceleration=9, maxLinearVelocity=9}
   }
   n52 = Control Rig(Source=Component To Local(ComponentPose=@n51)) {
    alphaBoolBlend {blendOption=Linear}
    alphaScaleBiasClamp {bInterpResult=true}
    controlRigAssetReference {blueprintRigClass=&/F_MED_Thrive_CtrlRig.F_MED_Thrive_CtrlRig_C}
   }
   n53 = "Transform (Modify) Bone - Bone: dyn_L_shoeTongue"!(ComponentPose=Local To Component(LocalPose=@n52), Alpha=Get L footAlpha() -> (L footAlpha:double), Translation="0,0,0", Scale="1,1,1") -> (Pose) {rotationMode=BMM_Additive, alphaBoolBlend {blendOption=Linear}, boneToModify {boneName=dyn_L_shoeTongue}, rotation {roll=60}}
   n54 = "Transform (Modify) Bone - Bone: dyn_R_shoeTongue"!(ComponentPose=@n53, Alpha=Get R footAlpha() -> (R footAlpha:double), Translation="0,0,0", Scale="1,1,1") -> (Pose) {rotationMode=BMM_Additive, alphaBoolBlend {blendOption=Linear}, boneToModify {boneName=dyn_R_shoeTongue}, rotation {roll=60}}
   n55 = "Constraint - dyn_skirtStar_A (PPPP)"!(
    ComponentPose=@n54
    Alpha=0.8
    ConstraintWeights_0=0.25
    ConstraintWeights_1:float=0.25
    ConstraintWeights_2:float=0.25
    ConstraintWeights_3:float=0.25
   ) -> (Pose) {
    alphaBoolBlend {blendOption=Linear}
    boneToModify {boneName=dyn_skirtStar_A}
    constraintSetup [
     {targetBone {boneName=dyn_center_lowerSkirt_2}}
     {targetBone {boneName=dyn_center_lowerSkirt_3}}
     {targetBone {boneName=dyn_L_front_LowerSkirt_3}}
     {targetBone {boneName=dyn_R_front_LowerSkirt_3}}
    ]
    constraintWeights [1,1,1,1]
   }
   n56 = "Constraint - dyn_skirtStar_B (PPPPPP)"!(
    ComponentPose=@n55
    Alpha=1
    ConstraintWeights_0=0.362
    ConstraintWeights_1:float=0.0457
    ConstraintWeights_2:float=0.0963
    ConstraintWeights_3:float=0.4484
    ConstraintWeights_4:float=0.021
    ConstraintWeights_5:float=0.0266
   ) -> (Pose) {
    alphaBoolBlend {blendOption=Linear}
    boneToModify {boneName=dyn_skirtStar_B}
    constraintSetup [
     {targetBone {boneName=pelvis}}
     {targetBone {boneName=dyn_L_front_LowerSkirt_1}}
     {targetBone {boneName=dyn_L_front_LowerSkirt_2}}
     {targetBone {boneName=dyn_L_front_LowerSkirt_3}}
     {targetBone {boneName=dyn_L_side_LowerSkirt_2}}
     {targetBone {boneName=dyn_L_side_LowerSkirt_3}}
    ]
    constraintWeights [1,1,1,1,1,1]
   }
   n57 = "Constraint - dyn_skirtStar_C (PPPPPPPP)"!(
    ComponentPose=@n56
    Alpha=1
    ConstraintWeights_0=0.4055
    ConstraintWeights_1:float=0.0119
    ConstraintWeights_2:float=0.0178
    ConstraintWeights_3:float=0.0391
    ConstraintWeights_4:float=0.127
    ConstraintWeights_5:float=0.3784
    ConstraintWeights_6:float=0.0096
    ConstraintWeights_7:float=0.0107
   ) -> (Pose) {
    alphaBoolBlend {blendOption=Linear}
    boneToModify {boneName=dyn_skirtStar_C}
    constraintSetup [
     {targetBone {boneName=pelvis}}
     {targetBone {boneName=dyn_L_front_LowerSkirt_1}}
     {targetBone {boneName=dyn_L_front_LowerSkirt_2}}
     {targetBone {boneName=dyn_L_front_LowerSkirt_3}}
     {targetBone {boneName=dyn_L_side_LowerSkirt_2}}
     {targetBone {boneName=dyn_L_side_LowerSkirt_3}}
     {targetBone {boneName=dyn_L_back_LowerSkirt_2}}
     {targetBone {boneName=dyn_L_back_LowerSkirt_3}}
    ]
    constraintWeights [1,1,1,1,1,1,1,1]
   }
   n58 = "Constraint - dyn_skirtStar_D (PPPPPPPP)"!(
    ComponentPose=@n57
    Alpha=1
    ConstraintWeights_0=0.4023
    ConstraintWeights_1:float=0.0018
    ConstraintWeights_2:float=0.0023
    ConstraintWeights_3:float=0.0028
    ConstraintWeights_4:float=0.0471
    ConstraintWeights_5:float=0.1154
    ConstraintWeights_6:float=0.1011
    ConstraintWeights_7:float=0.3273
   ) -> (Pose) {
    alphaBoolBlend {blendOption=Linear}
    boneToModify {boneName=dyn_skirtStar_D}
    constraintSetup [
     {targetBone {boneName=pelvis}}
     {targetBone {boneName=dyn_R_back_LowerSkirt_1}}
     {targetBone {boneName=dyn_R_back_LowerSkirt_2}}
     {targetBone {boneName=dyn_R_back_LowerSkirt_3}}
     {targetBone {boneName=dyn_L_side_LowerSkirt_2}}
     {targetBone {boneName=dyn_L_side_LowerSkirt_3}}
     {targetBone {boneName=dyn_L_back_LowerSkirt_2}}
     {targetBone {boneName=dyn_L_back_LowerSkirt_3}}
    ]
    constraintWeights [1,1,1,1,1,1,1,1]
   }
   n59 = "Constraint - dyn_skirtStar_E (PPPPPPPPPP)"!(
    ComponentPose=@n58
    Alpha=1
    ConstraintWeights_0=0.4984
    ConstraintWeights_1:float=0.0008
    ConstraintWeights_2:float=0.0007
    ConstraintWeights_3:float=0.0342
    ConstraintWeights_4:float=0.0634
    ConstraintWeights_5:float=0.2109
    ConstraintWeights_6:float=0.0006
    ConstraintWeights_7:float=0.0006
    ConstraintWeights_8:float=0.0477
    ConstraintWeights_9:float=0.1427
   ) -> (Pose) {
    alphaBoolBlend {blendOption=Linear}
    boneToModify {boneName=dyn_skirtStar_E}
    constraintSetup [
     {targetBone {boneName=pelvis}}
     {targetBone {boneName=dyn_R_side_LowerSkirt_1}}
     {targetBone {boneName=dyn_R_side_LowerSkirt_2}}
     {targetBone {boneName=dyn_R_back_LowerSkirt_1}}
     {targetBone {boneName=dyn_R_back_LowerSkirt_2}}
     {targetBone {boneName=dyn_R_back_LowerSkirt_3}}
     {targetBone {boneName=dyn_L_side_LowerSkirt_2}}
     {targetBone {boneName=dyn_L_side_LowerSkirt_3}}
     {targetBone {boneName=dyn_L_back_LowerSkirt_2}}
     {targetBone {boneName=dyn_L_back_LowerSkirt_3}}
    ]
    constraintWeights [1,1,1,1,1,1,1,1,1,1]
   }
   n60 = "Constraint - dyn_skirtStar_F (PPPPPPPPP)"!(
    ComponentPose=@n59
    Alpha=1
    ConstraintWeights_0=0.2849
    ConstraintWeights_1:float=0.0138
    ConstraintWeights_2:float=0.0258
    ConstraintWeights_3:float=0.0429
    ConstraintWeights_4:float=0.0552
    ConstraintWeights_5:float=0.1074
    ConstraintWeights_6:float=0.4638
    ConstraintWeights_7:float=0.0028
    ConstraintWeights_8:float=0.0034
   ) -> (Pose) {
    alphaBoolBlend {blendOption=Linear}
    boneToModify {boneName=dyn_skirtStar_F}
    constraintSetup [
     {targetBone {boneName=pelvis}}
     {targetBone {boneName=dyn_R_side_LowerSkirt_1}}
     {targetBone {boneName=dyn_R_side_LowerSkirt_2}}
     {targetBone {boneName=dyn_R_side_LowerSkirt_3}}
     {targetBone {boneName=dyn_R_back_LowerSkirt_1}}
     {targetBone {boneName=dyn_R_back_LowerSkirt_2}}
     {targetBone {boneName=dyn_R_back_LowerSkirt_3}}
     {targetBone {boneName=dyn_L_back_LowerSkirt_2}}
     {targetBone {boneName=dyn_L_back_LowerSkirt_3}}
    ]
    constraintWeights [1,1,1,1,1,1,1,1,1]
   }
   n61 = "Constraint - dyn_skirtStar_G (PPPPPPPPPP)"!(
    ComponentPose=@n60
    Alpha=1
    ConstraintWeights_0=0.3804
    ConstraintWeights_1:float=0.0116
    ConstraintWeights_2:float=0.0163
    ConstraintWeights_3:float=0.0315
    ConstraintWeights_4:float=0.064
    ConstraintWeights_5:float=0.1313
    ConstraintWeights_6:float=0.3057
    ConstraintWeights_7:float=0.0141
    ConstraintWeights_8:float=0.0177
    ConstraintWeights_9:float=0.0272
   ) -> (Pose) {
    alphaBoolBlend {blendOption=Linear}
    boneToModify {boneName=dyn_skirtStar_G}
    constraintSetup [
     {targetBone {boneName=pelvis}}
     {targetBone {boneName=dyn_R_front_LowerSkirt_1}}
     {targetBone {boneName=dyn_R_front_LowerSkirt_2}}
     {targetBone {boneName=dyn_R_front_LowerSkirt_3}}
     {targetBone {boneName=dyn_R_side_LowerSkirt_1}}
     {targetBone {boneName=dyn_R_side_LowerSkirt_2}}
     {targetBone {boneName=dyn_R_side_LowerSkirt_3}}
     {targetBone {boneName=dyn_R_back_LowerSkirt_1}}
     {targetBone {boneName=dyn_R_back_LowerSkirt_2}}
     {targetBone {boneName=dyn_R_back_LowerSkirt_3}}
    ]
    constraintWeights [1,1,1,1,1,1,1,1,1,1]
   }
   n62 = "Constraint - dyn_skirtStar_H (PPPPPPPP)"!(
    ComponentPose=@n61
    Alpha=1
    ConstraintWeights_0=0.3582
    ConstraintWeights_1:float=0.0547
    ConstraintWeights_2:float=0.115
    ConstraintWeights_3:float=0.4154
    ConstraintWeights_4:float=0.0134
    ConstraintWeights_5:float=0.0174
    ConstraintWeights_6:float=0.0252
    ConstraintWeights_7:float=0.0007
   ) -> (Pose) {
    alphaBoolBlend {blendOption=Linear}
    boneToModify {boneName=dyn_skirtStar_H}
    constraintSetup [
     {targetBone {boneName=pelvis}}
     {targetBone {boneName=dyn_R_front_LowerSkirt_1}}
     {targetBone {boneName=dyn_R_front_LowerSkirt_2}}
     {targetBone {boneName=dyn_R_front_LowerSkirt_3}}
     {targetBone {boneName=dyn_R_side_LowerSkirt_1}}
     {targetBone {boneName=dyn_R_side_LowerSkirt_2}}
     {targetBone {boneName=dyn_R_side_LowerSkirt_3}}
     {targetBone {boneName=dyn_L_front_LowerSkirt_1}}
    ]
    constraintWeights [1,1,1,1,1,1,1,1]
   }
   n63 = "Transform (Modify) Bone - Bone: dyn_skirtStar_A"(ComponentPose=@n62, Alpha=Get isCrouching() -> (isCrouching:double), Translation="0,0,0", Rotation="0,0,0", Scale="-1,-1,-1") {
    scaleMode=BMM_Additive
    alphaBoolBlend {blendOption=Linear}
    alphaScaleBiasClamp {bInterpResult=true}
    boneToModify {boneName=dyn_skirtStar_A}
    scale=<-1 -1 -1>
   }
   n64 = Get isCrouchMoving() -> (isCrouchMoving:double)
   n65 = "Transform (Modify) Bone - Bone: dyn_L_back_LowerSkirt_1"(ComponentPose=@n63, Alpha=@n64, Translation="5,0,0", Rotation="0,0,0", Scale="1,1,1") {
    rotationMode=BMM_Additive
    translationMode=BMM_Additive
    translationSpace=BCS_BoneSpace
    alphaBoolBlend {blendOption=Linear}
    alphaScaleBiasClamp {bInterpResult=true}
    boneToModify {boneName=dyn_L_back_LowerSkirt_1}
    translation {x=5}
   }
   n66 = "Transform (Modify) Bone - Bone: dyn_L_back_LowerSkirt_2"(ComponentPose=@n65, Alpha=@n64, Translation="0,0,0", Rotation="0,0,-35", Scale="1,1,1") {
    rotationMode=BMM_Additive
    alphaBoolBlend {blendOption=Linear}
    alphaScaleBiasClamp {bInterpResult=true}
    boneToModify {boneName=dyn_L_back_LowerSkirt_2}
    rotation {roll=-35}
   }
   n67 = "Transform (Modify) Bone - Bone: dyn_R_back_LowerSkirt_1"(ComponentPose=@n66, Alpha=@n64, Translation="5,0,0", Rotation="0,0,0", Scale="1,1,1") {
    rotationMode=BMM_Additive
    translationMode=BMM_Additive
    translationSpace=BCS_BoneSpace
    alphaBoolBlend {blendOption=Linear}
    alphaScaleBiasClamp {bInterpResult=true}
    boneToModify {boneName=dyn_R_back_LowerSkirt_1}
    translation {x=5}
   }
   n68 = "Transform (Modify) Bone - Bone: dyn_R_back_LowerSkirt_2"(ComponentPose=@n67, Alpha=@n64, Translation="0,0,0", Rotation="0,0,-35", Scale="1,1,1") {
    rotationMode=BMM_Additive
    alphaBoolBlend {blendOption=Linear}
    alphaScaleBiasClamp {bInterpResult=true}
    boneToModify {boneName=dyn_R_back_LowerSkirt_2}
    rotation {roll=-35}
   }
   n69 = Anim Dynamics - Bone: dyn_skirtStar_A(ComponentPose=@n68, Alpha=1, ExternalForce="0,0,0") {
    angularDampingOverride=1
    angularSpringConstant=1000
    bAngularSpring=true
    bLinearSpring=true
    bOverrideAngularDamping=true
    bOverrideLinearDamping=true
    bUseGravityOverride=true
    gravityScale=0.35
    linearDampingOverride=1
    linearSpringConstant=50
    simulationSpace=BoneRelative
    alphaBoolBlend {blendOption=Linear}
    alphaScaleBiasClamp {bInterpResult=true, scale=0.5}
    boundBone {boneName=dyn_skirtStar_A}
    componentLinearAccScale=<0.25 0.25 0.25>
    componentLinearVelScale=<0.25 0.25 0.25>
    gravityOverride {z=-980}
    physicsBodyDefinitions [{
      sphereCollisionRadius=0
      boundBone {boneName=dyn_skirtStar_A}
      boxExtents=<70 80 80>
      constraintSetup {angularLimitsMax=<0 0 0>, angularLimitsMin=<0 0 0>, angularTarget=<1 0 0>, linearAxesMax=<2.5 2.5 2.5>, linearAxesMin=<-2.5 -2.5 -2.5>}
      localJointOffset=<0 0 0>
     }]
    relativeSpaceBone {boneName=root}
   }
   n70 = Anim Dynamics - Bone: dyn_skirtStar_B(ComponentPose=@n69, Alpha=0.5, ExternalForce="0,0,0") {
    angularDampingOverride=1
    angularSpringConstant=1000
    bAngularSpring=true
    bLinearSpring=true
    bOverrideAngularDamping=true
    bOverrideLinearDamping=true
    bUseGravityOverride=true
    gravityScale=0.35
    linearDampingOverride=1
    linearSpringConstant=50
    simulationSpace=BoneRelative
    alphaBoolBlend {blendOption=Linear}
    alphaScaleBiasClamp {bInterpResult=true, scale=0.5}
    boundBone {boneName=dyn_skirtStar_B}
    componentLinearAccScale=<0.4 0.4 0.4>
    componentLinearVelScale=<0.4 0.4 0.4>
    gravityOverride {z=-980}
    physicsBodyDefinitions [{
      sphereCollisionRadius=0
      boundBone {boneName=dyn_skirtStar_B}
      boxExtents=<70 80 80>
      constraintSetup {angularLimitsMax=<0 0 0>, angularLimitsMin=<0 0 0>, angularTarget=<1 0 0>, linearAxesMax=<2.5 2.5 2.5>, linearAxesMin=<-2.5 -2.5 -2.5>}
      localJointOffset=<0 0 0>
     }]
    relativeSpaceBone {boneName=root}
   }
   n71 = Anim Dynamics - Bone: dyn_skirtStar_C(ComponentPose=@n70, Alpha=1, ExternalForce="0,0,0") {
    angularDampingOverride=1
    angularSpringConstant=1000
    bAngularSpring=true
    bLinearSpring=true
    bOverrideAngularDamping=true
    bOverrideLinearDamping=true
    bUseGravityOverride=true
    gravityScale=0.35
    linearDampingOverride=1
    linearSpringConstant=50
    simulationSpace=BoneRelative
    alphaBoolBlend {blendOption=Linear}
    alphaScaleBiasClamp {bInterpResult=true, scale=0.5}
    boundBone {boneName=dyn_skirtStar_C}
    componentLinearAccScale=<0.25 0.25 0.25>
    componentLinearVelScale=<0.25 0.25 0.25>
    gravityOverride {z=-980}
    physicsBodyDefinitions [{
      sphereCollisionRadius=0
      boundBone {boneName=dyn_skirtStar_C}
      boxExtents=<70 80 80>
      constraintSetup {angularLimitsMax=<0 0 0>, angularLimitsMin=<0 0 0>, angularTarget=<1 0 0>, linearAxesMax=<2.5 2.5 2.5>, linearAxesMin=<-2.5 -2.5 -2.5>}
      localJointOffset=<0 0 0>
     }]
    relativeSpaceBone {boneName=root}
   }
   n72 = Anim Dynamics - Bone: dyn_skirtStar_D(ComponentPose=@n71, Alpha=1, ExternalForce="0,0,0") {
    angularDampingOverride=1
    angularSpringConstant=1000
    bAngularSpring=true
    bLinearSpring=true
    bOverrideAngularDamping=true
    bOverrideLinearDamping=true
    bUseGravityOverride=true
    gravityScale=0.35
    linearDampingOverride=1
    linearSpringConstant=50
    simulationSpace=BoneRelative
    alphaBoolBlend {blendOption=Linear}
    alphaScaleBiasClamp {bInterpResult=true, scale=0.5}
    boundBone {boneName=dyn_skirtStar_D}
    componentLinearAccScale=<0.4 0.4 0.4>
    componentLinearVelScale=<0.4 0.4 0.4>
    gravityOverride {z=-980}
    physicsBodyDefinitions [{
      sphereCollisionRadius=0
      boundBone {boneName=dyn_skirtStar_D}
      boxExtents=<70 80 80>
      constraintSetup {angularLimitsMax=<0 0 0>, angularLimitsMin=<0 0 0>, angularTarget=<1 0 0>, linearAxesMax=<2.5 2.5 2.5>, linearAxesMin=<-2.5 -2.5 -2.5>}
      localJointOffset=<0 0 0>
     }]
    relativeSpaceBone {boneName=root}
   }
   n73 = Anim Dynamics - Bone: dyn_skirtStar_E(ComponentPose=@n72, Alpha=1, ExternalForce="0,0,0") {
    angularDampingOverride=1
    angularSpringConstant=1000
    bAngularSpring=true
    bLinearSpring=true
    bOverrideAngularDamping=true
    bOverrideLinearDamping=true
    bUseGravityOverride=true
    gravityScale=0.35
    linearDampingOverride=1
    linearSpringConstant=50
    simulationSpace=BoneRelative
    alphaBoolBlend {blendOption=Linear}
    alphaScaleBiasClamp {bInterpResult=true, scale=0.5}
    boundBone {boneName=dyn_skirtStar_E}
    componentLinearAccScale=<0.25 0.25 0.25>
    componentLinearVelScale=<0.25 0.25 0.25>
    gravityOverride {z=-980}
    physicsBodyDefinitions [{
      sphereCollisionRadius=0
      boundBone {boneName=dyn_skirtStar_E}
      boxExtents=<70 80 80>
      constraintSetup {angularLimitsMax=<0 0 0>, angularLimitsMin=<0 0 0>, angularTarget=<1 0 0>, linearAxesMax=<2.5 2.5 2.5>, linearAxesMin=<-2.5 -2.5 -2.5>}
      localJointOffset=<0 0 0>
     }]
    relativeSpaceBone {boneName=root}
   }
   n74 = Anim Dynamics - Bone: dyn_skirtStar_F(ComponentPose=@n73, Alpha=1, ExternalForce="0,0,0") {
    angularDampingOverride=1
    angularSpringConstant=1000
    bAngularSpring=true
    bLinearSpring=true
    bOverrideAngularDamping=true
    bOverrideLinearDamping=true
    bUseGravityOverride=true
    gravityScale=0.35
    linearDampingOverride=1
    linearSpringConstant=50
    simulationSpace=BoneRelative
    alphaBoolBlend {blendOption=Linear}
    alphaScaleBiasClamp {bInterpResult=true, scale=0.5}
    boundBone {boneName=dyn_skirtStar_F}
    componentLinearAccScale=<0.4 0.4 0.4>
    componentLinearVelScale=<0.4 0.4 0.4>
    gravityOverride {z=-980}
    physicsBodyDefinitions [{
      sphereCollisionRadius=0
      boundBone {boneName=dyn_skirtStar_F}
      boxExtents=<70 80 80>
      constraintSetup {angularLimitsMax=<0 0 0>, angularLimitsMin=<0 0 0>, angularTarget=<1 0 0>, linearAxesMax=<2.5 2.5 2.5>, linearAxesMin=<-2.5 -2.5 -2.5>}
      localJointOffset=<0 0 0>
     }]
    relativeSpaceBone {boneName=root}
   }
   n75 = Anim Dynamics - Bone: dyn_skirtStar_G(ComponentPose=@n74, Alpha=1, ExternalForce="0,0,0") {
    angularDampingOverride=1
    angularSpringConstant=1000
    bAngularSpring=true
    bLinearSpring=true
    bOverrideAngularDamping=true
    bOverrideLinearDamping=true
    bUseGravityOverride=true
    gravityScale=0.35
    linearDampingOverride=1
    linearSpringConstant=50
    simulationSpace=BoneRelative
    alphaBoolBlend {blendOption=Linear}
    alphaScaleBiasClamp {bInterpResult=true, scale=0.5}
    boundBone {boneName=dyn_skirtStar_G}
    componentLinearAccScale=<0.25 0.25 0.25>
    componentLinearVelScale=<0.25 0.25 0.25>
    gravityOverride {z=-980}
    physicsBodyDefinitions [{
      sphereCollisionRadius=0
      boundBone {boneName=dyn_skirtStar_G}
      boxExtents=<70 80 80>
      constraintSetup {angularLimitsMax=<0 0 0>, angularLimitsMin=<0 0 0>, angularTarget=<1 0 0>, linearAxesMax=<2.5 2.5 2.5>, linearAxesMin=<-2.5 -2.5 -2.5>}
      localJointOffset=<0 0 0>
     }]
    relativeSpaceBone {boneName=root}
   }
   n76 = Anim Dynamics - Bone: dyn_skirtStar_H(ComponentPose=@n75, Alpha=0.5, ExternalForce="0,0,0") {
    angularDampingOverride=1
    angularSpringConstant=1000
    bAngularSpring=true
    bLinearSpring=true
    bOverrideAngularDamping=true
    bOverrideLinearDamping=true
    bUseGravityOverride=true
    gravityScale=0.35
    linearDampingOverride=1
    linearSpringConstant=50
    simulationSpace=BoneRelative
    alphaBoolBlend {blendOption=Linear}
    alphaScaleBiasClamp {bInterpResult=true, scale=0.5}
    boundBone {boneName=dyn_skirtStar_H}
    componentLinearAccScale=<0.4 0.4 0.4>
    componentLinearVelScale=<0.4 0.4 0.4>
    gravityOverride {z=-980}
    physicsBodyDefinitions [{
      sphereCollisionRadius=0
      boundBone {boneName=dyn_skirtStar_H}
      boxExtents=<70 80 80>
      constraintSetup {angularLimitsMax=<0 0 0>, angularLimitsMin=<0 0 0>, angularTarget=<1 0 0>, linearAxesMax=<2.5 2.5 2.5>, linearAxesMin=<-2.5 -2.5 -2.5>}
      localJointOffset=<0 0 0>
     }]
    relativeSpaceBone {boneName=root}
   }
   n77 = Get sineRotation1() -> (sineRotation1:Rotator)
   n78 = Get sineRotation3()
   n79 = "Transform (Modify) Bone - Bone: dyn_L_sleeve_1"(
    ComponentPose="Transform (Modify) Bone - Bone: dyn_R_underSleeve"(
     ComponentPose="Transform (Modify) Bone - Bone: dyn_L_underSleeve"(ComponentPose=@n76, Alpha=1, Translation="0,0,0", Rotation=@n77, Scale="1,1,1") {rotationMode=BMM_Additive, alphaBoolBlend {blendOption=Linear}, boneToModify {boneName=dyn_L_underSleeve}}
     Alpha=1
     Translation="0,0,0"
     Rotation=Get sineRotation2() -> (sineRotation2:Rotator)
     Scale="1,1,1"
    ) {rotationMode=BMM_Additive, alphaBoolBlend {blendOption=Linear}, boneToModify {boneName=dyn_R_underSleeve}}
    Alpha=1
    Translation="0,0,0"
    Rotation=@n78
    Scale="1,1,1"
   ) {rotationMode=BMM_Replace, rotationSpace=BCS_BoneSpace, alphaBoolBlend {blendOption=Linear}, boneToModify {boneName=dyn_L_sleeve_1}}
   n80 = Get sineRotation4()
   n81 = "Transform (Modify) Bone - Bone: dyn_L_sleeve_2"(ComponentPose=@n79, Alpha=1, Translation="0,0,0", Rotation=@n80, Scale="1,1,1") {rotationMode=BMM_Replace, rotationSpace=BCS_BoneSpace, alphaBoolBlend {blendOption=Linear}, boneToModify {boneName=dyn_L_sleeve_2}}
   n82 = Get sineRotation5()
   n83 = "Transform (Modify) Bone - Bone: dyn_L_sleeve_3"(ComponentPose=@n81, Alpha=1, Translation="0,0,0", Rotation=@n82, Scale="1,1,1") {rotationMode=BMM_Replace, rotationSpace=BCS_BoneSpace, alphaBoolBlend {blendOption=Linear}, boneToModify {boneName=dyn_L_sleeve_3}}
   n84 = "Transform (Modify) Bone - Bone: dyn_R_sleeve_1"(ComponentPose=@n83, Alpha=1, Translation="0,0,0", Rotation=@n78, Scale="1,1,1") {rotationMode=BMM_Replace, rotationSpace=BCS_BoneSpace, alphaBoolBlend {blendOption=Linear}, boneToModify {boneName=dyn_R_sleeve_1}}
   n85 = "Transform (Modify) Bone - Bone: dyn_R_sleeve_2"(ComponentPose=@n84, Alpha=1, Translation="0,0,0", Rotation=@n80, Scale="1,1,1") {rotationMode=BMM_Replace, rotationSpace=BCS_BoneSpace, alphaBoolBlend {blendOption=Linear}, boneToModify {boneName=dyn_R_sleeve_2}}
   n86 = "Transform (Modify) Bone - Bone: dyn_R_sleeve_3"(ComponentPose=@n85, Alpha=1, Translation="0,0,0", Rotation=@n82, Scale="1,1,1") {rotationMode=BMM_Replace, rotationSpace=BCS_BoneSpace, alphaBoolBlend {blendOption=Linear}, boneToModify {boneName=dyn_R_sleeve_3}}
   n87 = Anim Dynamics - Bone: dyn_R_sleeve_1(ComponentPose=@n86, Alpha=1, ExternalForce="0,0,0") {
    angularDampingOverride=1
    angularSpringConstant=20000
    bAngularSpring=true
    bOverrideAngularDamping=true
    bUseGravityOverride=true
    gravityScale=0.35
    simulationSpace=RootRelative
    alphaBoolBlend {blendOption=Linear}
    boundBone {boneName=dyn_R_sleeve_1}
    componentLinearAccScale=<1 1 1>
    componentLinearVelScale=<1 1 1>
    gravityOverride {z=-980}
    physicsBodyDefinitions [{
      sphereCollisionRadius=0
      boundBone {boneName=dyn_R_sleeve_1}
      boxExtents=<30 40 40>
      constraintSetup {angularLimitsMax=<15 15 15>, angularLimitsMin=<-15 -15 -15>, angularTarget=<1 0 0>}
      localJointOffset=<4 0 0>
     }]
   }
   n88 = Anim Dynamics - Bone: dyn_R_sleeve_2(ComponentPose=@n87, Alpha=1, ExternalForce="0,0,0") {
    angularDampingOverride=1
    angularSpringConstant=20000
    bAngularSpring=true
    bOverrideAngularDamping=true
    bUseGravityOverride=true
    gravityScale=0.35
    simulationSpace=RootRelative
    alphaBoolBlend {blendOption=Linear}
    boundBone {boneName=dyn_R_sleeve_2}
    componentLinearAccScale=<1 1 1>
    componentLinearVelScale=<1 1 1>
    gravityOverride {z=-980}
    physicsBodyDefinitions [{
      sphereCollisionRadius=0
      boundBone {boneName=dyn_R_sleeve_2}
      boxExtents=<30 40 40>
      constraintSetup {angularLimitsMax=<15 15 15>, angularLimitsMin=<-15 -15 -15>, angularTarget=<1 0 0>}
      localJointOffset=<4 0 0>
     }]
   }
   n89 = Anim Dynamics - Bone: dyn_R_sleeve_3(ComponentPose=@n88, Alpha=1, ExternalForce="0,0,0") {
    angularDampingOverride=1
    angularSpringConstant=20000
    bAngularSpring=true
    bOverrideAngularDamping=true
    bUseGravityOverride=true
    gravityScale=0.35
    simulationSpace=RootRelative
    alphaBoolBlend {blendOption=Linear}
    boundBone {boneName=dyn_R_sleeve_3}
    componentLinearAccScale=<1 1 1>
    componentLinearVelScale=<1 1 1>
    gravityOverride {z=-980}
    physicsBodyDefinitions [{
      sphereCollisionRadius=0
      boundBone {boneName=dyn_R_sleeve_3}
      boxExtents=<30 40 40>
      constraintSetup {angularLimitsMax=<15 15 15>, angularLimitsMin=<-15 -15 -15>, angularTarget=<1 0 0>}
      localJointOffset=<4 0 0>
     }]
   }
   n90 = Anim Dynamics - Bone: dyn_R_underSleeve(ComponentPose=@n89, Alpha=1, ExternalForce="0,0,0") {
    angularDampingOverride=1
    angularSpringConstant=20000
    bAngularSpring=true
    bOverrideAngularDamping=true
    bUseGravityOverride=true
    gravityScale=0.35
    simulationSpace=RootRelative
    alphaBoolBlend {blendOption=Linear}
    boundBone {boneName=dyn_R_underSleeve}
    componentLinearAccScale=<1 1 1>
    componentLinearVelScale=<1 1 1>
    gravityOverride {z=-980}
    physicsBodyDefinitions [{
      sphereCollisionRadius=0
      boundBone {boneName=dyn_R_underSleeve}
      boxExtents=<20 30 30>
      constraintSetup {angularLimitsMax=<15 15 15>, angularLimitsMin=<-15 -15 -15>, angularTarget=<1 0 0>}
      localJointOffset=<4 0 0>
     }]
   }
   n91 = Anim Dynamics - Bone: dyn_L_sleeve_1(ComponentPose=@n90, Alpha=1, ExternalForce="0,0,0") {
    angularDampingOverride=1
    angularSpringConstant=20000
    bAngularSpring=true
    bOverrideAngularDamping=true
    bUseGravityOverride=true
    gravityScale=0.35
    simulationSpace=RootRelative
    alphaBoolBlend {blendOption=Linear}
    boundBone {boneName=dyn_L_sleeve_1}
    componentLinearAccScale=<1 1 1>
    componentLinearVelScale=<1 1 1>
    gravityOverride {z=-980}
    physicsBodyDefinitions [{
      sphereCollisionRadius=0
      boundBone {boneName=dyn_L_sleeve_1}
      boxExtents=<30 40 40>
      constraintSetup {angularLimitsMax=<15 15 15>, angularLimitsMin=<-15 -15 -15>, angularTarget=<1 0 0>}
      localJointOffset=<4 0 0>
     }]
   }
   n92 = Anim Dynamics - Bone: dyn_L_sleeve_2(ComponentPose=@n91, Alpha=1, ExternalForce="0,0,0") {
    angularDampingOverride=1
    angularSpringConstant=20000
    bAngularSpring=true
    bOverrideAngularDamping=true
    bUseGravityOverride=true
    gravityScale=0.35
    simulationSpace=RootRelative
    alphaBoolBlend {blendOption=Linear}
    boundBone {boneName=dyn_L_sleeve_2}
    componentLinearAccScale=<1 1 1>
    componentLinearVelScale=<1 1 1>
    gravityOverride {z=-980}
    physicsBodyDefinitions [{
      sphereCollisionRadius=0
      boundBone {boneName=dyn_L_sleeve_2}
      boxExtents=<30 40 40>
      constraintSetup {angularLimitsMax=<15 15 15>, angularLimitsMin=<-15 -15 -15>, angularTarget=<1 0 0>}
      localJointOffset=<4 0 0>
     }]
   }
   n93 = Anim Dynamics - Bone: dyn_L_sleeve_3(ComponentPose=@n92, Alpha=1, ExternalForce="0,0,0") {
    angularDampingOverride=1
    angularSpringConstant=20000
    bAngularSpring=true
    bOverrideAngularDamping=true
    bUseGravityOverride=true
    gravityScale=0.35
    simulationSpace=RootRelative
    alphaBoolBlend {blendOption=Linear}
    boundBone {boneName=dyn_L_sleeve_3}
    componentLinearAccScale=<1 1 1>
    componentLinearVelScale=<1 1 1>
    gravityOverride {z=-980}
    physicsBodyDefinitions [{
      sphereCollisionRadius=0
      boundBone {boneName=dyn_L_sleeve_3}
      boxExtents=<30 40 40>
      constraintSetup {angularLimitsMax=<15 15 15>, angularLimitsMin=<-15 -15 -15>, angularTarget=<1 0 0>}
      localJointOffset=<4 0 0>
     }]
   }
   n94 = Anim Dynamics - Bone: dyn_L_underSleeve(ComponentPose=@n93, Alpha=1, ExternalForce="0,0,0") {
    angularDampingOverride=1
    angularSpringConstant=20000
    bAngularSpring=true
    bOverrideAngularDamping=true
    bUseGravityOverride=true
    gravityScale=0.35
    simulationSpace=RootRelative
    alphaBoolBlend {blendOption=Linear}
    boundBone {boneName=dyn_L_underSleeve}
    componentLinearAccScale=<1 1 1>
    componentLinearVelScale=<1 1 1>
    gravityOverride {z=-980}
    physicsBodyDefinitions [{
      sphereCollisionRadius=0
      boundBone {boneName=dyn_L_underSleeve}
      boxExtents=<20 30 30>
      constraintSetup {angularLimitsMax=<15 15 15>, angularLimitsMin=<-15 -15 -15>, angularTarget=<1 0 0>}
      localJointOffset=<4 0 0>
     }]
   }
   n95 = "Transform (Modify) Bone - Bone: dyn_L_armClothFix"(ComponentPose=@n94, Alpha=Get L_LowerArmAlpha() -> (L_LowerArmAlpha:double), Translation="0,-1.2,0", Rotation="0,-80,0", Scale="1,1,1") {
    rotationMode=BMM_Additive
    rotationSpace=BCS_BoneSpace
    translationMode=BMM_Additive
    alphaBoolBlend {blendOption=Linear}
    boneToModify {boneName=dyn_L_armClothFix}
    rotation {yaw=-80}
    translation {y=-1.2}
   }
   n96 = "Transform (Modify) Bone - Bone: dyn_R_armClothFix"(ComponentPose=@n95, Alpha=Get R_LowerArmAlpha() -> (R_LowerArmAlpha:double), Translation="0,0,0", Rotation="0,-80,0", Scale="1,1,1") {
    rotationMode=BMM_Additive
    rotationSpace=BCS_BoneSpace
    translationMode=BMM_Additive
    alphaBoolBlend {blendOption=Linear}
    boneToModify {boneName=dyn_R_armClothFix}
    rotation {yaw=-80}
   }
   n97 = Get isCrouching()
   n98 = "Transform (Modify) Bone - Bone: dyn_L_bowStrand_1"!(ComponentPose=@n96, Alpha=@n97, Translation="0,0,0", Scale="1,1,1") -> (Pose) {
    rotationMode=BMM_Additive
    alphaBoolBlend {blendOption=Linear}
    alphaScaleBiasClamp {bInterpResult=true, interpSpeedDecreasing=5, interpSpeedIncreasing=5}
    boneToModify {boneName=dyn_L_bowStrand_1}
    rotation {roll=25}
   }
   n99 = "Transform (Modify) Bone - Bone: dyn_L_bowStrand_2"!(ComponentPose=@n98, Alpha=@n97, Translation="0,0,0", Scale="1,1,1") -> (Pose) {
    rotationMode=BMM_Additive
    alphaBoolBlend {blendOption=Linear}
    alphaScaleBiasClamp {bInterpResult=true, interpSpeedDecreasing=5, interpSpeedIncreasing=5}
    boneToModify {boneName=dyn_L_bowStrand_2}
    rotation {roll=25}
   }
   n100 = "Transform (Modify) Bone - Bone: dyn_L_bowStrand_3"!(ComponentPose=@n99, Alpha=@n97, Translation="0,0,0", Scale="1,1,1") -> (Pose) {
    rotationMode=BMM_Additive
    alphaBoolBlend {blendOption=Linear}
    alphaScaleBiasClamp {bInterpResult=true, interpSpeedDecreasing=5, interpSpeedIncreasing=5}
    boneToModify {boneName=dyn_L_bowStrand_3}
    rotation {roll=20}
   }
   n101 = "Transform (Modify) Bone - Bone: dyn_L_bowStrand_4"!(ComponentPose=@n100, Alpha=@n97, Translation="0,0,0", Scale="1,1,1") -> (Pose) {
    rotationMode=BMM_Additive
    alphaBoolBlend {blendOption=Linear}
    alphaScaleBiasClamp {bInterpResult=true, interpSpeedDecreasing=5, interpSpeedIncreasing=5}
    boneToModify {boneName=dyn_L_bowStrand_4}
    rotation {roll=15}
   }
   n102 = "Transform (Modify) Bone - Bone: dyn_R_bowStrand_1"!(ComponentPose=@n101, Alpha=@n97, Translation="0,0,0", Scale="1,1,1") -> (Pose) {
    rotationMode=BMM_Additive
    alphaBoolBlend {blendOption=Linear}
    alphaScaleBiasClamp {bInterpResult=true, interpSpeedDecreasing=5, interpSpeedIncreasing=5}
    boneToModify {boneName=dyn_R_bowStrand_1}
    rotation {roll=25}
   }
   n103 = "Transform (Modify) Bone - Bone: dyn_R_bowStrand_2"!(ComponentPose=@n102, Alpha=@n97, Translation="0,0,0", Scale="1,1,1") -> (Pose) {
    rotationMode=BMM_Additive
    alphaBoolBlend {blendOption=Linear}
    alphaScaleBiasClamp {bInterpResult=true, interpSpeedDecreasing=5, interpSpeedIncreasing=5}
    boneToModify {boneName=dyn_R_bowStrand_2}
    rotation {roll=25}
   }
   n104 = "Transform (Modify) Bone - Bone: dyn_R_bowStrand_3"!(ComponentPose=@n103, Alpha=@n97, Translation="0,0,0", Scale="1,1,1") -> (Pose) {
    rotationMode=BMM_Additive
    alphaBoolBlend {blendOption=Linear}
    alphaScaleBiasClamp {bInterpResult=true, interpSpeedDecreasing=5, interpSpeedIncreasing=5}
    boneToModify {boneName=dyn_R_bowStrand_3}
    rotation {roll=20}
   }
   n105 = "Transform (Modify) Bone - Bone: dyn_R_bowStrand_4"!(ComponentPose=@n104, Alpha=@n97, Translation="0,0,0", Scale="1,1,1") -> (Pose) {
    rotationMode=BMM_Additive
    alphaBoolBlend {blendOption=Linear}
    alphaScaleBiasClamp {bInterpResult=true, interpSpeedDecreasing=5, interpSpeedIncreasing=5}
    boneToModify {boneName=dyn_R_bowStrand_4}
    rotation {roll=15}
   }
   n106 = "Constraint - dyn_center_upperSkirt_2 (PPPPPPPPPPPP)"!(
    ComponentPose=@n105
    Alpha=1
    ConstraintWeights_0=0.4683
    ConstraintWeights_1:float=0.0085
    ConstraintWeights_2:float=0.012
    ConstraintWeights_3:float=0.0368
    ConstraintWeights_4:float=0.005
    ConstraintWeights_5:float=0.007
    ConstraintWeights_6:float=0.0258
    ConstraintWeights_7:float=0.018
    ConstraintWeights_8:float=0.0851
    ConstraintWeights_9:float=0.214
    ConstraintWeights_10:float=0.0223
    ConstraintWeights_11:float=0.0972
   ) -> (Pose) {
    alphaBoolBlend {blendOption=Linear}
    boneToModify {boneName=dyn_center_upperSkirt_2}
    constraintSetup [
     {targetBone {boneName=pelvis}}
     {targetBone {boneName=dyn_R_front_LowerSkirt_1}}
     {targetBone {boneName=dyn_R_front_LowerSkirt_2}}
     {targetBone {boneName=dyn_R_front_LowerSkirt_3}}
     {targetBone {boneName=dyn_L_front_LowerSkirt_1}}
     {targetBone {boneName=dyn_L_front_LowerSkirt_2}}
     {targetBone {boneName=dyn_L_front_LowerSkirt_3}}
     {targetBone {boneName=dyn_center_lowerSkirt_1}}
     {targetBone {boneName=dyn_center_lowerSkirt_2}}
     {targetBone {boneName=dyn_center_lowerSkirt_3}}
     {targetBone {boneName=thigh_r}}
     {targetBone {boneName=thigh_l}}
    ]
    constraintWeights [1,1,1,1,1,1,1,1,1,1,1,1]
   }
   n107 = "Transform (Modify) Bone - Bone: dyn_backpack_socket"(ComponentPose=@n106, Alpha=1, Translation="0,0,0", Rotation="0,0,10", Scale="1,1,1") {rotationMode=BMM_Additive, alphaBoolBlend {blendOption=Linear}, boneToModify {boneName=dyn_backpack_socket}, rotation {roll=10}}
   n108 = "Transform (Modify) Bone - Bone: dyn_backpack_socket"(ComponentPose=@n107, Alpha=Get rotatePackOnSkydive() -> (rotatePackOnSkydive:double), Translation="0,0,0", Rotation="0,0,17", Scale="1,1,1") {rotationMode=BMM_Additive, alphaBoolBlend {blendOption=Linear}, boneToModify {boneName=dyn_backpack_socket}, rotation {roll=17}}
   n109 = "Transform (Modify) Bone - Bone: dyn_R_front_UpperSkirt_1"(ComponentPose=@n108, Alpha=Get isJumping() -> (isJumping:double), Translation="0,2,-2", Rotation="5,0,30", Scale="1,1,1") {
    rotationMode=BMM_Additive
    translationMode=BMM_Additive
    alphaBoolBlend {blendOption=Linear}
    alphaScaleBiasClamp {bInterpResult=true}
    boneToModify {boneName=dyn_R_front_UpperSkirt_1}
    rotation {pitch=5, roll=30}
    translation {y=2, z=-2}
   }
   n110 = Anim Dynamics - Bone: dyn_necklace(ComponentPose=@n109, Alpha=1, ExternalForce="0,0,0") {
    angularDampingOverride=1
    angularSpringConstant=20000
    bAngularSpring=true
    bLinearSpring=true
    bOverrideAngularDamping=true
    bOverrideLinearDamping=true
    bUseGravityOverride=true
    gravityScale=0.35
    linearDampingOverride=1
    linearSpringConstant=10
    alphaBoolBlend {blendOption=Linear}
    alphaScaleBiasClamp {scale=0.7}
    boundBone {boneName=dyn_necklace}
    componentLinearAccScale=<0.4 0.4 0.4>
    componentLinearVelScale=<1.5 1.5 1.5>
    gravityOverride {z=-980}
    physicsBodyDefinitions [{
      sphereCollisionRadius=0
      boundBone {boneName=dyn_necklace}
      boxExtents=<100 100 100>
      constraintSetup {angularLimitsMax=<11 11 11>, angularLimitsMin=<-11 -11 -11>, angularTarget=<1 0 0>, linearAxesMax=<4 0.5 0.5>, linearAxesMin=<-1 -0.5 -0.5>}
      localJointOffset=<4 0 0>
     }]
    relativeSpaceBone {boneName=spine_03}
   }
   n111 = "Transform (Modify) Bone - Bone: dyn_necklace"(
    ComponentPose=@n110
    Alpha=Get turnOffNecklace() -> (turnOffNecklace:double)
    Translation=Get necklaceFloater() -> (necklaceFloater:Vector)
    Rotation="0,0,0"
    Scale="1,1,1"
   ) {
    rotationMode=BMM_Additive
    rotationSpace=BCS_BoneSpace
    translationMode=BMM_Additive
    alphaBoolBlend {blendOption=Linear}
    alphaScaleBiasClamp {bInterpResult=true, interpSpeedDecreasing=3, interpSpeedIncreasing=3}
    boneToModify {boneName=dyn_necklace}
   }
   n112 = "Transform (Modify) Bone - Bone: dyn_necklace"(ComponentPose=@n111, Alpha=Get armsUpAlpha() -> (armsUpAlpha:double), Translation="5,0,0", Rotation="0,0,0", Scale="1,1,1") {
    rotationMode=BMM_Additive
    rotationSpace=BCS_BoneSpace
    translationMode=BMM_Additive
    translationSpace=BCS_BoneSpace
    alphaBoolBlend {blendOption=Linear}
    alphaScaleBiasClamp {bInterpResult=true, interpSpeedDecreasing=3, interpSpeedIncreasing=3}
    boneToModify {boneName=dyn_necklace}
    translation {x=5}
   }
   n113 = Anim Dynamics - Bone: dyn_necklaceBall_1(ComponentPose=@n112, Alpha=1, ExternalForce="0,0,0") {
    angularDampingOverride=1
    angularSpringConstant=20000
    bAngularSpring=true
    bOverrideAngularDamping=true
    bUseGravityOverride=true
    gravityScale=0.35
    simulationSpace=RootRelative
    alphaBoolBlend {blendOption=Linear}
    boundBone {boneName=dyn_necklaceBall_1}
    componentLinearAccScale=<1 1 1>
    componentLinearVelScale=<1 1 1>
    gravityOverride {z=-980}
    physicsBodyDefinitions [{
      sphereCollisionRadius=0
      boundBone {boneName=dyn_necklaceBall_1}
      boxExtents=<20 30 30>
      constraintSetup {angularLimitsMax=<15 15 15>, angularLimitsMin=<-15 -15 -15>, angularTarget=<1 0 0>}
      localJointOffset=<4 0 0>
     }]
   }
   n114 = Anim Dynamics - Bone: dyn_necklaceBall_2(ComponentPose=@n113, Alpha=1, ExternalForce="0,0,0") {
    angularDampingOverride=1
    angularSpringConstant=20000
    bAngularSpring=true
    bOverrideAngularDamping=true
    bUseGravityOverride=true
    gravityScale=0.35
    simulationSpace=RootRelative
    alphaBoolBlend {blendOption=Linear}
    boundBone {boneName=dyn_necklaceBall_2}
    componentLinearAccScale=<1 1 1>
    componentLinearVelScale=<1 1 1>
    gravityOverride {z=-980}
    physicsBodyDefinitions [{
      sphereCollisionRadius=0
      boundBone {boneName=dyn_necklaceBall_2}
      boxExtents=<20 30 30>
      constraintSetup {angularLimitsMax=<15 15 15>, angularLimitsMin=<-15 -15 -15>, angularTarget=<1 0 0>}
      localJointOffset=<4 0 0>
     }]
   }
   n115 = Anim Dynamics - Bone: dyn_necklaceBall_3(ComponentPose=@n114, Alpha=1, ExternalForce="0,0,0") {
    angularDampingOverride=1
    angularSpringConstant=20000
    bAngularSpring=true
    bOverrideAngularDamping=true
    bUseGravityOverride=true
    gravityScale=0.35
    simulationSpace=RootRelative
    alphaBoolBlend {blendOption=Linear}
    boundBone {boneName=dyn_necklaceBall_3}
    componentLinearAccScale=<1 1 1>
    componentLinearVelScale=<1 1 1>
    gravityOverride {z=-980}
    physicsBodyDefinitions [{
      sphereCollisionRadius=0
      boundBone {boneName=dyn_necklaceBall_3}
      boxExtents=<20 30 30>
      constraintSetup {angularLimitsMax=<15 15 15>, angularLimitsMin=<-15 -15 -15>, angularTarget=<1 0 0>}
      localJointOffset=<4 0 0>
     }]
   }
   n116 = Anim Dynamics - Bone: dyn_necklaceBall_4(ComponentPose=@n115, Alpha=1, ExternalForce="0,0,0") {
    angularDampingOverride=1
    angularSpringConstant=20000
    bAngularSpring=true
    bOverrideAngularDamping=true
    bUseGravityOverride=true
    gravityScale=0.35
    simulationSpace=RootRelative
    alphaBoolBlend {blendOption=Linear}
    boundBone {boneName=dyn_necklaceBall_4}
    componentLinearAccScale=<1 1 1>
    componentLinearVelScale=<1 1 1>
    gravityOverride {z=-980}
    physicsBodyDefinitions [{
      sphereCollisionRadius=0
      boundBone {boneName=dyn_necklaceBall_4}
      boxExtents=<20 30 30>
      constraintSetup {angularLimitsMax=<15 15 15>, angularLimitsMin=<-15 -15 -15>, angularTarget=<1 0 0>}
      localJointOffset=<4 0 0>
     }]
   }
   n117 = Anim Dynamics - Bone: dyn_necklaceBall_5(ComponentPose=@n116, Alpha=1, ExternalForce="0,0,0") {
    angularDampingOverride=1
    angularSpringConstant=20000
    bAngularSpring=true
    bOverrideAngularDamping=true
    bUseGravityOverride=true
    gravityScale=0.35
    simulationSpace=RootRelative
    alphaBoolBlend {blendOption=Linear}
    boundBone {boneName=dyn_necklaceBall_5}
    componentLinearAccScale=<1 1 1>
    componentLinearVelScale=<1 1 1>
    gravityOverride {z=-980}
    physicsBodyDefinitions [{
      sphereCollisionRadius=0
      boundBone {boneName=dyn_necklaceBall_5}
      boxExtents=<20 30 30>
      constraintSetup {angularLimitsMax=<15 15 15>, angularLimitsMin=<-15 -15 -15>, angularTarget=<1 0 0>}
      localJointOffset=<4 0 0>
     }]
   }
   n118 = Get sineRotation1()
   n119 = "Transform (Modify) Bone - Bone: dyn_necklaceBall_1"(ComponentPose=@n117, Alpha=1, Translation="0,0,0", Rotation=@n118, Scale="1,1,1") {
    rotationMode=BMM_Additive
    rotationSpace=BCS_BoneSpace
    translationMode=BMM_Additive
    alphaBoolBlend {blendOption=Linear}
    boneToModify {boneName=dyn_necklaceBall_1}
   }
   n120 = Get sineRotation2()
   n121 = "Transform (Modify) Bone - Bone: dyn_necklaceBall_2"(ComponentPose=@n119, Alpha=1, Translation="0,0,0", Rotation=@n120, Scale="1,1,1") {
    rotationMode=BMM_Additive
    rotationSpace=BCS_BoneSpace
    translationMode=BMM_Additive
    alphaBoolBlend {blendOption=Linear}
    boneToModify {boneName=dyn_necklaceBall_2}
   }
   n122 = "Transform (Modify) Bone - Bone: dyn_necklaceBall_3"(ComponentPose=@n121, Alpha=1, Translation="0,0,0", Rotation=@n118, Scale="1,1,1") {
    rotationMode=BMM_Additive
    rotationSpace=BCS_BoneSpace
    translationMode=BMM_Additive
    alphaBoolBlend {blendOption=Linear}
    boneToModify {boneName=dyn_necklaceBall_3}
   }
   n123 = "Transform (Modify) Bone - Bone: dyn_necklaceBall_4"(ComponentPose=@n122, Alpha=1, Translation="0,0,0", Rotation=@n120, Scale="1,1,1") {
    rotationMode=BMM_Additive
    rotationSpace=BCS_BoneSpace
    translationMode=BMM_Additive
    alphaBoolBlend {blendOption=Linear}
    boneToModify {boneName=dyn_necklaceBall_4}
   }
   n124 = "Transform (Modify) Bone - Bone: dyn_necklaceBall_5"(ComponentPose=@n123, Alpha=1, Translation="0,0,0", Rotation=@n118, Scale="1,1,1") {
    rotationMode=BMM_Additive
    rotationSpace=BCS_BoneSpace
    translationMode=BMM_Additive
    alphaBoolBlend {blendOption=Linear}
    boneToModify {boneName=dyn_necklaceBall_5}
   }
   n125 = Get FrontEnd_Idle_On_float()
   n126 = "Transform (Modify) Bone - Bone: dyn_L_back_UpperSkirt_master"(ComponentPose=@n124, Alpha=@n125, Translation="0,0,0", Rotation="0,-10,0", Scale="1,1,1") {
    rotationMode=BMM_Additive
    rotationSpace=BCS_BoneSpace
    translationMode=BMM_Additive
    alphaBoolBlend {blendOption=Linear}
    boneToModify {boneName=dyn_L_back_UpperSkirt_master}
    rotation {yaw=-10}
   }
   n127 = "Transform (Modify) Bone - Bone: dyn_R_back_UpperSkirt_master"(ComponentPose=@n126, Alpha=@n125, Translation="0,0,0", Rotation="0,10,0", Scale="1,1,1") {
    rotationMode=BMM_Additive
    rotationSpace=BCS_BoneSpace
    translationMode=BMM_Additive
    alphaBoolBlend {blendOption=Linear}
    boneToModify {boneName=dyn_R_back_UpperSkirt_master}
    rotation {yaw=10}
   }
   n128 = "Transform (Modify) Bone - Bone: dyn_L_side_LowerSkirt_master"(ComponentPose=@n127, Alpha=@n125, Translation="0,0,0", Rotation="0,-20,0", Scale="1,1,1") {
    rotationMode=BMM_Additive
    rotationSpace=BCS_BoneSpace
    translationMode=BMM_Additive
    alphaBoolBlend {blendOption=Linear}
    boneToModify {boneName=dyn_L_side_LowerSkirt_master}
    rotation {yaw=-20}
   }
   Save cached pose 'preRibbon'(Pose=Component To Local(ComponentPose=@n128)) {cacheName=preRibbon}
   Get spinNecklace() -> (spinNecklace:Rotator)
  ]}
 EventGraph {nodes [
   Event BlueprintUpdateAnimation()
   > n1 = Sequence!(execute) -> (then_0, then_1, then_2:Exec, then_3:Exec)
   > n2 = Gravity Override State Function DEPRECATED(CharacterPartAnimInstance=Self-Reference(), GameTimeInSeconds=GetGameTimeInSeconds(), GravityOverrideParams=Get Gravity Override Params())
   > Set gravityOverride(gravityOverride:Vector=@n2) -> (Output_Get:Vector)
   n3 = Get Is Wearing Backbling()
   n4 = Get bIsSkyDiving()
   Set rotatePackOnSkydive(execute=@n1.then_1, rotatePackOnSkydive:double="To Float (Boolean)"(InBool=AND Boolean(A=@n4, B=@n3))) -> (Output_Get:double)
   n5 = Get bIsCrouching()
   > Set isCrouching(isCrouching:double="To Float (Boolean)"(InBool=@n5)) -> (Output_Get:double)
   n6 = Get bIsAccelerating2D()
   > Set isCrouchMoving(isCrouchMoving:double="To Float (Boolean)"(InBool=AND Boolean(A=@n5, B=@n6))) -> (Output_Get:double)
   > Set isJumping(isJumping:double="To Float (Boolean)"(InBool=Get bIsJumping())) -> (Output_Get:double)
   > Set toneDownRibbonAlpha(toneDownRibbonAlpha:double=SelectFloat(A=1, B=0.3, bPickA=Get bIsAccelerating2D())) -> (Output_Get:double)
   > Set toneDownRibbonAlpha_capeOn(toneDownRibbonAlpha_capeOn:double=SelectFloat(A=0.5, B=0.3, bPickA=Get bIsAccelerating2D())) -> (Output_Get:double)
   n7 = GetRealTimeSeconds()
   n8 = "Sin (Radians)"(A="Absolute (Float)"(A=float * float(A=@n7, B=2)))
   > Set sineRotation1(sineRotation1:Rotator=MakeRotator(Pitch=float * float(A=@n8, B=5))) -> (Output_Get:Rotator)
   > Set sineRotation2(sineRotation2:Rotator=MakeRotator(Pitch=float * float(A=@n8, B=-5))) -> (Output_Get:Rotator)
   n9 = "Sin (Radians)"(A=@n7)
   > Set sineRotation3(sineRotation3:Rotator=MakeRotator(Pitch=float * float(A=@n9, B=25))) -> (Output_Get:Rotator)
   n10 = "Sin (Radians)"(A=float + float(A=@n7, B=-1))
   > Set sineRotation4(sineRotation4:Rotator=MakeRotator(Pitch=float * float(A=@n10, B=27))) -> (Output_Get:Rotator)
   n11 = "Sin (Radians)"(A=float + float(A=@n7, B=-1.5))
   > Set sineRotation5(sineRotation5:Rotator=MakeRotator(Pitch=float * float(A=@n11, B=30))) -> (Output_Get:Rotator)
   n12 = "Sin (Radians)"(A=float + float(A=@n7, B=-2))
   > Set sineRotation6(sineRotation6:Rotator=MakeRotator(Pitch=float * float(A=@n12, B=30))) -> (Output_Get:Rotator)
   n13 = "Sin (Radians)"(A=float + float(A=@n7, B=-2.5))
   > Set sineRotation7(sineRotation7:Rotator=MakeRotator(Pitch=float * float(A=@n13, B=30))) -> (Output_Get:Rotator)
   n14 = float / float(A=GetRealTimeSeconds(), B=50) {is=title K2Node_CallFunction}
   > Set spinNecklace(spinNecklace:Rotator=MakeRotator(Roll=float * float(A=@n14, B=500))) -> (Output_Get:Rotator)
   n15 = GetRealTimeSeconds()
   n16 = float / float(A="Sin (Radians)"(A=float * float(A=@n15, B=1.25)), B=1.5) {is=title K2Node_PromotableOperator}
   > Set necklaceFloater(necklaceFloater:Vector=MakeVector(Y=0, Z=float + float(A=@n16, B=3))) -> (Output_Get:Vector)
   n17 = GetOwnerSkeletalMeshComponent()
   n18 = BreakTransform(InTransform=GetDeltaTransformFromRefPose(self=@n17, BoneName=foot_l, BaseName=calf_l))
   > Set L footAlpha(L footAlpha:double=MapRangeClamped(Value=BreakRotator(InRot=@n18.Rotation).Yaw, InRangeA=-5, InRangeB=-30)) -> (Output_Get:double)
   n19 = BreakTransform(InTransform=GetDeltaTransformFromRefPose(self=@n17, BoneName=foot_r, BaseName=calf_r))
   > Set R footAlpha(R footAlpha:double=MapRangeClamped(Value=BreakRotator(InRot=@n19.Rotation).Yaw, InRangeA=-5, InRangeB=-30)) -> (Output_Get:double)
   n20 = BreakTransform(InTransform=GetDeltaTransformFromRefPose(self=@n17, BoneName=lowerarm_l, BaseName=upperarm_l))
   > Set L_LowerArmAlpha(L_LowerArmAlpha:double=MapRangeClamped(Value=BreakRotator(InRot=@n20.Rotation).Yaw, InRangeA=45, InRangeB=80)) -> (Output_Get:double)
   n21 = BreakTransform(InTransform=GetDeltaTransformFromRefPose(self=@n17, BoneName=lowerarm_r, BaseName=upperarm_r))
   > Set R_LowerArmAlpha(R_LowerArmAlpha:double=MapRangeClamped(Value=BreakRotator(InRot=@n21.Rotation).Yaw, InRangeA=45, InRangeB=80)) -> (Output_Get:double)
   n22 = BreakTransform(InTransform=GetDeltaTransformFromRefPose(self=@n17, BoneName=clavicle_l, BaseName=upperarm_l))
   > Set L_UpperArmAlpha(L_UpperArmAlpha:double=MapRangeClamped(Value=BreakRotator(InRot=@n22.Rotation).Yaw, InRangeA=-20, InRangeB=0)) -> (Output_Get:double)
   n23 = BreakTransform(InTransform=GetDeltaTransformFromRefPose(self=@n17, BoneName=clavicle_r, BaseName=upperarm_r))
   > Set R_UpperArmAlpha(R_UpperArmAlpha:double=MapRangeClamped(Value=BreakRotator(InRot=@n23.Rotation).Yaw, InRangeA=-20, InRangeB=0)) -> (Output_Get:double)
   n24 = Get bIsAccelerating2D()
   n25 = Get bIsJumping()
   > Set isAccelerating2D(isAccelerating2D:double="To Float (Boolean)"(InBool=OR Boolean(A=@n24, B=@n25))) -> (Output_Get:double)
   n26 = Get L_UpperArmAlpha() -> (L_UpperArmAlpha:double)
   n27 = Get R_UpperArmAlpha() -> (R_UpperArmAlpha:double)
   > Set armsUpAlpha(armsUpAlpha:double=float / float(A=float + float(A=@n26, B=@n27), B=3)) -> (Output_Get:double)
   > n28 = Branch(Condition="float >= float"(A=Get armsUpAlpha(), B=0.4))
   > Set turnOffNecklace(turnOffNecklace:double=0) -> (Output_Get:double)
   Set turnOffNecklace(execute=@n28.else, turnOffNecklace=1)
   n29 = Get Anim_Inst() -> (Anim_Inst:Anim Instance Object Reference)
   Is Valid(exec=@n1.then_2, InputObject=@n29)
   > Branch(Condition=Get bIsInFrontEnd())
   n30 = Get bIsPlayingEmote()
   > n31 = Set FrontEnd_Idle_On(FrontEnd_Idle_On:bool=AND Boolean(
     A=Montage_IsPlaying(self=@n29, Montage=/BRCosmetics/Animation/Game/MainPlayer/Menu/BR/CustomSkins/_s19/ThriveF/Thrive_Female_ALT_Idle_Montage)
     B=NOT Boolean(A=@n30)
    )) -> (Output_Get:bool)
   > Set FrontEnd_Idle_On_float(FrontEnd_Idle_On_float:double="To Float (Boolean)"(InBool=@n31)) -> (Output_Get:double)
   n32 = Set Skel_Mesh(execute=@n1.then_3, Skel_Mesh=GetOwnerSkeletalMeshComponent())
   > Set Anim_Inst(Anim_Inst:Anim Instance Object Reference=GetAnimInstance(self=@n32)) -> (Output_Get:Anim Instance Object Reference)
   Event OnCharacterCustomizationLoaded()
   n33 = CheckForBackbling()
   > Set Is Wearing Backbling(Is Wearing Backbling=@n33)
   > Set Is Backbling A Cape(Is Backbling A Cape:bool=@n33.bIsBackblingACape) -> (Output_Get:bool)
   > n34 = Branch(Condition=Get Is Backbling A Cape())
   > Set turnOffSine(turnOffSine:double=0) -> (Output_Get:double)
   Set turnOffSine(execute=@n34.else, turnOffSine=1)
   Event BlueprintInitializeAnimation()
  ]}
]
variables [
 Anim_Inst {type=AnimInstance, default=null}
 armsUpAlpha {type=number}
 CtrlRigAlpha {type=number}
 FrontEnd_Idle_On {type=boolean}
 FrontEnd_Idle_On_float {type=number}
 Gravity Override Params {type=GravityOverrideParamsStruct, default {onGround_Moving {gravityOverride=<0 0 -800>}, onGround_Standing {gravityOverride=<0 0 -800>}}}
 gravityOverride {type=Vector}
 Is Backbling A Cape {type=boolean}
 Is Backbling Long {type=boolean}
 Is Wearing Backbling {type=boolean}
 isAccelerating2D {type=number}
 isCrouching {type=number}
 isCrouchMoving {type=number}
 isJumping {type=number}
 L footAlpha {type=number}
 L_LowerArmAlpha {type=number}
 L_UpperArmAlpha {type=number}
 necklaceFloater {type=Vector}
 R footAlpha {type=number}
 R_LowerArmAlpha {type=number}
 R_thighAlpha {type=number}
 R_UpperArmAlpha {type=number}
 RBANalpha {type=number}
 ribbonWind {type=Rotator}
 rotatePackOnSkydive {type=number}
 sineRotation1 {type=Rotator}
 sineRotation2 {type=Rotator}
 sineRotation3 {type=Rotator}
 sineRotation4 {type=Rotator}
 sineRotation5 {type=Rotator}
 sineRotation6 {type=Rotator}
 sineRotation7 {type=Rotator}
 Skel_Mesh {type=SkeletalMeshComponent}
 spinNecklace {type=Rotator}
 toneDownRibbonAlpha {type=number}
 toneDownRibbonAlpha_capeOn {type=number}
 turnOffNecklace {type=number}
 turnOffSine {type=number}
]
