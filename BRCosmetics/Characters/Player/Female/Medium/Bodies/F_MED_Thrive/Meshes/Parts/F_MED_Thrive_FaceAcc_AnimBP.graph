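# Post-process-style AnimBlueprint for the Thrive female face accessory mesh, animating earrings, crown pieces, hair and strands.
# The AnimGraph copies the parent mesh pose, then layers additive sine-driven bone offsets with Anim Dynamics springs, followed by extra rotation offsets on some strands.
# The EventGraph computes phase-offset sine values each update for a travelling wave, and sets crown physics alpha lower while grounded.
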AnimBlueprint /BRCosmetics/Characters/Player/Female/Medium/Bodies/F_MED_Thrive/Meshes/Parts/F_MED_Thrive_FaceAcc_AnimBP
target_skeleton=&/F_MED_Thrive_FaceAcc_Skeleton
graphs [
 AnimGraph {nodes [
   n1 = Anim Dynamics - Bone: dyn_dyn_R_earring(ComponentPose=Local To Component(LocalPose=Copy Pose From Mesh() {bUseAttachedParent=true}), Alpha=1, ExternalForce="0,0,0") {
    angularDampingOverride=1
    angularSpringConstant=1000
    bAngularSpring=true
    bOverrideAngularDamping=true
    bUseGravityOverride=true
    gravityScale=0.35
    alphaBoolBlend {blendOption=Linear}
    boundBone {boneName=dyn_dyn_R_earring}
    componentLinearAccScale=<1 1 1>
    componentLinearVelScale=<1 1 1>
    gravityOverride {z=-980}
    physicsBodyDefinitions [{
      sphereCollisionRadius=0
      boundBone {boneName=dyn_dyn_R_earring}
      boxExtents=<30 40 40>
      constraintSetup {angularLimitsMax=<0 40 40>, angularLimitsMin=<0 -10 -40>, angularTarget=<1 0 0>}
      localJointOffset=<4 0 0>
     }]
    relativeSpaceBone {boneName=spine_02}
   }
   n2 = Anim Dynamics - Bone: dyn_dyn_L_earring(ComponentPose=@n1, Alpha=1, ExternalForce="0,0,0") {
    angularDampingOverride=1
    angularSpringConstant=1000
    bAngularSpring=true
    bOverrideAngularDamping=true
    bUseGravityOverride=true
    gravityScale=0.35
    alphaBoolBlend {blendOption=Linear}
    boundBone {boneName=dyn_dyn_L_earring}
    componentLinearAccScale=<1 1 1>
    componentLinearVelScale=<1 1 1>
    gravityOverride {z=-980}
    physicsBodyDefinitions [{
      sphereCollisionRadius=0
      boundBone {boneName=dyn_dyn_L_earring}
      boxExtents=<30 40 40>
      constraintSetup {angularLimitsMax=<0 10 40>, angularLimitsMin=<0 -40 -40>, angularTarget=<1 0 0>}
      localJointOffset=<-4 0 0>
     }]
    relativeSpaceBone {boneName=spine_02}
   }
   n3 = Get sineTranslation high1() -> (sineTranslation high1:Vector)
   n4 = "Transform (Modify) Bone - Bone: dyn_crown_L_2"(ComponentPose=@n2, Alpha=1, Translation=@n3, Rotation="0,0,0", Scale="1,1,1") {translationMode=BMM_Additive, translationSpace=BCS_BoneSpace, alphaBoolBlend {blendOption=Linear}, boneToModify {boneName=dyn_crown_L_2}}
   n5 = Get sineTranslation high2() -> (sineTranslation high2:Vector)
   n6 = "Transform (Modify) Bone - Bone: dyn_crown_L_1"(ComponentPose=@n4, Alpha=1, Translation=@n5, Rotation="0,0,0", Scale="1,1,1") {translationMode=BMM_Additive, translationSpace=BCS_BoneSpace, alphaBoolBlend {blendOption=Linear}, boneToModify {boneName=dyn_crown_L_1}}
   n7 = "Transform (Modify) Bone - Bone: dyn_crown_C"(ComponentPose=@n6, Alpha=1, Translation=Get sineTranslation high3() -> (sineTranslation high3:Vector), Rotation="0,0,0", Scale="1,1,1") {translationMode=BMM_Additive, translationSpace=BCS_BoneSpace, alphaBoolBlend {blendOption=Linear}, boneToModify {boneName=dyn_crown_C}}
   n8 = "Transform (Modify) Bone - Bone: dyn_crown_R_1"(ComponentPose=@n7, Alpha=1, Translation=@n5, Rotation="0,0,0", Scale="1,1,1") {translationMode=BMM_Additive, translationSpace=BCS_BoneSpace, alphaBoolBlend {blendOption=Linear}, boneToModify {boneName=dyn_crown_R_1}}
   n9 = "Transform (Modify) Bone - Bone: dyn_crown_R_2"(ComponentPose=@n8, Alpha=1, Translation=@n3, Rotation="0,0,0", Scale="1,1,1") {translationMode=BMM_Additive, translationSpace=BCS_BoneSpace, alphaBoolBlend {blendOption=Linear}, boneToModify {boneName=dyn_crown_R_2}}
   n10 = Get groundCrown() -> (groundCrown:double)
   n11 = Anim Dynamics - Bone: dyn_crown_C(ComponentPose=@n9, Alpha=@n10, ExternalForce="0,0,0") {
    angularDampingOverride=1
    angularSpringConstant=20000
    bAngularSpring=true
    bOverrideAngularDamping=true
    bOverrideLinearDamping=true
    bUseGravityOverride=true
    gravityScale=0.35
    linearDampingOverride=1
    alphaBoolBlend {blendOption=Linear}
    alphaScaleBiasClamp {bInterpResult=true}
    boundBone {boneName=dyn_crown_C}
    componentLinearAccScale=<0.3 0.3 0.3>
    componentLinearVelScale=<0.3 0.3 0.3>
    gravityOverride {z=-600}
    physicsBodyDefinitions [{
      sphereCollisionRadius=0
      boundBone {boneName=dyn_crown_C}
      boxExtents=<50 60 60>
      constraintSetup {angularLimitsMax=<0 0 0>, angularLimitsMin=<0 0 0>, angularTarget=<1 0 0>, linearAxesMax=<3 0 0>, linearAxesMin=<-2 0 0>}
      localJointOffset=<4 0 0>
     }]
    relativeSpaceBone {boneName=pelvis}
   }
   n12 = Anim Dynamics - Bone: dyn_crown_R_1(ComponentPose=@n11, Alpha=@n10, ExternalForce="0,0,0") {
    angularDampingOverride=1
    angularSpringConstant=20000
    bAngularSpring=true
    bOverrideAngularDamping=true
    bOverrideLinearDamping=true
    bUseGravityOverride=true
    gravityScale=0.35
    linearDampingOverride=1
    alphaBoolBlend {blendOption=Linear}
    alphaScaleBiasClamp {bInterpResult=true}
    boundBone {boneName=dyn_crown_R_1}
    componentLinearAccScale=<0.28 0.28 0.28>
    componentLinearVelScale=<0.28 0.28 0.28>
    gravityOverride {z=-600}
    physicsBodyDefinitions [{
      sphereCollisionRadius=0
      boundBone {boneName=dyn_crown_R_1}
      boxExtents=<50 60 60>
      constraintSetup {angularLimitsMax=<0 0 0>, angularLimitsMin=<0 0 0>, angularTarget=<1 0 0>, linearAxesMax=<3 0 0>, linearAxesMin=<-2 0 0>}
      localJointOffset=<4 0 0>
     }]
    relativeSpaceBone {boneName=pelvis}
   }
   n13 = Anim Dynamics - Bone: dyn_crown_R_2(ComponentPose=@n12, Alpha=@n10, ExternalForce="0,0,0") {
    angularDampingOverride=1
    angularSpringConstant=20000
    bAngularSpring=true
    bOverrideAngularDamping=true
    bOverrideLinearDamping=true
    bUseGravityOverride=true
    gravityScale=0.35
    linearDampingOverride=1
    alphaBoolBlend {blendOption=Linear}
    alphaScaleBiasClamp {bInterpResult=true}
    boundBone {boneName=dyn_crown_R_2}
    componentLinearAccScale=<0.25 0.25 0.25>
    componentLinearVelScale=<0.25 0.25 0.25>
    gravityOverride {z=-600}
    physicsBodyDefinitions [{
      sphereCollisionRadius=0
      boundBone {boneName=dyn_crown_R_2}
      boxExtents=<50 60 60>
      constraintSetup {angularLimitsMax=<0 0 0>, angularLimitsMin=<0 0 0>, angularTarget=<1 0 0>, linearAxesMax=<3 0 0>, linearAxesMin=<-2 0 0>}
      localJointOffset=<4 0 0>
     }]
    relativeSpaceBone {boneName=pelvis}
   }
   n14 = Anim Dynamics - Bone: dyn_crown_L_1(ComponentPose=@n13, Alpha=@n10, ExternalForce="0,0,0") {
    angularDampingOverride=1
    angularSpringConstant=20000
    bAngularSpring=true
    bOverrideAngularDamping=true
    bOverrideLinearDamping=true
    bUseGravityOverride=true
    gravityScale=0.35
    linearDampingOverride=1
    alphaBoolBlend {blendOption=Linear}
    alphaScaleBiasClamp {bInterpResult=true}
    boundBone {boneName=dyn_crown_L_1}
    componentLinearAccScale=<0.28 0.28 0.28>
    componentLinearVelScale=<0.28 0.28 0.28>
    gravityOverride {z=-600}
    physicsBodyDefinitions [{
      sphereCollisionRadius=0
      boundBone {boneName=dyn_crown_L_1}
      boxExtents=<50 60 60>
      constraintSetup {angularLimitsMax=<0 0 0>, angularLimitsMin=<0 0 0>, angularTarget=<1 0 0>, linearAxesMax=<3 0 0>, linearAxesMin=<-2 0 0>}
      localJointOffset=<4 0 0>
     }]
    relativeSpaceBone {boneName=pelvis}
   }
   n15 = Anim Dynamics - Bone: dyn_crown_L_2(ComponentPose=@n14, Alpha=@n10, ExternalForce="0,0,0") {
    angularDampingOverride=1
    angularSpringConstant=20000
    bAngularSpring=true
    bOverrideAngularDamping=true
    bOverrideLinearDamping=true
    bUseGravityOverride=true
    gravityScale=0.35
    linearDampingOverride=1
    alphaBoolBlend {blendOption=Linear}
    alphaScaleBiasClamp {bInterpResult=true}
    boundBone {boneName=dyn_crown_L_2}
    componentLinearAccScale=<0.25 0.25 0.25>
    componentLinearVelScale=<0.25 0.25 0.25>
    gravityOverride {z=-600}
    physicsBodyDefinitions [{
      sphereCollisionRadius=0
      boundBone {boneName=dyn_crown_L_2}
      boxExtents=<50 60 60>
      constraintSetup {angularLimitsMax=<0 0 0>, angularLimitsMin=<0 0 0>, angularTarget=<1 0 0>, linearAxesMax=<3 0 0>, linearAxesMin=<-2 0 0>}
      localJointOffset=<4 0 0>
     }]
    relativeSpaceBone {boneName=pelvis}
   }
   n16 = Get sineRotation1() -> (sineRotation1:Rotator)
   n17 = "Transform (Modify) Bone - Bone: dyn_main_hair_base"(ComponentPose=@n15, Alpha=1, Translation="0,0,0", Rotation=@n16, Scale="1,1,1") {rotationMode=BMM_Additive, alphaBoolBlend {blendOption=Linear}, alphaScaleBiasClamp {scale=0.5}, boneToModify {boneName=dyn_main_hair_base}}
   n18 = Get sineRotation2() -> (sineRotation2:Rotator)
   n19 = "Transform (Modify) Bone - Bone: dyn_main_hair_1"(ComponentPose=@n17, Alpha=1, Translation="0,0,0", Rotation=@n18, Scale="1,1,1") {rotationMode=BMM_Additive, alphaBoolBlend {blendOption=Linear}, alphaScaleBiasClamp {scale=0.5}, boneToModify {boneName=dyn_main_hair_1}}
   n20 = Get sineRotation3() -> (sineRotation3:Rotator)
   n21 = "Transform (Modify) Bone - Bone: dyn_main_hair_2"(ComponentPose=@n19, Alpha=1, Translation="0,0,0", Rotation=@n20, Scale="1,1,1") {rotationMode=BMM_Additive, alphaBoolBlend {blendOption=Linear}, alphaScaleBiasClamp {scale=0.5}, boneToModify {boneName=dyn_main_hair_2}}
   n22 = Get sineRotation4() -> (sineRotation4:Rotator)
   n23 = "Transform (Modify) Bone - Bone: dyn_main_hair_3"(ComponentPose=@n21, Alpha=1, Translation="0,0,0", Rotation=@n22, Scale="1,1,1") {rotationMode=BMM_Additive, alphaBoolBlend {blendOption=Linear}, alphaScaleBiasClamp {scale=0.5}, boneToModify {boneName=dyn_main_hair_3}}
   n24 = Get sineRotation5() -> (sineRotation5:Rotator)
   n25 = "Transform (Modify) Bone - Bone: dyn_main_hair_4"(ComponentPose=@n23, Alpha=1, Translation="0,0,0", Rotation=@n24, Scale="1,1,1") {rotationMode=BMM_Additive, alphaBoolBlend {blendOption=Linear}, alphaScaleBiasClamp {scale=0.5}, boneToModify {boneName=dyn_main_hair_4}}
   n26 = "Transform (Modify) Bone - Bone: dyn_strand_C_1"(
    ComponentPose="Transform (Modify) Bone - Bone: dyn_hair_offshoot_2"(
     ComponentPose="Transform (Modify) Bone - Bone: dyn_hair_offshoot_1"(ComponentPose=@n25, Alpha=1, Translation="0,0,0", Rotation=@n16, Scale="1,1,1") {rotationMode=BMM_Additive, alphaBoolBlend {blendOption=Linear}, boneToModify {boneName=dyn_hair_offshoot_1}}
     Alpha=1
     Translation="0,0,0"
     Rotation=@n20
     Scale="1,1,1"
    ) {rotationMode=BMM_Additive, alphaBoolBlend {blendOption=Linear}, boneToModify {boneName=dyn_hair_offshoot_2}}
    Alpha=1
    Translation="0,0,0"
    Rotation=@n16
    Scale="1,1,1"
   ) {rotationMode=BMM_Additive, alphaBoolBlend {blendOption=Linear}, alphaScaleBiasClamp {scale=0.2}, boneToModify {boneName=dyn_strand_C_1}}
   n27 = "Transform (Modify) Bone - Bone: dyn_strand_C_2"(ComponentPose=@n26, Alpha=1, Translation="0,0,0", Rotation=@n18, Scale="1,1,1") {rotationMode=BMM_Additive, alphaBoolBlend {blendOption=Linear}, alphaScaleBiasClamp {scale=0.2}, boneToModify {boneName=dyn_strand_C_2}}
   n28 = "Transform (Modify) Bone - Bone: dyn_strand_C_3"(ComponentPose=@n27, Alpha=1, Translation="0,0,0", Rotation=@n20, Scale="1,1,1") {rotationMode=BMM_Additive, alphaBoolBlend {blendOption=Linear}, alphaScaleBiasClamp {scale=0.5}, boneToModify {boneName=dyn_strand_C_3}}
   n29 = "Transform (Modify) Bone - Bone: dyn_strand_D_1"(
    ComponentPose="Transform (Modify) Bone - Bone: dyn_strand_C_5"(
     ComponentPose="Transform (Modify) Bone - Bone: dyn_strand_C_4"(ComponentPose=@n28, Alpha=1, Translation="0,0,0", Rotation=@n22, Scale="1,1,1") {rotationMode=BMM_Additive, alphaBoolBlend {blendOption=Linear}, boneToModify {boneName=dyn_strand_C_4}}
     Alpha=1
     Translation="0,0,0"
     Rotation=@n24
     Scale="1,1,1"
    ) {rotationMode=BMM_Additive, alphaBoolBlend {blendOption=Linear}, boneToModify {boneName=dyn_strand_C_5}}
    Alpha=1
    Translation="0,0,0"
    Rotation=@n16
    Scale="1,1,1"
   ) {rotationMode=BMM_Additive, alphaBoolBlend {blendOption=Linear}, alphaScaleBiasClamp {scale=0.2}, boneToModify {boneName=dyn_strand_D_1}}
   n30 = "Transform (Modify) Bone - Bone: dyn_strand_D_2"(ComponentPose=@n29, Alpha=1, Translation="0,0,0", Rotation=@n18, Scale="1,1,1") {rotationMode=BMM_Additive, alphaBoolBlend {blendOption=Linear}, alphaScaleBiasClamp {scale=0.2}, boneToModify {boneName=dyn_strand_D_2}}
   n31 = "Transform (Modify) Bone - Bone: dyn_strand_D_3"(ComponentPose=@n30, Alpha=1, Translation="0,0,0", Rotation=@n20, Scale="1,1,1") {rotationMode=BMM_Additive, alphaBoolBlend {blendOption=Linear}, alphaScaleBiasClamp {scale=0.2}, boneToModify {boneName=dyn_strand_D_3}}
   n32 = "Transform (Modify) Bone - Bone: dyn_strand_D_4"(ComponentPose=@n31, Alpha=1, Translation="0,0,0", Rotation=@n22, Scale="1,1,1") {rotationMode=BMM_Additive, alphaBoolBlend {blendOption=Linear}, alphaScaleBiasClamp {scale=0.2}, boneToModify {boneName=dyn_strand_D_4}}
   n33 = "Transform (Modify) Bone - Bone: dyn_strand_D_5"(ComponentPose=@n32, Alpha=1, Translation="0,0,0", Rotation=@n24, Scale="1,1,1") {rotationMode=BMM_Additive, alphaBoolBlend {blendOption=Linear}, alphaScaleBiasClamp {scale=0.2}, boneToModify {boneName=dyn_strand_D_5}}
   n34 = Get sineRotation high1() -> (sineRotation high1:Rotator)
   n35 = "Transform (Modify) Bone - Bone: dyn_strand_A_1"(ComponentPose=@n33, Alpha=1, Translation="0,0,0", Rotation=@n34, Scale="1,1,1") {rotationMode=BMM_Additive, alphaBoolBlend {blendOption=Linear}, boneToModify {boneName=dyn_strand_A_1}}
   n36 = Get sineRotation high2() -> (sineRotation high2:Rotator)
   n37 = Get sineRotation high3() -> (sineRotation high3:Rotator)
   n38 = "Transform (Modify) Bone - Bone: dyn_strand_B_1"(
    ComponentPose="Transform (Modify) Bone - Bone: dyn_strand_A_3"(
     ComponentPose="Transform (Modify) Bone - Bone: dyn_strand_A_2"(ComponentPose=@n35, Alpha=1, Translation="0,0,0", Rotation=@n36, Scale="1,1,1") {rotationMode=BMM_Additive, alphaBoolBlend {blendOption=Linear}, boneToModify {boneName=dyn_strand_A_2}}
     Alpha=1
     Translation="0,0,0"
     Rotation=@n37
     Scale="1,1,1"
    ) {rotationMode=BMM_Additive, alphaBoolBlend {blendOption=Linear}, boneToModify {boneName=dyn_strand_A_3}}
    Alpha=1
    Translation="0,0,0"
    Rotation=@n34
    Scale="1,1,1"
   ) {rotationMode=BMM_Additive, rotationSpace=BCS_BoneSpace, alphaBoolBlend {blendOption=Linear}, boneToModify {boneName=dyn_strand_B_1}}
   n39 = "Transform (Modify) Bone - Bone: dyn_strand_B_2"(ComponentPose=@n38, Alpha=1, Translation="0,0,0", Rotation=@n36, Scale="1,1,1") {rotationMode=BMM_Additive, rotationSpace=BCS_BoneSpace, alphaBoolBlend {blendOption=Linear}, boneToModify {boneName=dyn_strand_B_2}}
   n40 = "Transform (Modify) Bone - Bone: dyn_strand_B_3"(ComponentPose=@n39, Alpha=1, Translation="0,0,0", Rotation=@n37, Scale="1,1,1") {rotationMode=BMM_Additive, rotationSpace=BCS_BoneSpace, alphaBoolBlend {blendOption=Linear}, boneToModify {boneName=dyn_strand_B_3}}
   n41 = Anim Dynamics - Bone: dyn_attached_strand_1(ComponentPose=@n40, Alpha=1, ExternalForce="0,0,0") {
    angularDampingOverride=0.999
    angularSpringConstant=20000
    bAngularSpring=true
    bOverrideAngularDamping=true
    bUseGravityOverride=true
    gravityScale=0.35
    alphaBoolBlend {blendOption=Linear}
    boundBone {boneName=dyn_attached_strand_1}
    componentLinearAccScale=<1 1 1>
    componentLinearVelScale=<1 1 1>
    gravityOverride {z=-980}
    physicsBodyDefinitions [{
      sphereCollisionRadius=0
      boundBone {boneName=dyn_attached_strand_1}
      boxExtents=<50 60 60>
      constraintSetup {angularLimitsMax=<15 30 30>, angularLimitsMin=<-15 -30 -30>, angularTarget=<1 0 0>}
      localJointOffset=<4 0 0>
     }]
    relativeSpaceBone {boneName=pelvis}
   }
   n42 = Anim Dynamics - Bone: dyn_attached_strand_2(ComponentPose=@n41, Alpha=1, ExternalForce="0,0,0") {
    angularDampingOverride=0.999
    angularSpringConstant=20000
    bAngularSpring=true
    bOverrideAngularDamping=true
    bUseGravityOverride=true
    gravityScale=0.35
    alphaBoolBlend {blendOption=Linear}
    boundBone {boneName=dyn_attached_strand_2}
    componentLinearAccScale=<1 1 1>
    componentLinearVelScale=<1 1 1>
    gravityOverride {z=-980}
    physicsBodyDefinitions [{
      sphereCollisionRadius=0
      boundBone {boneName=dyn_attached_strand_2}
      boxExtents=<50 60 60>
      constraintSetup {angularLimitsMax=<15 30 30>, angularLimitsMin=<-15 -30 -30>, angularTarget=<1 0 0>}
      localJointOffset=<4 0 0>
     }]
    relativeSpaceBone {boneName=pelvis}
   }
   n43 = Anim Dynamics - Bone: dyn_hair_offshoot_1(ComponentPose=@n42, Alpha=1, ExternalForce="0,0,0") {
    angularDampingOverride=0.999
    angularSpringConstant=20000
    bAngularSpring=true
    bOverrideAngularDamping=true
    bUseGravityOverride=true
    gravityScale=0.35
    alphaBoolBlend {blendOption=Linear}
    boundBone {boneName=dyn_hair_offshoot_1}
    componentLinearAccScale=<1 1 1>
    componentLinearVelScale=<1 1 1>
    gravityOverride {z=-980}
    physicsBodyDefinitions [{
      sphereCollisionRadius=0
      boundBone {boneName=dyn_hair_offshoot_1}
      boxExtents=<50 60 60>
      constraintSetup {angularLimitsMax=<15 30 30>, angularLimitsMin=<-15 -30 -30>, angularTarget=<1 0 0>}
      localJointOffset=<4 0 0>
     }]
    relativeSpaceBone {boneName=pelvis}
   }
   n44 = Anim Dynamics - Bone: dyn_hair_offshoot_2(ComponentPose=@n43, Alpha=1, ExternalForce="0,0,0") {
    angularDampingOverride=0.999
    angularSpringConstant=20000
    bAngularSpring=true
    bOverrideAngularDamping=true
    bUseGravityOverride=true
    gravityScale=0.35
    alphaBoolBlend {blendOption=Linear}
    boundBone {boneName=dyn_hair_offshoot_2}
    componentLinearAccScale=<1 1 1>
    componentLinearVelScale=<1 1 1>
    gravityOverride {z=-980}
    physicsBodyDefinitions [{
      sphereCollisionRadius=0
      boundBone {boneName=dyn_hair_offshoot_2}
      boxExtents=<50 60 60>
      constraintSetup {angularLimitsMax=<15 30 30>, angularLimitsMin=<-15 -30 -30>, angularTarget=<1 0 0>}
      localJointOffset=<4 0 0>
     }]
    relativeSpaceBone {boneName=pelvis}
   }
   n45 = Anim Dynamics - Bone: dyn_main_hair_base(ComponentPose=@n44, Alpha=1, ExternalForce="0,0,0") {
    angularDampingOverride=0.999
    angularSpringConstant=20000
    bAngularSpring=true
    bOverrideAngularDamping=true
    bUseGravityOverride=true
    gravityScale=0.35
    alphaBoolBlend {blendOption=Linear}
    boundBone {boneName=dyn_main_hair_base}
    componentLinearAccScale=<1 1 1>
    componentLinearVelScale=<1 1 1>
    gravityOverride {z=-980}
    physicsBodyDefinitions [{
      sphereCollisionRadius=0
      boundBone {boneName=dyn_main_hair_base}
      boxExtents=<50 60 60>
      constraintSetup {angularLimitsMax=<15 40 30>, angularLimitsMin=<-15 -30 -30>, angularTarget=<1 0 0>}
      localJointOffset=<4 0 0>
     }]
    relativeSpaceBone {boneName=pelvis}
   }
   n46 = Anim Dynamics - Bone: dyn_main_hair_1(ComponentPose=@n45, Alpha=1, ExternalForce="0,0,0") {
    angularDampingOverride=0.999
    angularSpringConstant=20000
    bAngularSpring=true
    bOverrideAngularDamping=true
    bUseGravityOverride=true
    gravityScale=0.35
    alphaBoolBlend {blendOption=Linear}
    boundBone {boneName=dyn_main_hair_1}
    componentLinearAccScale=<1 1 1>
    componentLinearVelScale=<1 1 1>
    gravityOverride {z=-980}
    physicsBodyDefinitions [{
      sphereCollisionRadius=0
      boundBone {boneName=dyn_main_hair_1}
      boxExtents=<50 60 60>
      constraintSetup {angularLimitsMax=<15 40 30>, angularLimitsMin=<-15 -30 -30>, angularTarget=<1 0 0>}
      localJointOffset=<4 0 0>
     }]
    relativeSpaceBone {boneName=pelvis}
   }
   n47 = Anim Dynamics - Bone: dyn_main_hair_2(ComponentPose=@n46, Alpha=1, ExternalForce="0,0,0") {
    angularDampingOverride=0.999
    angularSpringConstant=20000
    bAngularSpring=true
    bOverrideAngularDamping=true
    bUseGravityOverride=true
    gravityScale=0.35
    alphaBoolBlend {blendOption=Linear}
    boundBone {boneName=dyn_main_hair_2}
    componentLinearAccScale=<1 1 1>
    componentLinearVelScale=<1 1 1>
    gravityOverride {z=-980}
    physicsBodyDefinitions [{
      sphereCollisionRadius=0
      boundBone {boneName=dyn_main_hair_2}
      boxExtents=<50 60 60>
      constraintSetup {angularLimitsMax=<15 40 30>, angularLimitsMin=<-15 -30 -30>, angularTarget=<1 0 0>}
      localJointOffset=<4 0 0>
     }]
    relativeSpaceBone {boneName=pelvis}
   }
   n48 = Anim Dynamics - Bone: dyn_main_hair_3(ComponentPose=@n47, Alpha=1, ExternalForce="0,0,0") {
    angularDampingOverride=0.999
    angularSpringConstant=20000
    bAngularSpring=true
    bOverrideAngularDamping=true
    bUseGravityOverride=true
    gravityScale=0.35
    alphaBoolBlend {blendOption=Linear}
    boundBone {boneName=dyn_main_hair_3}
    componentLinearAccScale=<1 1 1>
    componentLinearVelScale=<1 1 1>
    gravityOverride {z=-980}
    physicsBodyDefinitions [{
      sphereCollisionRadius=0
      boundBone {boneName=dyn_main_hair_3}
      boxExtents=<50 60 60>
      constraintSetup {angularLimitsMax=<15 40 30>, angularLimitsMin=<-15 -30 -30>, angularTarget=<1 0 0>}
      localJointOffset=<4 0 0>
     }]
    relativeSpaceBone {boneName=pelvis}
   }
   n49 = Anim Dynamics - Bone: dyn_main_hair_4(ComponentPose=@n48, Alpha=1, ExternalForce="0,0,0") {
    angularDampingOverride=0.999
    angularSpringConstant=20000
    bAngularSpring=true
    bOverrideAngularDamping=true
    bUseGravityOverride=true
    gravityScale=0.35
    alphaBoolBlend {blendOption=Linear}
    boundBone {boneName=dyn_main_hair_4}
    componentLinearAccScale=<1 1 1>
    componentLinearVelScale=<1 1 1>
    gravityOverride {z=-980}
    physicsBodyDefinitions [{
      sphereCollisionRadius=0
      boundBone {boneName=dyn_main_hair_4}
      boxExtents=<50 60 60>
      constraintSetup {angularLimitsMax=<15 40 30>, angularLimitsMin=<-15 -30 -30>, angularTarget=<1 0 0>}
      localJointOffset=<4 0 0>
     }]
    relativeSpaceBone {boneName=pelvis}
   }
   n50 = Anim Dynamics - Bone: dyn_strand_A_1(ComponentPose=@n49, Alpha=1, ExternalForce="0,0,0") {
    angularDampingOverride=0.999
    angularSpringConstant=20000
    bAngularSpring=true
    bOverrideAngularDamping=true
    bUseGravityOverride=true
    gravityScale=0.35
    alphaBoolBlend {blendOption=Linear}
    boundBone {boneName=dyn_strand_A_1}
    componentLinearAccScale=<1 1 1>
    componentLinearVelScale=<1 1 1>
    gravityOverride {z=-980}
    physicsBodyDefinitions [{
      sphereCollisionRadius=0
      boundBone {boneName=dyn_strand_A_1}
      boxExtents=<50 60 60>
      constraintSetup {angularLimitsMax=<15 10 10>, angularLimitsMin=<-15 -10 -10>, angularTarget=<1 0 0>}
      localJointOffset=<4 0 0>
     }]
    relativeSpaceBone {boneName=pelvis}
   }
   n51 = Anim Dynamics - Bone: dyn_strand_A_2(ComponentPose=@n50, Alpha=1, ExternalForce="0,0,0") {
    angularDampingOverride=0.999
    angularSpringConstant=20000
    bAngularSpring=true
    bOverrideAngularDamping=true
    bUseGravityOverride=true
    gravityScale=0.35
    alphaBoolBlend {blendOption=Linear}
    boundBone {boneName=dyn_strand_A_2}
    componentLinearAccScale=<1 1 1>
    componentLinearVelScale=<1 1 1>
    gravityOverride {z=-980}
    physicsBodyDefinitions [{
      sphereCollisionRadius=0
      boundBone {boneName=dyn_strand_A_2}
      boxExtents=<50 60 60>
      constraintSetup {angularLimitsMax=<15 10 10>, angularLimitsMin=<-15 -10 -10>, angularTarget=<1 0 0>}
      localJointOffset=<4 0 0>
     }]
    relativeSpaceBone {boneName=pelvis}
   }
   n52 = Anim Dynamics - Bone: dyn_strand_A_3(ComponentPose=@n51, Alpha=1, ExternalForce="0,0,0") {
    angularDampingOverride=0.999
    angularSpringConstant=20000
    bAngularSpring=true
    bOverrideAngularDamping=true
    bUseGravityOverride=true
    gravityScale=0.35
    alphaBoolBlend {blendOption=Linear}
    boundBone {boneName=dyn_strand_A_3}
    componentLinearAccScale=<1 1 1>
    componentLinearVelScale=<1 1 1>
    gravityOverride {z=-980}
    physicsBodyDefinitions [{
      sphereCollisionRadius=0
      boundBone {boneName=dyn_strand_A_3}
      boxExtents=<50 60 60>
      constraintSetup {angularLimitsMax=<15 10 10>, angularLimitsMin=<-15 -10 -10>, angularTarget=<1 0 0>}
      localJointOffset=<4 0 0>
     }]
    relativeSpaceBone {boneName=pelvis}
   }
   n53 = Anim Dynamics - Bone: dyn_strand_B_1(ComponentPose=@n52, Alpha=1, ExternalForce="0,0,0") {
    angularDampingOverride=0.999
    angularSpringConstant=20000
    bAngularSpring=true
    bOverrideAngularDamping=true
    bUseGravityOverride=true
    gravityScale=0.35
    alphaBoolBlend {blendOption=Linear}
    boundBone {boneName=dyn_strand_B_1}
    componentLinearAccScale=<1 1 1>
    componentLinearVelScale=<1 1 1>
    gravityOverride {z=-980}
    physicsBodyDefinitions [{
      sphereCollisionRadius=0
      boundBone {boneName=dyn_strand_B_1}
      boxExtents=<50 60 60>
      constraintSetup {angularLimitsMax=<15 0 30>, angularLimitsMin=<-15 -30 -30>, angularTarget=<1 0 0>}
      localJointOffset=<4 0 0>
     }]
    relativeSpaceBone {boneName=pelvis}
   }
   n54 = Anim Dynamics - Bone: dyn_strand_B_2(ComponentPose=@n53, Alpha=1, ExternalForce="0,0,0") {
    angularDampingOverride=0.999
    angularSpringConstant=20000
    bAngularSpring=true
    bOverrideAngularDamping=true
    bUseGravityOverride=true
    gravityScale=0.35
    alphaBoolBlend {blendOption=Linear}
    boundBone {boneName=dyn_strand_B_2}
    componentLinearAccScale=<1 1 1>
    componentLinearVelScale=<1 1 1>
    gravityOverride {z=-980}
    physicsBodyDefinitions [{
      sphereCollisionRadius=0
      boundBone {boneName=dyn_strand_B_2}
      boxExtents=<50 60 60>
      constraintSetup {angularLimitsMax=<15 0 30>, angularLimitsMin=<-15 -30 -30>, angularTarget=<1 0 0>}
      localJointOffset=<4 0 0>
     }]
    relativeSpaceBone {boneName=pelvis}
   }
   n55 = Anim Dynamics - Bone: dyn_strand_B_3(ComponentPose=@n54, Alpha=1, ExternalForce="0,0,0") {
    angularDampingOverride=0.999
    angularSpringConstant=20000
    bAngularSpring=true
    bOverrideAngularDamping=true
    bUseGravityOverride=true
    gravityScale=0.35
    alphaBoolBlend {blendOption=Linear}
    boundBone {boneName=dyn_strand_B_3}
    componentLinearAccScale=<1 1 1>
    componentLinearVelScale=<1 1 1>
    gravityOverride {z=-980}
    physicsBodyDefinitions [{
      sphereCollisionRadius=0
      boundBone {boneName=dyn_strand_B_3}
      boxExtents=<50 60 60>
      constraintSetup {angularLimitsMax=<15 0 30>, angularLimitsMin=<-15 -30 -30>, angularTarget=<1 0 0>}
      localJointOffset=<4 0 0>
     }]
    relativeSpaceBone {boneName=pelvis}
   }
   n56 = Anim Dynamics - Bone: dyn_strand_C_1(ComponentPose=@n55, Alpha=1, ExternalForce="0,0,0") {
    angularDampingOverride=0.999
    angularSpringConstant=20000
    bAngularSpring=true
    bOverrideAngularDamping=true
    bUseGravityOverride=true
    gravityScale=0.35
    alphaBoolBlend {blendOption=Linear}
    boundBone {boneName=dyn_strand_C_1}
    componentLinearAccScale=<1 1 1>
    componentLinearVelScale=<1 1 1>
    gravityOverride {z=-900}
    physicsBodyDefinitions [{
      sphereCollisionRadius=0
      boundBone {boneName=dyn_strand_C_1}
      boxExtents=<50 60 60>
      constraintSetup {angularLimitsMax=<15 0 30>, angularLimitsMin=<-15 -30 -30>, angularTarget=<1 0 0>}
      localJointOffset=<4 0 0>
     }]
    relativeSpaceBone {boneName=pelvis}
   }
   n57 = Anim Dynamics - Bone: dyn_strand_C_2(ComponentPose=@n56, Alpha=1, ExternalForce="0,0,0") {
    angularDampingOverride=0.999
    angularSpringConstant=20000
    bAngularSpring=true
    bOverrideAngularDamping=true
    bUseGravityOverride=true
    gravityScale=0.35
    alphaBoolBlend {blendOption=Linear}
    boundBone {boneName=dyn_strand_C_2}
    componentLinearAccScale=<1 1 1>
    componentLinearVelScale=<1 1 1>
    gravityOverride {z=-900}
    physicsBodyDefinitions [{
      sphereCollisionRadius=0
      boundBone {boneName=dyn_strand_C_2}
      boxExtents=<50 60 60>
      constraintSetup {angularLimitsMax=<15 0 30>, angularLimitsMin=<-15 -30 -30>, angularTarget=<1 0 0>}
      localJointOffset=<4 0 0>
     }]
    relativeSpaceBone {boneName=pelvis}
   }
   n58 = Anim Dynamics - Bone: dyn_strand_C_3(ComponentPose=@n57, Alpha=1, ExternalForce="0,0,0") {
    angularDampingOverride=0.999
    angularSpringConstant=20000
    bAngularSpring=true
    bOverrideAngularDamping=true
    bUseGravityOverride=true
    gravityScale=0.35
    alphaBoolBlend {blendOption=Linear}
    boundBone {boneName=dyn_strand_C_3}
    componentLinearAccScale=<1 1 1>
    componentLinearVelScale=<1 1 1>
    gravityOverride {z=-900}
    physicsBodyDefinitions [{
      sphereCollisionRadius=0
      boundBone {boneName=dyn_strand_C_3}
      boxExtents=<50 60 60>
      constraintSetup {angularLimitsMax=<15 0 30>, angularLimitsMin=<-15 -30 -30>, angularTarget=<1 0 0>}
      localJointOffset=<4 0 0>
     }]
    relativeSpaceBone {boneName=pelvis}
   }
   n59 = Anim Dynamics - Bone: dyn_strand_C_4(ComponentPose=@n58, Alpha=1, ExternalForce="0,0,0") {
    angularDampingOverride=0.999
    angularSpringConstant=20000
    bAngularSpring=true
    bOverrideAngularDamping=true
    bUseGravityOverride=true
    gravityScale=0.35
    alphaBoolBlend {blendOption=Linear}
    boundBone {boneName=dyn_strand_C_4}
    componentLinearAccScale=<1 1 1>
    componentLinearVelScale=<1 1 1>
    gravityOverride {z=-900}
    physicsBodyDefinitions [{
      sphereCollisionRadius=0
      boundBone {boneName=dyn_strand_C_4}
      boxExtents=<50 60 60>
      constraintSetup {angularLimitsMax=<15 0 30>, angularLimitsMin=<-15 -30 -30>, angularTarget=<1 0 0>}
      localJointOffset=<4 0 0>
     }]
    relativeSpaceBone {boneName=pelvis}
   }
   n60 = Anim Dynamics - Bone: dyn_strand_C_5(ComponentPose=@n59, Alpha=1, ExternalForce="0,0,0") {
    angularDampingOverride=0.999
    angularSpringConstant=20000
    bAngularSpring=true
    bOverrideAngularDamping=true
    bUseGravityOverride=true
    gravityScale=0.35
    alphaBoolBlend {blendOption=Linear}
    boundBone {boneName=dyn_strand_C_5}
    componentLinearAccScale=<1 1 1>
    componentLinearVelScale=<1 1 1>
    gravityOverride {z=-900}
    physicsBodyDefinitions [{
      sphereCollisionRadius=0
      boundBone {boneName=dyn_strand_C_5}
      boxExtents=<50 60 60>
      constraintSetup {angularLimitsMax=<15 0 30>, angularLimitsMin=<-15 -30 -30>, angularTarget=<1 0 0>}
      localJointOffset=<4 0 0>
     }]
    relativeSpaceBone {boneName=pelvis}
   }
   n61 = Anim Dynamics - Bone: dyn_strand_D_1(ComponentPose=@n60, Alpha=1, ExternalForce="0,0,0") {
    angularDampingOverride=0.999
    angularSpringConstant=20000
    bAngularSpring=true
    bOverrideAngularDamping=true
    bUseGravityOverride=true
    gravityScale=0.35
    alphaBoolBlend {blendOption=Linear}
    boundBone {boneName=dyn_strand_D_1}
    componentLinearAccScale=<1 1 1>
    componentLinearVelScale=<1 1 1>
    gravityOverride {z=-980}
    physicsBodyDefinitions [{
      sphereCollisionRadius=0
      boundBone {boneName=dyn_strand_D_1}
      boxExtents=<50 60 60>
      constraintSetup {angularLimitsMax=<15 30 30>, angularLimitsMin=<-15 -30 -30>, angularTarget=<1 0 0>}
      localJointOffset=<4 0 0>
     }]
    relativeSpaceBone {boneName=pelvis}
   }
   n62 = Anim Dynamics - Bone: dyn_strand_D_2(ComponentPose=@n61, Alpha=1, ExternalForce="0,0,0") {
    angularDampingOverride=0.999
    angularSpringConstant=20000
    bAngularSpring=true
    bOverrideAngularDamping=true
    bUseGravityOverride=true
    gravityScale=0.35
    alphaBoolBlend {blendOption=Linear}
    boundBone {boneName=dyn_strand_D_2}
    componentLinearAccScale=<1 1 1>
    componentLinearVelScale=<1 1 1>
    gravityOverride {z=-980}
    physicsBodyDefinitions [{
      sphereCollisionRadius=0
      boundBone {boneName=dyn_strand_D_2}
      boxExtents=<50 60 60>
      constraintSetup {angularLimitsMax=<15 30 30>, angularLimitsMin=<-15 -30 -30>, angularTarget=<1 0 0>}
      localJointOffset=<4 0 0>
     }]
    relativeSpaceBone {boneName=pelvis}
   }
   n63 = Anim Dynamics - Bone: dyn_strand_D_3(ComponentPose=@n62, Alpha=1, ExternalForce="0,0,0") {
    angularDampingOverride=0.999
    angularSpringConstant=20000
    bAngularSpring=true
    bOverrideAngularDamping=true
    bUseGravityOverride=true
    gravityScale=0.35
    alphaBoolBlend {blendOption=Linear}
    boundBone {boneName=dyn_strand_D_3}
    componentLinearAccScale=<1 1 1>
    componentLinearVelScale=<1 1 1>
    gravityOverride {z=-980}
    physicsBodyDefinitions [{
      sphereCollisionRadius=0
      boundBone {boneName=dyn_strand_D_3}
      boxExtents=<50 60 60>
      constraintSetup {angularLimitsMax=<15 30 30>, angularLimitsMin=<-15 -30 -30>, angularTarget=<1 0 0>}
      localJointOffset=<4 0 0>
     }]
    relativeSpaceBone {boneName=pelvis}
   }
   n64 = Anim Dynamics - Bone: dyn_strand_D_4(ComponentPose=@n63, Alpha=1, ExternalForce="0,0,0") {
    angularDampingOverride=0.999
    angularSpringConstant=20000
    bAngularSpring=true
    bOverrideAngularDamping=true
    bUseGravityOverride=true
    gravityScale=0.35
    alphaBoolBlend {blendOption=Linear}
    boundBone {boneName=dyn_strand_D_4}
    componentLinearAccScale=<1 1 1>
    componentLinearVelScale=<1 1 1>
    gravityOverride {z=-980}
    physicsBodyDefinitions [{
      sphereCollisionRadius=0
      boundBone {boneName=dyn_strand_D_4}
      boxExtents=<50 60 60>
      constraintSetup {angularLimitsMax=<15 30 30>, angularLimitsMin=<-15 -30 -30>, angularTarget=<1 0 0>}
      localJointOffset=<4 0 0>
     }]
    relativeSpaceBone {boneName=pelvis}
   }
   n65 = Anim Dynamics - Bone: dyn_strand_D_5(ComponentPose=@n64, Alpha=1, ExternalForce="0,0,0") {
    angularDampingOverride=0.999
    angularSpringConstant=20000
    bAngularSpring=true
    bOverrideAngularDamping=true
    bUseGravityOverride=true
    gravityScale=0.35
    alphaBoolBlend {blendOption=Linear}
    boundBone {boneName=dyn_strand_D_5}
    componentLinearAccScale=<1 1 1>
    componentLinearVelScale=<1 1 1>
    gravityOverride {z=-980}
    physicsBodyDefinitions [{
      sphereCollisionRadius=0
      boundBone {boneName=dyn_strand_D_5}
      boxExtents=<50 60 60>
      constraintSetup {angularLimitsMax=<15 30 30>, angularLimitsMin=<-15 -30 -30>, angularTarget=<1 0 0>}
      localJointOffset=<4 0 0>
     }]
    relativeSpaceBone {boneName=pelvis}
   }
   n66 = Get Rotation_0() -> (Rotation_0:Rotator)
   n67 = "Transform (Modify) Bone - Bone: dyn_strand_C_1"!(ComponentPose=@n65, Translation="0,0,0", Rotation=@n66, Scale="1,1,1") -> (Pose) {
    rotationMode=BMM_Additive
    rotationSpace=BCS_BoneSpace
    alphaBoolBlend {blendOption=Linear}
    boneToModify {boneName=dyn_strand_C_1}
    rotation {yaw=-20}
   }
   n68 = "Transform (Modify) Bone - Bone: dyn_strand_C_2"!(ComponentPose=@n67, Translation="0,0,0", Rotation=@n66, Scale="1,1,1") -> (Pose) {
    rotationMode=BMM_Additive
    rotationSpace=BCS_BoneSpace
    alphaBoolBlend {blendOption=Linear}
    boneToModify {boneName=dyn_strand_C_2}
    rotation {yaw=-20}
   }
   n69 = Get Rotation_1() -> (Rotation_1:Rotator)
   n70 = "Transform (Modify) Bone - Bone: dyn_strand_D_1"!(ComponentPose=@n68, Translation="0,0,0", Rotation=@n69, Scale="1,1,1") -> (Pose) {
    rotationMode=BMM_Additive
    rotationSpace=BCS_BoneSpace
    alphaBoolBlend {blendOption=Linear}
    boneToModify {boneName=dyn_strand_D_1}
    rotation {yaw=-20}
   }
   n71 = "Transform (Modify) Bone - Bone: dyn_strand_D_2"!(ComponentPose=@n70, Translation="0,0,0", Rotation=@n69, Scale="1,1,1") -> (Pose) {
    rotationMode=BMM_Additive
    rotationSpace=BCS_BoneSpace
    alphaBoolBlend {blendOption=Linear}
    boneToModify {boneName=dyn_strand_D_2}
    rotation {yaw=-20}
   }
   n72 = "Transform (Modify) Bone - Bone: dyn_strand_A_1"!(ComponentPose=@n71, Translation="0,0,0", Rotation=Get Rotation_2() -> (Rotation_2:Rotator), Scale="1,1,1") -> (Pose) {
    rotationMode=BMM_Additive
    rotationSpace=BCS_BoneSpace
    alphaBoolBlend {blendOption=Linear}
    boneToModify {boneName=dyn_strand_A_1}
    rotation {yaw=-20}
   }
   n73 = Get Rotation_3() -> (Rotation_3:Rotator)
   n74 = "Transform (Modify) Bone - Bone: dyn_strand_B_1"!(ComponentPose=@n72, Translation="0,0,0", Rotation=@n73, Scale="1,1,1") -> (Pose) {
    rotationMode=BMM_Additive
    rotationSpace=BCS_BoneSpace
    alphaBoolBlend {blendOption=Linear}
    boneToModify {boneName=dyn_strand_B_1}
    rotation {yaw=-20}
   }
   n75 = "Transform (Modify) Bone - Bone: dyn_strand_B_2"!(ComponentPose=@n74, Translation="0,0,0", Rotation=@n73, Scale="1,1,1") -> (Pose) {
    rotationMode=BMM_Additive
    rotationSpace=BCS_BoneSpace
    alphaBoolBlend {blendOption=Linear}
    boneToModify {boneName=dyn_strand_B_2}
    rotation {yaw=-20}
   }
   Output Pose(Result=Component To Local(ComponentPose=@n75)) {layerGroup=None}
  ]}
 EventGraph {nodes [
   Event BlueprintUpdateAnimation()
   > n1 = Sequence!(execute) -> (then_0, then_1, then_2:Exec)
   n2 = float / float(A=GetRealTimeSeconds(), B=-500) {is=title K2Node_CallFunction}
   > Set Spin(Spin:Rotator=MakeRotator(Roll=float * float(A=@n2, B=30000))) -> (Output_Get:Rotator)
   n3 = Branch(execute=@n1.then_1, Condition=Get bIsOnGround() -> (bIsOnGround:bool))
   > Set groundCrown(groundCrown:double=0.35) -> (Output_Get:double)
   Set groundCrown(execute=@n3.else, groundCrown=1)
   n4 = float * float(A=GetGameTimeInSeconds(), B=1.5) {is=title K2Node_PromotableOperator}
   n5 = float + float(A=float * float(A="Sin (Radians)"(A=@n4), B=10) {is=title K2Node_CommutativeAssociativeBinaryOperator}, B=10)
   Set sineRotation high1(execute=@n1.then_2, sineRotation high1:Rotator=MakeRotator(Yaw=@n5)) -> (Output_Get:Rotator)
   n6 = float + float(A=@n4, B=-1)
   n7 = float + float(A=float * float(A="Sin (Radians)"(A=@n6), B=10), B=10)
   > Set sineRotation high2(sineRotation high2:Rotator=MakeRotator(Yaw=@n7)) -> (Output_Get:Rotator)
   n8 = float + float(A=@n4, B=-2)
   n9 = float + float(A=float * float(A="Sin (Radians)"(A=@n8), B=10), B=10)
   > Set sineRotation high3(sineRotation high3:Rotator=MakeRotator(Yaw=@n9)) -> (Output_Get:Rotator)
   n10 = float + float(A=float * float(A="Sin (Radians)"(A=@n4), B=15), B=10)
   > Set sineRotation1(sineRotation1:Rotator=MakeRotator(Pitch=@n10, Yaw=@n10)) -> (Output_Get:Rotator)
   n11 = float + float(A=@n4, B=-1)
   n12 = float + float(A=float * float(A="Sin (Radians)"(A=@n11), B=21), B=10)
   > Set sineRotation2(sineRotation2:Rotator=MakeRotator(Pitch=@n12, Yaw=@n12)) -> (Output_Get:Rotator)
   n13 = float + float(A=@n4, B=-1.5)
   n14 = float + float(A=float * float(A="Sin (Radians)"(A=@n13), B=30), B=10)
   > Set sineRotation3(sineRotation3:Rotator=MakeRotator(Pitch=@n14, Yaw=@n14)) -> (Output_Get:Rotator)
   n15 = float * float(A="Sin (Radians)"(A=float + float(A=@n4, B=-2)), B=10)
   > Set sineRotation4(sineRotation4:Rotator=MakeRotator(Pitch=float + float(A=@n15, B=10))) -> (Output_Get:Rotator)
   n16 = float * float(A="Sin (Radians)"(A=float + float(A=@n4, B=-2.5)), B=10)
   > Set sineRotation5(sineRotation5:Rotator=MakeRotator(Pitch=float + float(A=@n16, B=10))) -> (Output_Get:Rotator)
   n17 = float - float(A=@n5, B=9)
   > Set sineTranslation high1(sineTranslation high1:Vector=MakeVector(X=float / float(A=@n17, B=3) {is=title K2Node_PromotableOperator}, Y=0)) -> (Output_Get:Vector)
   n18 = float - float(A=@n7, B=9)
   > Set sineTranslation high2(sineTranslation high2:Vector=MakeVector(X=float / float(A=@n18, B=3), Y=0)) -> (Output_Get:Vector)
   n19 = float - float(A=@n9, B=9)
   > Set sineTranslation high3(sineTranslation high3:Vector=MakeVector(X=float / float(A=@n19, B=3), Y=0)) -> (Output_Get:Vector)
   n20 = float * float(A=GetGameTimeInSeconds(), B=1.5) {is=title K2Node_PromotableOperator}
   n21 = "Sin (Radians)"(A=float + float(A=@n20, B=-2.5))
   MakeRotator(Pitch=float + float(A=float * float(A=@n21, B=10) {is=title K2Node_CommutativeAssociativeBinaryOperator}, B=10))
   n22 = "Sin (Radians)"(A=float + float(A=@n20, B=-1))
   MakeRotator(Yaw=float + float(A=float * float(A=@n22, B=5), B=10))
   n23 = "Sin (Radians)"(A=float + float(A=@n20, B=-2))
   MakeRotator(Yaw=float + float(A=float * float(A=@n23, B=5), B=10))
   n24 = "Sin (Radians)"(A=float + float(A=@n20, B=-1))
   MakeRotator(Pitch=float + float(A=float * float(A=@n24, B=7), B=10))
   n25 = "Sin (Radians)"(A=float + float(A=@n20, B=-1.5))
   MakeRotator(Pitch=float + float(A=float * float(A=@n25, B=10), B=10))
   n26 = "Sin (Radians)"(A=float + float(A=@n20, B=-2))
   MakeRotator(Pitch=float + float(A=float * float(A=@n26, B=10), B=10))
   n27 = "Sin (Radians)"(A=@n20)
   MakeRotator(Pitch=float + float(A=float * float(A=@n27, B=5), B=10))
   MakeRotator(Yaw=float + float(A=float * float(A=0, B=5), B=10))
  ]}
]
variables [
 groundCrown {type=number}
 Rotation {type=Rotator, default=<rot 40 0 0>}
 Rotation_0 {type=Rotator, default=<rot 5 0 0>}
 Rotation_1 {type=Rotator, default=<rot 0 -5 0>}
 Rotation_2 {type=Rotator, default=<rot 10 10 0>}
 Rotation_3 {type=Rotator, default=<rot 10 -15 0>}
 sineRotation high1 {type=Rotator}
 sineRotation high2 {type=Rotator}
 sineRotation high3 {type=Rotator}
 sineRotation1 {type=Rotator}
 sineRotation2 {type=Rotator}
 sineRotation3 {type=Rotator}
 sineRotation4 {type=Rotator}
 sineRotation5 {type=Rotator}
 sineTranslation high1 {type=Vector}
 sineTranslation high2 {type=Vector}
 sineTranslation high3 {type=Vector}
 Spin {type=Rotator}
]
